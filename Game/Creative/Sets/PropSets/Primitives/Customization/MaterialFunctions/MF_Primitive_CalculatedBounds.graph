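# Material function for Primitive props that outputs world position, local position, world normals and calculated bounds.
# Calculated Bounds offsets pre-skinned position from bounds Min, but divides by Max minus Max, a zero span as wired.
# Stage switches feed interpolated vertex values to pixel shaders; World Normals are absolute there but signed in vertex shaders.
# Leftover triplanar-style axis selection, blend weight and pre-skinned normal branches reach no output.
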MaterialFunction /Game/Creative/Sets/PropSets/Primitives/Customization/MaterialFunctions/MF_Primitive_CalculatedBounds
graph {nodes [
  n1 = WorldPosition()
  n2 = MaterialFunctionCall(Float3=Frac(None=Divide(A=@n1.XYZ) {constB=100})) {materialFunction=&BreakOutFloat3Components}
  AppendVector(A=@n2.R, B=OneMinus(None=@n2.B))
  n3 = Abs(None=VertexNormalWS())
  n4 = MaterialFunctionCall(Float3=@n3) {materialFunction=&BreakOutFloat3Components}
  Divide(A=@n3, B=Add(A=Add(A=@n4.R, B=@n4.R), B=@n4.B))
  n5 = PreSkinnedPosition()
  n6 = ShaderStageSwitch(PixelShader=VertexInterpolator(VS=@n5).PS, VertexShader=@n5)
  n7 = PreSkinnedLocalBounds()
  FunctionOutput(None=Divide(A=Subtract(A=@n6, B=@n7.Min), B=Subtract(A=@n7.Max, B=@n7.Max))) {outputName=Calculated Bounds, sortPriority=2}
  n8 = VertexNormalWS()
  n9 = Abs(None=@n8)
  FunctionOutput(None=ShaderStageSwitch(PixelShader=VertexInterpolator(VS=@n9).PS, VertexShader=@n8)) {outputName=World Normals, sortPriority=3}
  n10 = WorldPosition()
  FunctionOutput(None=@n10.XYZ) {outputName=WorldPosition}
  FunctionOutput(None=TransformPosition(None=@n10.XYZ) {transformSourceType=TRANSFORMPOSSOURCE_World}) {outputName=LocalPosition, sortPriority=1}
  If(
   A=@n4.B
   B=Max(A=@n4.R, B=@n4.R)
   "A > B"=ComponentMask(None=@n1.XYZ) {desc=YX, g=true, r=true}
   "A < B"=If(
    A=@n4.R
    B=@n4.R
    "A > B"=ComponentMask(None=@n1.XYZ) {b=true, desc=YZ, g=true}
    "A < B"=ComponentMask(None=@n1.XYZ) {b=true, desc=XZ, r=true}
   )
  )
  n11 = PreSkinnedNormal()
  ShaderStageSwitch(PixelShader=VertexInterpolator(VS=@n11).PS, VertexShader=@n11)
  Transform(None=Abs(None=PreSkinnedNormal()))
 ]}
