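# Material function that triplanar-projects a diffuse and a normal texture for Creative primitive props, with shared scale, per-axis rotation and transition contrast.
# Position and normal inputs default to MF_Primitive_CalculatedBounds, feeding the wrap WorldAlignedTexture for RGBA and PreSkinnedAlignedNormal for normals.
# The projected normal's XY is added onto the material normal, blended by an intensity input.
# RotationY drives RotZAxis and RotationZ drives RotYAxis in both projections.
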MaterialFunction /Game/Creative/Sets/PropSets/Primitives/Customization/MaterialFunctions/MF_Primitive_TriplanarUVs
graph {nodes [
  Constant3Vector() {constant=<rgba 0 0 0.5 0>}
  n1 = StaticBool() {value=true}
  n2 = FunctionInput(Preview=TextureObject() {texture=/Game/Weapons/WeaponSkins/Wraps/Textures/UVCheaker}) {bUsePreviewValueAsDefault=true, inputName=TextureObject_Diffuse, inputType=FunctionInput_Texture2D, sortPriority=-20}
  n3 = FunctionInput(Preview=Constant() {r=1}) {
   bUsePreviewValueAsDefault=true
   description=Size of texture in worldspace units
   inputName=TextureScale
   sortPriority=-19
   previewValue=<64 64 64 6>
  }
  n4 = MaterialFunctionCall() {materialFunction=&/MF_Primitive_CalculatedBounds}
  n5 = FunctionInput(Preview=@n4.WorldPosition) {bUsePreviewValueAsDefault=true, inputName=CalculatedBounds, sortPriority=32, previewValue=<1 0 0 1>}
  n6 = FunctionInput(Preview=@n4.World Normals) {bUsePreviewValueAsDefault=true, inputName=WorldNormals, sortPriority=33, previewValue=<1 0 0 1>}
  n7 = VectorParameter() {group=TriplanarUVs, parameterName=Rotation, defaultValue=<rgba 1 1 1 0>}
  n8 = FunctionInput(Preview=@n7.G) {bUsePreviewValueAsDefault=true, inputName=RotationY, inputType=FunctionInput_Scalar, sortPriority=35}
  n9 = FunctionInput(Preview=@n7.R) {bUsePreviewValueAsDefault=true, inputName=RotationX, inputType=FunctionInput_Scalar, sortPriority=34}
  n10 = FunctionInput(Preview=@n7.B) {bUsePreviewValueAsDefault=true, inputName=RotationZ, inputType=FunctionInput_Scalar, sortPriority=36}
  n11 = FunctionInput() {
   bUsePreviewValueAsDefault=true
   inputName=ProjectionTransitionContrast
   inputType=FunctionInput_Scalar
   sortPriority=38
   previewValue=<1 0 0 1>
  }
  FunctionOutput(None=ComponentMask(None=MaterialFunctionCall(
     Export Float 4=@n1
     TextureObject=@n2
     TextureSize=@n3
     WorldPosition=@n5
     World Space Normal=@n6
     RotZAxis=@n8
     RotXAxis=@n9
     RotYAxis=@n10
     ProjectionTransitionContrast=@n11
    ) {materialFunction=/Game/Weapons/WeaponSkins/Wraps/MaterialFunctions/WorldAlignedTexture}.XYZ Texture) {a=true, b=true, g=true, r=true}) {outputName=RGBA}
  n12 = TextureSample() {samplerSource=SSM_Wrap_WorldGroupSettings, samplerType=SAMPLERTYPE_Normal, texture=&FlatNormalBlue}
  n13 = ShadingPathSwitch(Default=FunctionInput(Preview=@n12.RGB) {bUsePreviewValueAsDefault=true, inputName=Material Normals, sortPriority=49}) {desc=Migrated from FeatureLevelSwitch}
  n14 = TextureObject() {samplerType=SAMPLERTYPE_Normal, texture=&FlatNormalBlue}
  n15 = Multiply(
   A=MaterialFunctionCall!(
    TextureObject=FunctionInput(Preview=@n14) {bUsePreviewValueAsDefault=true, inputName=NormalTexture, inputType=FunctionInput_Texture2D, sortPriority=50}
    TextureSize=@n3
    RotYAxis=@n10
    RotZAxis=@n8
    RotXAxis=@n9
    PreSkinNormal=@n6
    ProjectionTransitionContrast=@n11
    WorldPosition=@n5
   ) {materialFunction=/Game/Packages/Fortress_SharedMaterials/Base_Material_Functions/BaseCustomizationFunctions/PreSkinnedAlignedNormal}.XYZ Texture
   B=Constant3Vector() {constant=<rgba 1 1 0 0>}
  )
  n16 = FunctionInput() {bUsePreviewValueAsDefault=true, inputName=Normal Texture Intensity, inputType=FunctionInput_Scalar, sortPriority=51, previewValue=<1 0 0 1>}
  FunctionOutput(None=LinearInterpolate(A=@n13, B=Add(A=@n13, B=@n15), Alpha=@n16)) {outputName=Normal Texture, sortPriority=10}
 ]}
