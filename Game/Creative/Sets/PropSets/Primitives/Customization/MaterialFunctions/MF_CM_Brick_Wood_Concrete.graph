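# Material function adding a brick, wood or concrete look to Creative primitive props, passing incoming attributes through unless the Primitive_Brick_Wood_Concrete static switch is on.
# Scaled positions and normals from MF_Primitive_CalculatedBounds drive triplanar projection of diffuse, normal and noise textures, defaulting to Apollo brick.
# Diffuse is hue-shifted and tinted; noise red and alpha feed roughness and specular ranges.
# The switched-on branch also sets metallic and overrides emissive with a constant.
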MaterialFunction /Game/Creative/Sets/PropSets/Primitives/Customization/MaterialFunctions/MF_CM_Brick_Wood_Concrete
graph {nodes [
  n1 = VectorParameter() {group=BWC, parameterName=CM_BWC_Color_Multi, defaultValue=<rgba 1 1 1 0>}
  AppendVector(A=@n1.RGB, B=@n1.RGB)
  n2 = GetMaterialAttributes(None=FunctionInput() {inputType=FunctionInput_MaterialAttributes}) {attributeGetTypes [#1]}
  n3 = SetMaterialAttributes!(
   MaterialAttributes=@n2.MaterialAttributes
   BaseColor=Multiply(A=NamedRerouteUsage(), B=VectorParameter() {group=BWC, parameterName=CM_BWC_Color_Multi, defaultValue=<rgba 1 1 1 0>}.RGB)
   Specular=NamedRerouteUsage()
   Roughness=NamedRerouteUsage()
   Normal=NamedRerouteUsage()
   Metallic=ScalarParameter() {parameterName=CM_BWC_Metallic}
   EmissiveColor=Constant()
  ) {attributeSetTypes [
    #2
    #3
    #4
    #1
    #5
    #6
   ]}
  FunctionOutput(None=StaticSwitchParameter(True=@n3, False=@n2.MaterialAttributes) {group=Customization, parameterName=Primitive_Brick_Wood_Concrete})
  NamedRerouteDeclaration(None=@n2.Normal) {name=InNormal, nodeColor=<rgba 0 0.171 1 1>}
  NamedRerouteDeclaration(None=@n2.MaterialAttributes) {name=Material Attribute, nodeColor=<rgba 0.7323 0 1 1>}
  n4 = TextureObjectParameter() {
   group=BWC
   parameterName=CM_BWC_NoiseTex
   texture=/Game/Creative/Sets/PropSets/Primitives/Customization/Textures/NeonGasCloud/T_CM_NeonGasCloud_M2
  }
  n5 = MaterialFunctionCall!(
   TextureObject=@n4
   TextureSize=ScalarParameter() {defaultValue=1, group=BWC, parameterName=CM_BWC_Noise_UVTile}
   WorldPosition=NamedRerouteUsage()
   World Space Normal=NamedRerouteUsage()
   ProjectionTransitionContrast=ScalarParameter() {defaultValue=10, group=BWC, parameterName=CM_BWC_Noise_TriplannarContrast}
  ) {materialFunction=&/MF_Primitive_TriplanarUVSimple}
  NamedRerouteDeclaration(None=ComponentMask(None=@n5.XYZ Texture) {r=true}) {name=Roughness, nodeColor=<rgba 1 0.1172 0 1>}
  NamedRerouteDeclaration(None=ComponentMask(None=@n5.XYZ Texture) {a=true}) {name=Specular, nodeColor=<rgba 0 0.6016 1 1>}
  n6 = TextureObjectParameter() {
   group=BWC
   parameterName=CM_BWC_NormalTexture
   samplerType=SAMPLERTYPE_Normal
   texture=/Game/Environments/Apollo/Textures/T_Apollo_Brick_Wall_Exterior_N
  }
  NamedRerouteDeclaration(None=MaterialFunctionCall!(
    TextureObject_Diffuse
    TextureScale=ScalarParameter() {defaultValue=1, group=BWC, parameterName=CM_BWC_BaseUVTile}
    CalculatedBounds=NamedRerouteUsage()
    WorldNormals=NamedRerouteUsage()
    RotationX
    RotationY
    RotationZ
    ProjectionTransitionContrast=ScalarParameter() {defaultValue=100, group=BWC, parameterName=CM_BWC_BaseTriplannarContrast}
    Material Normals=NamedRerouteUsage()
    NormalTexture=@n6
    Normal Texture Intensity=ScalarParameter() {defaultValue=1, group=BWC, parameterName=CM_BWC_NormalIntensity}
   ) {materialFunction=&/MF_Primitive_TriplanarUVs}.Normal Texture) {name=FinalNormal, nodeColor=<rgba 0.8047 0 1 1>}
  n7 = MaterialFunctionCall() {materialFunction=&/MF_Primitive_CalculatedBounds}
  n8 = Divide(A=@n7.WorldPosition, B=ScalarParameter() {defaultValue=100, group=BWC, parameterName=CM_BWC_WorldScale})
  NamedRerouteDeclaration(None=@n8) {name=UVS, nodeColor=<rgba 0.9922 0 1 1>}
  NamedRerouteDeclaration(None=@n8) {name=CalculatedBounds, nodeColor=<rgba 1 0.1964 0 1>}
  NamedRerouteDeclaration(None=@n7.World Normals) {name=WorldNorm, nodeColor=<rgba 0 1 0.9531 1>}
  n9 = TextureObjectParameter() {group=BWC, parameterName=CM_BWC_DiffuseTex, texture=/Game/Environments/Apollo/Textures/T_Apollo_Brick_Wall_Exterior_D}
  n10 = ScalarParameter() {defaultValue=1, group=BWC, parameterName=CM_BWC_BaseUVTile}
  n11 = NamedRerouteUsage()
  n12 = NamedRerouteUsage()
  n13 = ScalarParameter() {defaultValue=100, group=BWC, parameterName=CM_BWC_BaseTriplannarContrast}
  NamedRerouteDeclaration(None=MaterialFunctionCall(
    Hue Shift Percentage=ScalarParameter() {group=BWC, parameterName=CM_BWC_HueShift, sliderMax=1}
    Texture=MaterialFunctionCall(TextureObject=@n9, TextureSize=@n10, WorldPosition=@n11, World Space Normal=@n12, ProjectionTransitionContrast=@n13) {materialFunction=&MF_Primitive_TriplanarUVSimple}.XYZ Texture
   ) {materialFunction=&HueShift}.Result) {name=Base Color, nodeColor=<rgba 0 1 0.1094 1>}
  NamedRerouteDeclaration(None=LinearInterpolate(
    A=ScalarParameter() {group=BWC, parameterName=CM_BWC_Specular}
    B=ScalarParameter() {defaultValue=1, group=BWC, parameterName=CM_BWC_SpecularHi}
    Alpha=NamedRerouteUsage()
   )) {name=FinalSpec, nodeColor=<rgba 0 0.4072 0.3968 1>}
  NamedRerouteDeclaration(None=LinearInterpolate(
    A=ScalarParameter() {defaultValue=1, group=BWC, parameterName=CM_BWC_Roughness}
    B=ScalarParameter() {defaultValue=1, group=BWC, parameterName=CM_BWC_RoughnessHi}
    Alpha=NamedRerouteUsage()
   )) {name=FinalRough, nodeColor=<rgba 0.4072 0.2831 0 1>}
 ]}
