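# Material function for customizable Primitive props that outputs a single music-reactive value.
# It reads Spectrum_Music and Amplitude_Music from the MPC_Audio_Reactive_Parameters collection in the Athena audio analysis events folder.
# The spectrum's R and B channels are each summed twice and divided by four, giving their average, while G and A go unused.
# That average is added to the amplitude to form the output.
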MaterialFunction /Game/Creative/Sets/PropSets/Primitives/Customization/MaterialFunctions/MF_Primitive_AudioSetup
graph {nodes [
  n1 = MaterialFunctionCall(Float4=CollectionParameter() {collection=/Game/Athena/Events/AudioAnalysis/MPC_Audio_Reactive_Parameters, parameterName=Spectrum_Music}) {materialFunction=&BreakOutFloat4Components}
  n2 = Add(A=Add(A=@n1.R, B=@n1.R), B=Add(A=@n1.B, B=@n1.B))
  FunctionOutput(None=Add(
    A=CollectionParameter() {collection=&MPC_Audio_Reactive_Parameters, parameterName=Amplitude_Music}
    B=Divide(A=@n2) {constB=4}
   ))
 ]}
