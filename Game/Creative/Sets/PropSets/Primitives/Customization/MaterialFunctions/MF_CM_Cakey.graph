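# Creative primitive-prop customization material function that, under the Primitive_Cakey static switch, turns input material attributes into a frosted cake look.
# Triplanar textures build separate cake and frosting layers, combined with BlendMaterialAttributes.
# An up-facing mask, broken up by noise and remapped through a curve atlas, appears to place the frosting.
# Frosting color passes through a hue shift, apparently music-reactive via audio parameters, and its normal blends in an additional normal.
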MaterialFunction /Game/Creative/Sets/PropSets/Primitives/Customization/MaterialFunctions/MF_CM_Cakey
graph {nodes [
  n1 = NamedRerouteUsage()
  Add(A=VertexNormalWS(), B=Multiply(A=@n1, B=ScalarParameter() {defaultValue=0.5, group=Cakey, parameterName=CM_Cakey_UPZNoiseAmount}))
  n2 = GetMaterialAttributes(None=FunctionInput() {inputType=FunctionInput_MaterialAttributes}) {attributeGetTypes [#1]}
  n3 = SetMaterialAttributes!(
   MaterialAttributes=@n2.MaterialAttributes
   BaseColor=NamedRerouteUsage()
   Specular=NamedRerouteUsage()
   Roughness=NamedRerouteUsage()
   Normal=NamedRerouteUsage()
   Metallic=ScalarParameter() {group=Cakey, parameterName=CM_Cakey_Metallic}
   EmissiveColor=ScalarParameter() {group=Cakey, parameterName=CM_Cakey_Emissive}
  ) {attributeSetTypes [
    #2
    #3
    #4
    #1
    #5
    #6
   ]}
  n4 = SetMaterialAttributes(
   MaterialAttributes=NamedRerouteUsage()
   BaseColor=NamedRerouteUsage()
   Specular=NamedRerouteUsage()
   Roughness=ScalarParameter() {defaultValue=0.325, group=Cakey, parameterName=CM_Cakey_FrostRoughness}
   Normal=NamedRerouteUsage()
   Metallic=ScalarParameter() {group=Cakey, parameterName=CM_Cakey_FrostMetallic}
   EmissiveColor=ScalarParameter() {group=Cakey, parameterName=CM_Cakey_FrostEmissive}
  ) {attributeSetTypes [
    #2
    #3
    #4
    #1
    #5
    #6
   ]}
  n5 = NamedRerouteUsage()
  FunctionOutput(None=StaticSwitchParameter(True=BlendMaterialAttributes(A=@n3, B=@n4, Alpha=@n5), False=@n2.MaterialAttributes) {group=Customization, parameterName=Primitive_Cakey})
  n6 = NamedRerouteUsage()
  n7 = ComponentMask(None=Add(A=ComponentMask(None=@n6) {b=true}, B=ScalarParameter() {defaultValue=-0.85, group=Cakey, parameterName=CM_Cakey_UPZOffset})) {b=true}
  n8 = Multiply(A=@n1, B=ScalarParameter() {defaultValue=0.5, group=Cakey, parameterName=CM_Cakey_UPZNoiseAmount})
  n9 = CurveAtlasRowParameter(CurveTime=MaterialFunctionCall(In=Add(A=@n7, B=@n8), Contrast=ScalarParameter() {defaultValue=2, group=Cakey, parameterName=CM_Cakey_UPZContrast}) {materialFunction=&CheapContrast}.Result) {
   atlas=/Game/Creative/Sets/PropSets/Primitives/Customization/Textures/ColorCurve/CC_CM_Primitives_01
   curve=/Game/Creative/Sets/PropSets/Primitives/Customization/Textures/ColorCurve/C_CM_UPZNormal
   defaultValue=11
   group=Wrap_FlapjackWrang
   parameterName=FJW_Syrup_RemaptoColorCurve ForNormal
  }
  NamedRerouteDeclaration(None=Saturate(None=Multiply(A=@n9.G, B=@n9.G))) {name=UpZ, nodeColor=<rgba 0.0312 1 0 1>}
  NamedRerouteDeclaration(None=@n2.MaterialAttributes) {name=Material Attribute, nodeColor=<rgba 0.6641 0 1 1>}
  NamedRerouteDeclaration(None=LinearInterpolate(
    A=ScalarParameter() {group=Cakey, parameterName=CM_Cakey_FrostSpecular}
    B=ScalarParameter() {defaultValue=1, group=Cakey, parameterName=CM_Cakey_CakeSpecularHi}
    Alpha=NamedRerouteUsage()
   )) {name=FrostFinalSpec, nodeColor=<rgba 0 0.4072 0.3968 1>}
  n10 = Multiply(A=Time(), B=ScalarParameter() {defaultValue=0.125, group=Cakey, parameterName=CM_Cakey_EventHueshiftAmount})
  n11 = Divide(A=Add(A=Sine(None=@n10)))
  n12 = NamedRerouteUsage()
  NamedRerouteDeclaration(None=Add(A=LinearInterpolate(B=@n11, Alpha=@n12), B=Saturate(None=@n12))) {name=React, nodeColor=<rgba 0 1 0.6016 1>}
  n13 = TextureObjectParameter() {group=Cakey, parameterName=CM_Cakey_CakeTex, texture=/Game/Creative/Sets/PropSets/Primitives/Customization/Textures/Cakey/T_CM_Cakey_BC}
  n14 = ScalarParameter() {defaultValue=3, group=Cakey, parameterName=CM_Cakey_CakeUVTile}
  n15 = NamedRerouteUsage()
  n16 = NamedRerouteUsage()
  n17 = Constant()
  n18 = ScalarParameter() {defaultValue=4, group=Cakey, parameterName=CM_Cakey_TriplannarContrast}
  n19 = TextureObjectParameter() {
   group=Cakey
   parameterName=CM_Cakey_CakeNormalTex
   samplerType=SAMPLERTYPE_Normal
   texture=/Game/Creative/Sets/PropSets/Primitives/Customization/Textures/Cakey/T_CM_Cakey_N
  }
  n20 = MaterialFunctionCall!(
   TextureObject_Diffuse=@n13
   TextureScale=@n14
   CalculatedBounds=@n15
   WorldNormals=@n16
   RotationX=@n17
   RotationY=@n17
   RotationZ=@n17
   ProjectionTransitionContrast=@n18
   Material Normals=NamedRerouteUsage()
   NormalTexture=@n19
   Normal Texture Intensity=ScalarParameter() {defaultValue=1, group=Cakey, parameterName=CM_Cakey_CakeNormalIntensity}
  ) {materialFunction=&/MF_Primitive_TriplanarUVs}
  NamedRerouteDeclaration(None=ComponentMask(None=@n20.RGBA) {a=true}) {name=CakeRoughness, nodeColor=<rgba 0 0.8594 1 1>}
  n21 = TextureObjectParameter() {group=Cakey, parameterName=CM_Cakey_CakeMatTex, texture=/Game/Creative/Sets/PropSets/Primitives/Customization/Textures/Cakey/T_CM_Cakey_M1}
  NamedRerouteDeclaration(None=ComponentMask(None=MaterialFunctionCall!(TextureObject=@n21, TextureSize=@n14, WorldPosition=@n15, World Space Normal=@n16, ProjectionTransitionContrast=@n18) {materialFunction=&/MF_Primitive_TriplanarUVSimple}.XYZ Texture) {r=true}) {name=CakeSpecular, nodeColor=<rgba 0 0.6016 1 1>}
  n22 = TextureObjectParameter() {
   group=Cakey
   parameterName=CM_Cakey_NoiseMatTex
   texture=/Game/Creative/Sets/PropSets/Primitives/Customization/Textures/Cakey/T_CM_Cakey_Noise
  }
  n23 = MaterialFunctionCall(
   TextureObject=@n22
   TextureSize=ScalarParameter() {defaultValue=3, group=Cakey, parameterName=CM_Cakey_NoiseUVTile}
   WorldPosition=@n15
   World Space Normal=@n16
   ProjectionTransitionContrast=@n18
  ) {materialFunction=&MF_Primitive_TriplanarUVSimple}
  NamedRerouteDeclaration(None=ComponentMask(None=@n23.XYZ Texture) {a=true}) {name=NoiseMask, nodeColor=<rgba 0 0.6016 1 1>}
  NamedRerouteDeclaration(None=NamedRerouteUsage()) {name=UVs, nodeColor=<rgba 0 1 0.4375 1>}
  NamedRerouteDeclaration(None=ComponentMask(None=@n20.RGBA) {b=true, g=true, r=true}) {name=CakeBaseColor, nodeColor=<rgba 0 0.8594 1 1>}
  NamedRerouteDeclaration(None=@n20.Normal Texture) {name=CakeNormal, nodeColor=<rgba 0 0.8594 1 1>}
  n24 = MaterialFunctionCall() {materialFunction=&/MF_Primitive_CalculatedBounds}
  NamedRerouteDeclaration(None=@n24.Calculated Bounds) {name=Calcs, nodeColor=<rgba 0.4766 0 1 1>}
  NamedRerouteDeclaration(None=@n23.XYZ Texture) {name=Masks, nodeColor=<rgba 0 1 0.2031 1>}
  n25 = CollectionParameter() {collection=/Game/Athena/Events/AudioAnalysis/MPC_Audio_Reactive_Parameters, parameterName=Amplitude_Music}
  n26 = MaterialFunctionCall() {materialFunction=&/MF_Primitive_AudioSetup}
  n27 = ScalarParameter() {defaultValue=0.1, group=Cakey, parameterName=CM_Cakey_EventAudioBlend}
  n28 = Min(A=Multiply(
    A=LinearInterpolate(A=@n25, B=@n26.Result, Alpha=@n27) {constAlpha=0.25}
    B=ScalarParameter() {defaultValue=1, group=Cakey, parameterName=CM_Cakey_EventAudioMultiplier}
   )) {constB=30, desc=prevents crazy values}
  NamedRerouteDeclaration(None=Saturate(None=Abs(None=@n28))) {name=Hueshift, nodeColor=<rgba 0.6172 0 1 1>}
  n29 = TextureObjectParameter() {group=Cakey, parameterName=CM_Cakey_FrostTex, texture=/Game/Creative/Sets/PropSets/Primitives/Customization/Textures/Cakey/T_CM_Cakey_BC2}
  n30 = Constant()
  n31 = TextureObjectParameter() {
   group=Cakey
   parameterName=CM_Cakey_FrostNormalTex
   samplerType=SAMPLERTYPE_Normal
   texture=/Game/Creative/Sets/PropSets/Primitives/Customization/Textures/Cakey/T_CM_Cakey_N2
  }
  n32 = MaterialFunctionCall(
   TextureObject_Diffuse=@n29
   TextureScale=ScalarParameter() {defaultValue=3, group=Cakey, parameterName=CM_Cakey_FrostUVTile}
   CalculatedBounds=NamedRerouteUsage()
   WorldNormals=NamedRerouteUsage()
   RotationX=@n30
   RotationY=@n30
   RotationZ=@n30
   ProjectionTransitionContrast=ScalarParameter() {defaultValue=4, group=Cakey, parameterName=CM_Cakey_FrostTriplannarContrast}
   Material Normals=NamedRerouteUsage()
   NormalTexture=@n31
   Normal Texture Intensity=ScalarParameter() {defaultValue=1, group=Cakey, parameterName=CM_Cakey_FrostNormalIntensity}
  ) {materialFunction=&MF_Primitive_TriplanarUVs}
  NamedRerouteDeclaration(None=ComponentMask(None=@n32.RGBA) {a=true}) {name=FrostSpec, nodeColor=<rgba 0 0.8594 1 1>}
  NamedRerouteDeclaration(None=MaterialFunctionCall(Hue Shift Percentage=NamedRerouteUsage(), Texture=ComponentMask(None=@n32.RGBA) {b=true, g=true, r=true}) {materialFunction=&HueShift}.Result) {name=FrostBaseColor, nodeColor=<rgba 0 0.8594 1 1>}
  NamedRerouteDeclaration(None=MaterialFunctionCall(BaseNormal=@n32.Normal Texture, AdditionalNormal=NamedRerouteUsage()) {materialFunction=&BlendAngleCorrectedNormals}.Result) {name=FrostNormal, nodeColor=<rgba 0 0.8594 1 1>}
  n33 = MaterialFunctionCall!(Value=NamedRerouteUsage()) {materialFunction=/Engine/Functions/Engine_MaterialFunctions03/Procedurals/PreparePerturbNormalHQ}
  NamedRerouteDeclaration(None=Transform(None=MaterialFunctionCall!(
     Bump one pixel right=@n33.Value one pixel right
     Bump center=@n33.Value one pixel right
     Bump one pixel down=@n33.Value one pixel right
     WS Normal
    ) {materialFunction=/Engine/Functions/Engine_MaterialFunctions03/Procedurals/PerturbNormalHQ}.WS Normal) {transformSourceType=TRANSFORMSOURCE_World, transformType=TRANSFORM_Tangent}) {name=dynamicnormal, nodeColor=<rgba 1 0.0469 0 1>}
  NamedRerouteDeclaration(None=@n2.Normal) {name=InNormal, nodeColor=<rgba 0 0.2266 1 1>}
  NamedRerouteDeclaration(None=LinearInterpolate(
    A=ScalarParameter() {group=Cakey, parameterName=CM_Cakey_CakeSpecular}
    B=ScalarParameter() {defaultValue=1, group=Cakey, parameterName=CM_Cakey_CakeSpecularHi}
    Alpha=NamedRerouteUsage()
   )) {name=FinalSpec, nodeColor=<rgba 0 0.4072 0.3968 1>}
  NamedRerouteDeclaration(None=LinearInterpolate(
    A=ScalarParameter() {group=Cakey, parameterName=CM_Cakey_CakeRoughness}
    B=ScalarParameter() {defaultValue=1, group=Cakey, parameterName=CM_Cakey_CakeRoughnessHi}
    Alpha=NamedRerouteUsage()
   )) {name=FinalRough, nodeColor=<rgba 0.4072 0.2831 0 1>}
  n34 = Min(A=Multiply(A=@n25, B=ScalarParameter() {defaultValue=1, group=DiscoLights, parameterName=CM_Cakey_EventWarpMultiplier})) {constB=30, desc=prevents crazy values}
  NamedRerouteDeclaration(None=MaterialFunctionCall!(In=@n34, Ramp Width) {materialFunction=/Engine/Functions/Engine_MaterialFunctions01/Math/SmoothCeil}.Result) {name=Warp, nodeColor=<rgba 0.2188 0 1 1>}
  NamedRerouteDeclaration(None=Ceil(None=Min(A=@n26.Result) {constB=30, desc=prevents crazy values})) {name=Switch, nodeColor=<rgba 0.7812 0 1 1>}
  NamedRerouteDeclaration(None=@n34) {name=Amplitude, nodeColor=<rgba 0 1 0.0859 1>}
  NamedRerouteDeclaration(None=Divide(A=@n24.WorldPosition, B=ScalarParameter() {defaultValue=100, group=Cakey, parameterName=CM_CM_Cakey_WorldScale})) {name=CalculatedBounds, nodeColor=<rgba 1 0.1406 0 1>}
  NamedRerouteDeclaration(None=@n24.World Normals) {name=WorldNorm, nodeColor=<rgba 0.7344 0 1 1>}
  Saturate(None=ComponentMask(None=NamedRerouteUsage()) {a=true})
  TextureCoordinate()
 ]}
