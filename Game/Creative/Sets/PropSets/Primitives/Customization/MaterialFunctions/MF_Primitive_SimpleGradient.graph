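# MF_Primitive_SimpleGradient is a material function that outputs an unclamped scalar gradient along a given axis, in the Creative primitive prop customization set.
# It adds a scalar Offset to the Bounds vector, divides by FallOff, and multiplies by the ForwardAxis vector.
# The result is split into components and reduced by Max nodes, which read R twice and B once, so the G component never affects the output.
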MaterialFunction /Game/Creative/Sets/PropSets/Primitives/Customization/MaterialFunctions/MF_Primitive_SimpleGradient
graph {nodes [
  n1 = Add(A=FunctionInput() {inputName=Bounds}, B=FunctionInput() {inputName=Offset, inputType=FunctionInput_Scalar, sortPriority=2})
  n2 = FunctionInput() {inputName=FallOff, inputType=FunctionInput_Scalar, sortPriority=3}
  n3 = MaterialFunctionCall(Float3=Multiply(A=Divide(A=@n1, B=@n2), B=FunctionInput() {inputName=ForwardAxis, sortPriority=3})) {materialFunction=&BreakOutFloat3Components}
  FunctionOutput(None=Max(A=Max(A=@n3.R, B=@n3.R), B=@n3.B))
 ]}
