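# Material function that applies a customizable clear-coated carbon fiber look to Creative primitive props, gated by the Primitive_CarbonFiber static switch, which otherwise passes input attributes through.
# Triplanar-projected weave textures supply masks and normals for layered base color, roughness, metallic and anisotropy, with view-angle falloffs.
# Scrolling noise warps the projection by an audio-driven amount, and a three-color object-height gradient feeds emissive.
# The result runs through MF_MetalsLUT with a Fresnel mask.
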MaterialFunction /Game/Creative/Sets/PropSets/Primitives/Customization/MaterialFunctions/MF_CM_CarbonFiber
graph {nodes [
  CurveAtlasRowParameter() {
   atlas=/Game/Creative/Sets/PropSets/Primitives/Customization/Textures/ColorCurve/CC_CM_Primitives_01
   curve=/Game/Creative/Sets/PropSets/Primitives/Customization/Textures/ColorCurve/C_CM_Simple_01
   group=CarbonFiber
   parameterName=CM_Carbon_CC
  }
  n1 = Fresnel()
  n2 = VectorParameter() {group=CarbonFiber, parameterName=CM_Carbon_BaseColor2, defaultValue=<rgba 0.02 0.02 0.02 1>}
  n3 = VectorParameter() {group=CarbonFiber, parameterName=CM_Carbon_BaseColor1, defaultValue=<rgba 0.06 0.06 0.06 1>}
  n4 = NamedRerouteUsage()
  n5 = VectorParameter() {group=CarbonFiber, parameterName=CM_Carbon_BaseColorCrevice, defaultValue=<rgba 0.0467 0.0247 0.0397 1>}
  n6 = NamedRerouteUsage()
  n7 = ScalarParameter() {defaultValue=3, group=CarbonFiber, parameterName=CM_Carbon_CrevicePower}
  n8 = LinearInterpolate(
   A=LinearInterpolate(
    A=VectorParameter() {group=CarbonFiber, parameterName=CM_Carbon_BaseColor0, defaultValue=<rgba 0.025 0.025 0.025 1>}.RGB
    B=LinearInterpolate(A=@n3.RGB, B=@n2.RGB, Alpha=@n4)
    Alpha=NamedRerouteUsage()
   )
   B=@n5.RGB
   Alpha=Multiply(A=Power(Base=@n6, Exp=@n7) {constExponent=4}, B=ScalarParameter() {defaultValue=1, group=CarbonFiber, parameterName=CM_Carbon_CreviceAmount})
  )
  n9 = NamedRerouteUsage()
  n10 = Multiply(A=NamedRerouteUsage(), B=ScalarParameter() {defaultValue=1, group=CarbonFiber, parameterName=CM_Carbon_GradientAmount})
  n11 = Multiply(A=@n8, B=ScalarParameter() {defaultValue=1, group=CarbonFiber, parameterName=CM_Carbon_BaseColorEmissiveAmount})
  n12 = TextureSample() {samplerType=SAMPLERTYPE_Normal, texture=/Game/Weapons/FORT_Rifles/T_SCAR_N}
  n13 = Constant3Vector() {constant=<rgba 1 0 0 0>}
  n14 = GetMaterialAttributes(None=FunctionInput(Preview=SetMaterialAttributes!(MaterialAttributes, Normal=@n12.RGB, BaseColor=@n13) {attributeSetTypes [#1, #2]}) {inputType=FunctionInput_MaterialAttributes}) {attributeGetTypes [#1]}
  n15 = SetMaterialAttributes!(
   MaterialAttributes=@n14.MaterialAttributes
   BaseColor=LinearInterpolate(A=@n2.RGB, B=@n8, Alpha=@n9)
   Metallic=NamedRerouteUsage()
   Specular=ScalarParameter() {defaultValue=1, group=CarbonFiber, parameterName=CM_Carbon_Specular}
   Roughness=NamedRerouteUsage()
   Normal=NamedRerouteUsage()
   EmissiveColor=Multiply(A=@n9, B=Multiply(A=@n10, B=@n11))
   CustomData0=ScalarParameter() {defaultValue=1, group=CarbonFiber, parameterName=CM_Carbon_ClearCoat}
   CustomData1=ScalarParameter() {group=CarbonFiber, parameterName=CM_Carbon_ClearCoatRoughness}
   Anisotropy=NamedRerouteUsage()
  ) {attributeSetTypes [
    #2
    #3
    #4
    #5
    #1
    #6
    #7
    #8
    #9
   ]}
  FunctionOutput(None=StaticSwitchParameter(
    True=MaterialFunctionCall(Mask=@n1, Materials=@n15) {materialFunction=&MF_MetalsLUT}.Result
    False=@n14.MaterialAttributes
   ) {group=Customization, parameterName=Primitive_CarbonFiber, sortPriority=-100})
  n16 = MaterialFunctionCall() {materialFunction=&/MF_Primitive_CalculatedBounds}
  n17 = ScalarParameter() {defaultValue=100, group=CarbonFiber, parameterName=CM_Carbon_UVTile}
  n18 = TextureObjectParameter() {
   group=CarbonFiber
   parameterName=CM_Carbon_UVWarpTexture
   samplerType=SAMPLERTYPE_Masks
   texture=/Game/Global/Textures/Default/Noise/T_WrapEdgeWear_Noise
  }
  n19 = Divide(A=NamedRerouteUsage(), B=ScalarParameter() {defaultValue=800, group=CarbonFiber, parameterName=CM_Carbon_UVWarpTile})
  n20 = Multiply(A=Time(), B=VectorParameter() {group=CarbonFiber, parameterName=CM_Carbon_UVWarpSpeed, defaultValue=<rgba -0.05 -0.05 -0.05 1>}.RGB)
  n21 = ComponentMask(None=MaterialFunctionCall!(
    TextureObject=@n18
    TextureSize=Constant() {r=1}
    WorldPosition=Add(A=@n19, B=@n20)
    World Space Normal=NamedRerouteUsage()
    ProjectionTransitionContrast=Constant() {r=2}
   ) {materialFunction=&/MF_Primitive_TriplanarUVSimple}.XYZ Texture) {r=true}
  n22 = Min(A=Multiply(
    A=MaterialFunctionCall() {materialFunction=&/MF_Primitive_AudioSetup}.Result
    B=ScalarParameter() {defaultValue=1, group=CarbonFiber, parameterName=CM_Carbon_EventAudioMultiplier}
   )) {constB=30, desc=prevents crazy values}
  n23 = LinearInterpolate(
   A=ScalarParameter() {group=CarbonFiber, parameterName=CM_Carbon_UVWarpAmount_Low}
   B=ScalarParameter() {defaultValue=0.5, group=CarbonFiber, parameterName=CM_Carbon_UVWarpAmount_High}
   Alpha=@n22
  )
  n24 = Constant()
  n25 = TextureObject() {samplerType=SAMPLERTYPE_Normal, texture=/Game/Creative/Sets/PropSets/Primitives/Customization/Textures/Carbon/T_CM_CarbonFiberWeave_02_N}
  n26 = MaterialFunctionCall!(
   TextureObject_Diffuse=TextureObject() {texture=/Game/Creative/Sets/PropSets/Primitives/Customization/Textures/Carbon/T_CM_CarbonFiberWeave_02_M}
   TextureScale=ScalarParameter() {defaultValue=1, group=CarbonFiber, parameterName=CM_Carbon_WorldTile}
   CalculatedBounds=Add(A=Divide(A=@n16.WorldPosition, B=@n17), B=Multiply(A=@n21, B=@n23))
   WorldNormals=NamedRerouteUsage()
   RotationX=@n24
   RotationY=@n24
   RotationZ=@n24
   ProjectionTransitionContrast=ScalarParameter() {defaultValue=10, group=CarbonFiber, parameterName=CM_Carbon_Contrast}
   Material Normals=NamedRerouteUsage()
   NormalTexture=@n25
   Normal Texture Intensity=ScalarParameter() {group=CarbonFiber, parameterName=CM_Carbon_NormalIntensity}
  ) {materialFunction=&/MF_Primitive_TriplanarUVs}
  n27 = NamedRerouteDeclaration(None=@n26.Normal Texture) {name=Normal, nodeColor=<rgba 1 0.7734 0 1>}
  LinearInterpolate(A=Constant3Vector() {constant=<rgba 0 0 1 0>}, B=@n27, Alpha=NamedRerouteUsage())
  n28 = Divide(
   A=Subtract(A=Time(), B=ScalarParameter() {defaultValue=-9999999, group=TIMEFromBP, parameterName=CM_TimeFrom_BP, sortPriority=600})
   B=ScalarParameter() {defaultValue=1, group=CarbonFiber, parameterName=CM_Carbon_Time_Duration, sortPriority=601}
  )
  Multiply(A=@n22, B=OneMinus(None=Saturate(None=@n28)))
  n29 = NamedRerouteUsage()
  n30 = Multiply(A=NamedRerouteUsage(), B=ScalarParameter() {defaultValue=4, group=CarbonFiber, parameterName=CM_Carbon_Roughness1})
  NamedRerouteDeclaration(None=LinearInterpolate(
    A=ScalarParameter() {defaultValue=0.3, group=CarbonFiber, parameterName=CM_Carbon_Roughness0}
    B=Multiply(A=@n29, B=@n30)
    Alpha=NamedRerouteUsage()
   )) {name=Roughness, nodeColor=<rgba 0.6641 0 1 1>}
  NamedRerouteDeclaration(None=@n14.Normal) {name=In Normal, nodeColor=<rgba 0 1 0.3203 1>}
  n31 = ComponentMask(None=WorldPosition().XYZ) {b=true}
  n32 = ObjectRadius()
  n33 = Subtract(A=ComponentMask(None=ObjectPositionWS()) {b=true}, B=Max(A=@n32) {constB=50})
  n34 = Saturate(None=Divide(A=Subtract(A=@n31, B=@n33), B=Multiply(A=@n32) {constB=1.8}) {constB=100})
  n35 = ScalarParameter() {group=CarbonFiber, parameterName=CM_Carbon_GradientOffset}
  n36 = Constant3Vector() {constant=<rgba 0 0 1 0>}
  n37 = ScalarParameter() {defaultValue=1, group=CarbonFiber, parameterName=CM_Carbon_GradientFalloff}
  NamedRerouteDeclaration(None=MaterialFunctionCall(
    A=VectorParameter() {group=CarbonFiber, parameterName=CM_Carbon_GradientColor3, defaultValue=<rgba 0.3712 0.2502 0.1746 1>}.RGB
    B=VectorParameter() {group=CarbonFiber, parameterName=CM_Carbon_GradientColor1, defaultValue=<rgba 0.2502 0.2664 0.305 1>}.RGB
    C=VectorParameter() {group=CarbonFiber, parameterName=CM_Carbon_GradientColor0, defaultValue=<rgba 0.2159 0.305 0.4233 1>}.RGB
    Alpha=MaterialFunctionCall!(Bounds=@n34, Offset=@n35, ForwardAxis=@n36, FallOff=@n37) {materialFunction=&/MF_Primitive_SimpleGradient}.Result
   ) {materialFunction=&Lerp_3Color}.Result) {name=Simple Gradient, nodeColor=<rgba 0.6406 1 0 1>}
  NamedRerouteDeclaration(None=@n27) {name=FinalNormal, nodeColor=<rgba 0 1 0.5781 1>}
  NamedRerouteDeclaration(None=@n16.World Normals) {name=WN, nodeColor=<rgba 0.0078 0 1 1>}
  n38 = DotProduct(A=VertexNormalWS(), B=CameraVectorWS())
  NamedRerouteDeclaration(None=Power(Base=@n38)) {name=Inset, nodeColor=<rgba 1 0 0.8672 1>}
  NamedRerouteDeclaration(None=@n16.WorldPosition) {name=WorldPosition, nodeColor=<rgba 0.2656 0 1 1>}
  NamedRerouteDeclaration(None=Power(Base=@n38) {constExponent=1.25}) {name=InvertFresnel, nodeColor=<rgba 1 0.2344 0 1>}
  n39 = Multiply(
   A=NamedRerouteUsage()
   B=Multiply(A=NamedRerouteUsage(), B=ScalarParameter() {defaultValue=0.6, group=CarbonFiber, parameterName=CM_Carbon_Metallic})
  )
  n40 = NamedRerouteUsage()
  n41 = ScalarParameter() {group=CarbonFiber, parameterName=CM_Carbon_MetallicOverride}
  NamedRerouteDeclaration(None=LinearInterpolate(A=LinearInterpolate(B=@n39, Alpha=@n40), B=@n41, Alpha=@n41)) {name=Metallic, nodeColor=<rgba 1 0 0.1406 1>}
  n42 = LinearInterpolate(Alpha=NamedRerouteUsage()) {constA=-0.3, constB=10}
  NamedRerouteDeclaration(None=LinearInterpolate(
    A=VectorParameter() {group=CarbonFiber, parameterName=CM_Carbon_Aniso1, defaultValue=<rgba 0.0833 0.0833 0.0833 1>}.RGB
    B=VectorParameter() {group=CarbonFiber, parameterName=CM_Carbon_Aniso2, defaultValue=<rgba 0.7292 0.7292 0.7292 1>}.RGB
    Alpha=Saturate(None=@n42)
   )) {name=Aniso, nodeColor=<rgba 0.8984 1 0 1>}
  NamedRerouteDeclaration(None=Power(Base=@n38) {constExponent=3.3}) {name=Fresnel, nodeColor=<rgba 0 1 0.0625 1>}
  NamedRerouteDeclaration(None=ComponentMask(None=@n26.RGBA) {r=true}) {name=Red, nodeColor=<rgba 0.4297 1 0 1>}
  NamedRerouteDeclaration(None=ComponentMask(None=@n26.RGBA) {g=true}) {name=Green, nodeColor=<rgba 0.875 1 0 1>}
  NamedRerouteDeclaration(None=ComponentMask(None=@n26.RGBA) {b=true}) {name=Blue, nodeColor=<rgba 0 1 0.5547 1>}
  n43 = VertexInterpolator(VS=PreSkinnedPosition())
  n44 = PreSkinnedLocalBounds()
  NamedRerouteDeclaration(None=Divide(A=Subtract(A=@n43.PS, B=@n44.Min), B=Subtract(A=@n44.Max, B=@n44.Max))) {name=Bounds, nodeColor=<rgba 0.4531 1 0 1>}
  NamedRerouteUsage()
  NamedRerouteUsage()
 ]}
