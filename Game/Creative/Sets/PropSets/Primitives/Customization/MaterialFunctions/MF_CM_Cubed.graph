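# Material function for Creative primitive props that overrides base color, emissive, specular, roughness and normal with a purple "Cubed" look behind the Primitive_Cubed static switch.
# Triplanar packed base, normal and noise textures supply edge, glow, height and sparkle masks, with HeightLerps blending the surface layers.
# Emissive glow layers four colors using view-angle masks, a camera-facing virtual-plane hotspot and time pulses.
# Music amplitude feeds hue-shift and warp values, and quality switches simplify the glow.
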MaterialFunction /Game/Creative/Sets/PropSets/Primitives/Customization/MaterialFunctions/MF_CM_Cubed
graph {nodes [
  n1 = GetMaterialAttributes(None=FunctionInput() {inputType=FunctionInput_MaterialAttributes}) {attributeGetTypes [#1]}
  n2 = SetMaterialAttributes!(
   MaterialAttributes=@n1.MaterialAttributes
   BaseColor=NamedRerouteUsage()
   EmissiveColor=NamedRerouteUsage()
   Specular=NamedRerouteUsage()
   Roughness=NamedRerouteUsage()
   Normal=NamedRerouteUsage()
  ) {attributeSetTypes [
    #2
    #3
    #4
    #5
    #1
   ]}
  FunctionOutput(None=StaticSwitchParameter(True=@n2, False=@n1.MaterialAttributes) {group=Customization, parameterName=Primitive_Cubed})
  n3 = ObjectPositionWS()
  n4 = Normalize(VectorInput=Subtract(A=@n3, B=CameraPositionWS()))
  n5 = Normalize(VectorInput=CrossProduct(A=@n4, B=Constant3Vector() {constant=<rgba 0 0 1 1>}))
  n6 = PreSkinnedLocalBounds()
  n7 = MaterialFunctionCall(Float3=Subtract(A=@n6.Max, B=@n6.Max)) {materialFunction=&BreakOutFloat3Components}
  n8 = Multiply(
   A=Max(A=Max(A=@n7.R, B=@n7.R), B=@n7.B)
   B=MaterialFunctionCall() {materialFunction=/Engine/Functions/Engine_MaterialFunctions02/WorldPositionOffset/ObjectScale}.Scale Z
  )
  n9 = Constant() {r=0.8}
  n10 = Constant() {r=0.5}
  n11 = NamedRerouteUsage()
  n12 = Multiply(
   A=SphereMask(
    A=MaterialFunctionCall!(Plane Normal Axis=@n4, Plane X Axis=@n5, Plane Center=@n3, UV Scale=@n8) {materialFunction=/Game/Athena/Environments/Nexus/Materials/VirtualPlaneCoordinatesAbs}.UVs
    B=Constant() {r=0.5}
    Radius=LinearInterpolate(A=@n9, B=@n10, Alpha=@n11)
    Hardness=Constant()
   )
   B=ScalarParameter() {defaultValue=0.5, group=CUBED, parameterName=CM_CUBED_CameraRadius_DarkenAmount}
  )
  NamedRerouteDeclaration(None=@n12) {name=VirtualPlane, nodeColor=<rgba 0 1 0.2734 1>}
  n13 = ScalarParameter() {defaultValue=0.5, group=CUBED, parameterName=CM_CUBED_PulseSpeed}
  n14 = Add(
   A=Multiply(A=Time(), B=@n13)
   B=LinearInterpolate(A=@n13, B=ScalarParameter() {defaultValue=0.5, group=CUBED, parameterName=CM_CUBED_PulseSpeedReactive}, Alpha=NamedRerouteUsage())
  )
  NamedRerouteDeclaration(None=MaterialFunctionCall(Value=Abs(None=@n14)) {materialFunction=&LinearSine}."Rounded Linear Sine ") {name=SineWave, nodeColor=<rgba 0.7344 1 0 1>}
  n15 = Saturate(None=Fresnel(ExponentIn=ScalarParameter() {defaultValue=4, group=CUBED, parameterName=CM_CUBED_Fresnel}) {baseReflectFraction=0})
  NamedRerouteDeclaration(None=@n15) {name=Fresnel, nodeColor=<rgba 0 0.2969 1 1>}
  n16 = CollectionParameter() {collection=/Game/Athena/Events/AudioAnalysis/MPC_Audio_Reactive_Parameters, parameterName=Amplitude_Music}
  n17 = MaterialFunctionCall() {materialFunction=&/MF_Primitive_AudioSetup}
  n18 = ScalarParameter() {defaultValue=0.1, group=CUBED, parameterName=CM_CUBED_EventAudioBlend}
  n19 = Min(A=Multiply(
    A=LinearInterpolate(A=@n16, B=@n17.Result, Alpha=@n18) {constAlpha=0.25}
    B=ScalarParameter() {defaultValue=1, group=CUBED, parameterName=CM_CUBED_EventAudioMultiplier}
   )) {constB=30, desc=prevents crazy values}
  NamedRerouteDeclaration(None=Saturate(None=Abs(None=@n19))) {name=Hueshift, nodeColor=<rgba 0.6172 0 1 1>}
  n20 = Multiply(A=@n16, B=ScalarParameter() {defaultValue=1, group=CUBED, parameterName=CM_CUBED_EventWarpMultiplier})
  NamedRerouteDeclaration(None=MaterialFunctionCall!(In=Min(A=@n20) {constB=30, desc=prevents crazy values}, Ramp Width) {materialFunction=/Engine/Functions/Engine_MaterialFunctions01/Math/SmoothCeil}.Result) {name=Warp, nodeColor=<rgba 0.2188 0 1 1>}
  n21 = VectorParameter() {group=CUBED, parameterName=CM_CUBED_BaseColor, defaultValue=<rgba 0.015 0.0013 0.1 1>}
  n22 = VectorParameter() {group=CUBED, parameterName=CM_CUBED_Sparkles_Color, defaultValue=<rgba 0.7322 0.25 1.25 1>}
  n23 = TextureObjectParameter() {group=CUBED, parameterName=CM_CUBED_NoiseTex, samplerType=SAMPLERTYPE_Masks, texture=/Game/Global/Textures/Default/Noise/T_CubeNoise}
  n24 = MaterialFunctionCall!(
   TextureObject=@n23
   TextureSize=ScalarParameter() {defaultValue=0.1, group=CUBED, parameterName=CM_CUBED_Sparkle_UVTile}
   WorldPosition=NamedRerouteUsage()
   World Space Normal=NamedRerouteUsage()
   ProjectionTransitionContrast=ScalarParameter() {defaultValue=4, group=CUBED, parameterName=CM_CUBED_Noise_TriplannarContrast}
  ) {materialFunction=&/MF_Primitive_TriplanarUVSimple}
  n25 = Multiply(
   A=Power(Base=ComponentMask(None=@n24.XYZ Texture) {b=true}, Exp=Constant() {r=7})
   B=ScalarParameter() {defaultValue=40, group=CUBED, parameterName=CM_CUBED_Sparkle_Intensity}
  )
  n26 = Multiply(A=@n22.RGB, B=@n25)
  n27 = Constant() {r=1}
  n28 = TextureObjectParameter() {group=CUBED, parameterName=CM_CUBED_NoiseTex, samplerType=SAMPLERTYPE_Masks, texture=&T_CubeNoise}
  n29 = MaterialFunctionCall(
   TextureObject=@n28
   TextureSize=ScalarParameter() {defaultValue=0.5, group=CUBED, parameterName=CM_CUBED_NoiseShadow_UVTile}
   WorldPosition=NamedRerouteUsage()
   World Space Normal=NamedRerouteUsage()
   ProjectionTransitionContrast=ScalarParameter() {defaultValue=4, group=CUBED, parameterName=CM_CUBED_Noise_TriplannarContrast}
  ) {materialFunction=&MF_Primitive_TriplanarUVSimple}
  n30 = ComponentMask(None=@n29.XYZ Texture) {r=true}
  n31 = Add(A=Power(Base=@n30, Exp=Constant() {r=3}), B=Constant() {r=-0.05})
  n32 = Multiply(
   A=Multiply(A=ShadingPathSwitch(Default=@n26, Mobile=@n27) {desc=Migrated from FeatureLevelSwitch}, B=Saturate(None=@n31))
   B=OneMinus(None=NamedRerouteUsage())
  )
  n33 = NamedRerouteUsage()
  n34 = NamedRerouteUsage()
  n35 = LinearInterpolate(
   A=LinearInterpolate(
    A=@n21.RGB
    B=VectorParameter() {group=CUBED, parameterName=CM_CUBED_BaseColorEdge, defaultValue=<rgba 0.006 0.0005 0.04 1>}.RGB
    Alpha=Add(A=@n32, B=@n33)
   )
   B=@n21.RGB
   Alpha=@n34
  )
  n36 = Add(A=Multiply(A=NamedRerouteUsage()) {constB=0.25}, B=@n35)
  n37 = NamedRerouteUsage()
  n38 = NamedRerouteUsage()
  NamedRerouteDeclaration(None=MaterialFunctionCall(A=LinearInterpolate(A=@n35, B=@n36, Alpha=@n37), B=@n36, Transition Phase=Subtract(A=@n38, B=@n33), Height Texture=@n34) {materialFunction=&HeightLerp}.Results) {name=BaseColor, nodeColor=<rgba 0 1 0.1094 1>}
  n39 = ScalarParameter() {defaultValue=0.4, group=CUBED, parameterName=CM_CUBED_Specular_Low}
  n40 = ScalarParameter() {defaultValue=0.1, group=CUBED, parameterName=CM_CUBED_Specular_High}
  n41 = Multiply(A=NamedRerouteUsage(), B=Constant() {r=2})
  n42 = NamedRerouteUsage()
  n43 = MaterialFunctionCall(Transition Phase=Power(Base=Saturate(None=@n41), Exp=Constant() {r=1.8}), Height Texture=@n42) {materialFunction=&HeightLerp}
  n44 = Subtract(A=Saturate(None=Add(A=@n43.Lerp Alpha No Contrast, B=@n42)), B=NamedRerouteUsage())
  NamedRerouteDeclaration(None=LinearInterpolate(A=LinearInterpolate(A=@n39, B=@n40, Alpha=@n44), B=@n39, Alpha=NamedRerouteUsage())) {name=Specular, nodeColor=<rgba 0 1 0.7891 1>}
  n45 = ScalarParameter() {defaultValue=0.29, group=CUBED, parameterName=CM_CUBED_Roughness_Low}
  n46 = ScalarParameter() {defaultValue=0.5, group=CUBED, parameterName=CM_CUBED_Roughness_High}
  n47 = Saturate(None=Multiply(A=NamedRerouteUsage(), B=Constant() {r=1.4}))
  n48 = Constant() {r=1.5}
  n49 = LinearInterpolate(
   A=@n45
   B=@n46
   Alpha=MaterialFunctionCall(Transition Phase=Power(Base=@n47, Exp=@n48), Height Texture=NamedRerouteUsage()) {materialFunction=&HeightLerp}.Lerp Alpha No Contrast
  )
  n50 = NamedRerouteUsage()
  NamedRerouteDeclaration(None=LinearInterpolate(A=LinearInterpolate(A=@n49, B=@n46, Alpha=@n50), B=@n45, Alpha=NamedRerouteUsage())) {name=Roughness, nodeColor=<rgba 1 0.2109 0 1>}
  NamedRerouteDeclaration(None=@n1.Normal) {name=InNormal, nodeColor=<rgba 0 0.2266 1 1>}
  n51 = Constant()
  n52 = TextureObjectParameter() {
   group=CUBED
   parameterName=CM_CUBED_BaseNormalTexture
   samplerType=SAMPLERTYPE_Normal
   texture=/Game/Creative/Sets/PropSets/Primitives/Customization/Textures/Cubed/T_CM_Cubed_N
  }
  NamedRerouteDeclaration(None=MaterialFunctionCall!(
    TextureObject_Diffuse
    TextureScale=ScalarParameter() {defaultValue=1, group=CUBED, parameterName=CM_CUBED_Base_UVTile}
    CalculatedBounds=NamedRerouteUsage()
    WorldNormals=NamedRerouteUsage()
    RotationX=@n51
    RotationY=@n51
    RotationZ=@n51
    ProjectionTransitionContrast=ScalarParameter() {defaultValue=10, group=CUBED, parameterName=CM_CUBED_Base_TriplannarContrast}
    Material Normals=NamedRerouteUsage()
    NormalTexture=@n52
    Normal Texture Intensity=ScalarParameter() {defaultValue=1, group=CUBED, parameterName=CM_CUBED_Base_NormalIntensity}
   ) {materialFunction=&/MF_Primitive_TriplanarUVs}.Normal Texture) {name=Normal, nodeColor=<rgba 1 0.2109 0 1>}
  NamedRerouteDeclaration(None=@n25) {name=Sparkles, nodeColor=<rgba 1 0 0.7031 1>}
  n53 = VectorParameter() {group=CUBED, parameterName=CM_CUBED_GlowColor04, defaultValue=<rgba 0.3583 0 0.5 1>}
  NamedRerouteDeclaration(None=@n53.RGB) {name=Highlight, nodeColor=<rgba 0.6406 0 1 1>}
  n54 = ScalarParameter() {defaultValue=1, group=CUBED, parameterName=CM_CUBED_Noise_UVTile}
  n55 = NamedRerouteUsage()
  n56 = NamedRerouteUsage()
  n57 = ScalarParameter() {defaultValue=4, group=CUBED, parameterName=CM_CUBED_Noise_TriplannarContrast}
  NamedRerouteDeclaration(None=ComponentMask(None=MaterialFunctionCall(TextureObject=@n28, TextureSize=@n54, WorldPosition=@n55, World Space Normal=@n56, ProjectionTransitionContrast=@n57) {materialFunction=&MF_Primitive_TriplanarUVSimple}.XYZ Texture) {r=true}) {name=Mask_RedChannel, nodeColor=<rgba 0 1 0.7891 1>}
  n58 = ScalarParameter() {defaultValue=4, group=CUBED, parameterName=CM_CUBED_Emissive_InnerGlowAmount}
  n59 = Multiply(A=Multiply(A=NamedRerouteUsage(), B=@n58) {constB=4}, B=NamedRerouteUsage())
  n60 = VectorParameter() {group=CUBED, parameterName=CM_CUBED_GlowColor01, defaultValue=<rgba 0.012 0.0011 0.08 1>}
  n61 = VectorParameter() {group=CUBED, parameterName=CM_CUBED_GlowColor02, defaultValue=<rgba 0.0356 0.001 0.2 1>}
  n62 = VectorParameter() {group=CUBED, parameterName=CM_CUBED_GlowColor03, defaultValue=<rgba 0.1264 0 0.3 1>}
  n63 = DotProduct(A=VertexNormalWS(), B=CameraVectorWS())
  n64 = NamedRerouteUsage()
  n65 = LinearInterpolate(
   A=ScalarParameter() {defaultValue=4, group=CUBED, parameterName=CM_CUBED_SphereMask_Radius}
   B=ScalarParameter() {defaultValue=8, group=CUBED, parameterName=CM_CUBED_SphereMask_Radius High}
   Alpha=@n64
  )
  n66 = Fresnel(ExponentIn=LinearInterpolate(
    A=ScalarParameter() {defaultValue=0.15, group=CUBED, parameterName=CM_CUBED_Emissive_FresnelRadius}
    B=ScalarParameter() {defaultValue=0.25, group=CUBED, parameterName=CM_CUBED_Emissive_FresnelRadius High}
    Alpha=@n64
   ))
  n67 = Power(
   Base=Max(A=Power(Base=@n63, Exp=@n65), B=OneMinus(None=@n66))
   Exp=ScalarParameter() {defaultValue=1.5, group=CUBED, parameterName=CM_CUBED_Emissive_Falloff}
  )
  n68 = Saturate(None=Add(A=NamedRerouteUsage(), B=ScalarParameter() {defaultValue=0.25, group=CUBED, parameterName=CM_CUBED_MaskTexture_Soften}))
  n69 = ScalarParameter() {defaultValue=2.75, group=CUBED, parameterName=CM_CUBED_MaskTexture_ShadowSinePowerMax}
  n70 = ScalarParameter() {defaultValue=1.5, group=CUBED, parameterName=CM_CUBED_MaskTexture_ShadowSinePowerMin}
  n71 = NamedRerouteUsage()
  n72 = Multiply(A=@n68, B=Power(Base=NamedRerouteUsage(), Exp=LinearInterpolate(A=@n69, B=@n70, Alpha=@n71)))
  n73 = MaterialFunctionCall(
   Transition Phase=@n12
   Height Texture=Multiply(A=NamedRerouteDeclaration(None=@n67) {name=SphereMask, nodeColor=<rgba 1 0.2578 0 1>}, B=QualitySwitch(Default=@n72, Low=@n68, High=@n72))
  ) {materialFunction=&HeightLerp}
  n74 = Multiply(A=@n12, B=ScalarParameter() {defaultValue=1, group=CUBED, parameterName=CM_CUBED_CameraRadius_ActiveHotspot})
  n75 = Multiply(A=@n74, B=LinearInterpolate(Alpha=NamedRerouteUsage()) {constA=0.33, constB=0.66})
  n76 = QualitySwitch(Default=@n75, Low=@n74, High=@n75)
  n77 = Multiply(
   A=Add(
    A=MaterialFunctionCall(1=@n60.RGB, 2=@n61.RGB, 3=@n62.RGB, 4=@n53.RGB, A=@n73.Lerp Alpha No Contrast) {materialFunction=&Lerp_Multiple_Float3}.Lerp 4 Inputs
    B=Multiply(A=@n53.RGB, B=@n76)
   )
   B=NamedRerouteUsage()
  )
  n78 = NamedRerouteUsage()
  n79 = NamedRerouteUsage()
  n80 = LinearInterpolate(
   A=LinearInterpolate(A=@n77, B=Add(A=@n77, B=@n59), Alpha=@n78)
   B=NamedRerouteUsage()
   Alpha=Multiply(
    A=LinearInterpolate(B=@n79, Alpha=@n78)
    B=ScalarParameter() {defaultValue=1, group=CUBED, parameterName=CM_CUBED_Emissive_SparkleGlowAmount}
   )
  )
  n81 = Multiply(A=NamedRerouteUsage(), B=@n58)
  n82 = Multiply(A=Time(), B=ScalarParameter() {defaultValue=0.033, group=CUBED, parameterName=CM_CUBED_EventGlowAmount})
  n83 = Divide(A=Add(A=Sine(None=@n82)))
  n84 = NamedRerouteUsage()
  n85 = Multiply(A=Power(Base=NamedRerouteUsage()), B=Add(A=LinearInterpolate(B=@n83, Alpha=@n84), B=Saturate(None=@n84)))
  n86 = ScalarParameter() {defaultValue=1, group=CUBED, parameterName=CM_CUBED_Emissive_FinalAmount}
  n87 = ScalarParameter() {defaultValue=2, group=CUBED, parameterName=CM_CUBED_Emissive_FinalAmountReactive}
  n88 = NamedRerouteUsage()
  NamedRerouteDeclaration(None=Multiply(A=LinearInterpolate(A=@n80, B=@n81, Alpha=@n85), B=LinearInterpolate(A=@n86, B=@n87, Alpha=@n88))) {name=Emissive, nodeColor=<rgba 0.4766 1 0 1>}
  n89 = TextureObjectParameter() {group=CUBED, parameterName=CM_CUBED_BaseTexture, texture=/Game/Creative/Sets/PropSets/Primitives/Customization/Textures/Cubed/T_CM_Cubed_BC}
  n90 = Constant()
  n91 = MaterialFunctionCall(
   TextureObject_Diffuse=@n89
   TextureScale=ScalarParameter() {defaultValue=1, group=CUBED, parameterName=CM_CUBED_Base_UVTile}
   CalculatedBounds=NamedRerouteUsage()
   WorldNormals=NamedRerouteUsage()
   RotationX=@n90
   RotationY=@n90
   RotationZ=@n90
   ProjectionTransitionContrast=ScalarParameter() {defaultValue=10, group=CUBED, parameterName=CM_CUBED_Base_TriplannarContrast}
  ) {materialFunction=&MF_Primitive_TriplanarUVs}
  NamedRerouteDeclaration(None=ComponentMask(None=@n91.RGBA) {r=true}) {name=OuterEdge, nodeColor=<rgba 0 0.9531 1 1>}
  NamedRerouteDeclaration(None=ComponentMask(None=@n91.RGBA) {g=true}) {name=InnerGlow, nodeColor=<rgba 0.7109 0 1 1>}
  NamedRerouteDeclaration(None=ComponentMask(None=@n91.RGBA) {b=true}) {name=MainHeight, nodeColor=<rgba 1 0 0.4687 1>}
  NamedRerouteDeclaration(None=@n22.RGB) {name=SparkleColor, nodeColor=<rgba 0 0.5781 1 1>}
  NamedRerouteDeclaration(None=@n15) {name=BoxMask, nodeColor=<rgba 0.2422 1 0 1>}
  NamedRerouteDeclaration(None=Power(Base=@n30, Exp=Constant() {r=3})) {name=ShadowMask_RedChannel, nodeColor=<rgba 1 0.0975 0 1>}
  NamedRerouteDeclaration(None=ComponentMask(None=@n29.XYZ Texture) {g=true}) {name=ShadowMask_GreenChannel, nodeColor=<rgba 0.0165 1 0 1>}
  n92 = MaterialFunctionCall() {materialFunction=&/MF_Primitive_CalculatedBounds}
  NamedRerouteDeclaration(None=Divide(A=@n92.WorldPosition, B=ScalarParameter() {defaultValue=175, group=CUBED, parameterName=CM_CUBED_WorldScale})) {name=CalculatedBounds, nodeColor=<rgba 1 0.1406 0 1>}
  NamedRerouteDeclaration(None=@n92.World Normals) {name=WorldNorm, nodeColor=<rgba 0.7344 0 1 1>}
  NamedRerouteUsage()
  NamedRerouteUsage()
 ]}
