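# Material function that adds an optional dirt-and-gravel surface to Creative primitive props, toggled by the Primitive_Dirt static switch, which otherwise passes incoming attributes through.
# Triplanar-projects Apollo dirt and gravel diffuse and normal textures using calculated-bounds positions, with a tiled noise texture supplying gravel weight.
# Landscape biome functions tint the color, and a packed mask supplies specular and roughness.
# Some landscape height-blend nodes remain unconnected to the output.
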MaterialFunction /Game/Creative/Sets/PropSets/Primitives/Customization/MaterialFunctions/MF_CM_Dirt
graph {nodes [
  Constant() {r=1}
  n1 = GetMaterialAttributes(None=FunctionInput() {inputType=FunctionInput_MaterialAttributes}) {attributeGetTypes [#1]}
  n2 = SetMaterialAttributes!(
   MaterialAttributes=@n1.MaterialAttributes
   BaseColor=NamedRerouteUsage()
   Specular=NamedRerouteUsage()
   Roughness=NamedRerouteUsage()
   Normal=NamedRerouteUsage()
   Metallic=ScalarParameter() {group=Dirt, parameterName=CM_Dirt_Metallic}
   EmissiveColor=Constant()
  ) {attributeSetTypes [
    #2
    #3
    #4
    #1
    #5
    #6
   ]}
  FunctionOutput(None=StaticSwitchParameter(True=@n2, False=@n1.MaterialAttributes) {group=Customization, parameterName=Primitive_Dirt})
  n3 = Multiply(A=LandscapeLayerCoords() {mappingScale=1}, B=ScalarParameter() {defaultValue=0.1, group=Rock Small, parameterName=EPIC_GravelA-Scale})
  n4 = TextureSample(UVs=@n3) {samplerSource=SSM_Wrap_WorldGroupSettings, texture=/Game/Environments/Apollo/Sets/Terrain/Textures/T_Apollo_DirtHeight_D}
  n5 = MaterialFunctionCall!(
   Diffuse 1
   Diffuse 2=Multiply(A=ScalarParameter() {defaultValue=0.85, group=Rock Small, parameterName=EPIC_GravelA-DiffuseMod}, B=@n4.RGB)
   Transistion Phase
   Height Texture 1
   Height Texture 2=@n4.A
   Contrast
  ) {materialFunction=/Engine/Functions/Engine_MaterialFunctions02/Texturing/HeightLerpWithTwoHeightMaps}
  n6 = LinearInterpolate(B=NamedRerouteUsage(), Alpha=@n5.Lerp Alpha No Contrast)
  n7 = Constant() {r=1}
  n8 = Multiply(A=Abs(None=Subtract(A=@n7, B=@n5.Lerp Alpha No Contrast)), B=@n5.Lerp Alpha No Contrast)
  n9 = ScalarParameter() {defaultValue=1.25, parameterName=EPIC_GravelMacroMask-Mod}
  LinearInterpolate(
   A=@n6
   B=MaterialFunctionCall(AdditionalNormal=@n6) {materialFunction=&BlendAngleCorrectedNormals}.Result
   Alpha=Saturate(None=Multiply(A=@n8, B=@n9))
  )
  n10 = NamedRerouteUsage()
  n11 = QualitySwitch(Default=@n10, Low=@n10, High=@n10, Medium=@n10, Epic=@n10)
  MaterialFunctionCall!(Grass, Gravel=@n11, Mud, Sand, UseBiomeColorMask=StaticBool() {value=true}, UseLandscapeCoords) {materialFunction=/Game/Athena/Apollo/Environments/Landscape/Materials/Biome/CommonMaterials/MF_BiomeColors}
  MaterialFunctionCall!(
   TextureObject_Diffuse
   TextureScale=ScalarParameter() {defaultValue=1, group=Dirt, parameterName=CM_Dirt_BaseUVTile}
   CalculatedBounds=NamedRerouteUsage()
   WorldNormals=NamedRerouteUsage()
   RotationX
   RotationY
   RotationZ
   ProjectionTransitionContrast=ScalarParameter() {defaultValue=10, group=Dirt, parameterName=CM_Dirt_TriplannarContrast}
   Material Normals=NamedRerouteUsage()
   NormalTexture
   Normal Texture Intensity=ScalarParameter() {defaultValue=1, group=Dirt, parameterName=CM_Dirt_NormalIntensity}
  ) {materialFunction=&/MF_Primitive_TriplanarUVs}
  NamedRerouteDeclaration(None=@n1.Normal) {name=InNormal, nodeColor=<rgba 0 0.171 1 1>}
  NamedRerouteDeclaration(None=@n1.MaterialAttributes) {name=Material Attribute, nodeColor=<rgba 0.7323 0 1 1>}
  n12 = TextureObjectParameter() {
   group=Dirt
   parameterName=CM_Dirt_SpecRoughTex
   samplerType=SAMPLERTYPE_Masks
   texture=/Game/Environments/Apollo/Sets/Terrain/Textures/T_Apollo_Dirt_S_01
  }
  n13 = MaterialFunctionCall!(
   TextureObject=@n12
   TextureSize=ScalarParameter() {defaultValue=1, group=Dirt, parameterName=CM_Dirt_BaseUVTile}
   WorldPosition=NamedRerouteUsage()
   World Space Normal=NamedRerouteUsage()
   ProjectionTransitionContrast=ScalarParameter() {defaultValue=10, group=Dirt, parameterName=CM_Dirt_TriplannarContrast}
  ) {materialFunction=&/MF_Primitive_TriplanarUVSimple}
  NamedRerouteDeclaration(None=ComponentMask(None=@n13.XYZ Texture) {b=true}) {name=Roughness, nodeColor=<rgba 1 0.1172 0 1>}
  n14 = MaterialFunctionCall() {materialFunction=/Game/Athena/Apollo/Environments/Landscape/Materials/Biome/CommonMaterials/MF_GravelAndMudColors}
  n15 = MaterialFunctionCall(
   TextureObject=TextureObjectParameter() {group=Dirt, parameterName=CM_Dirt_DiffuseTex, texture=/Game/Environments/Apollo/Sets/Terrain/Textures/T_Apollo_Dirt_D_01}
   TextureSize=ScalarParameter() {defaultValue=1, group=Dirt, parameterName=CM_Dirt_BaseUVTile}
   WorldPosition=NamedRerouteUsage()
   World Space Normal=NamedRerouteUsage()
   ProjectionTransitionContrast=ScalarParameter() {defaultValue=10, group=Dirt, parameterName=CM_Dirt_TriplannarContrast}
  ) {materialFunction=&MF_Primitive_TriplanarUVSimple}
  n16 = MaterialFunctionCall(
   TextureObject=TextureObjectParameter() {group=Dirt, parameterName=CM_Dirt_DiffuseGravel, texture=/Game/Environments/Apollo/Sets/Terrain/Textures/T_Apollo_Gravel_D_01}
   TextureSize=ScalarParameter() {defaultValue=1, group=Dirt, parameterName=CM_Dirt_BaseUVTile}
   WorldPosition=NamedRerouteUsage()
   World Space Normal=NamedRerouteUsage()
   ProjectionTransitionContrast=ScalarParameter() {defaultValue=10, group=Dirt, parameterName=CM_Dirt_TriplannarContrast}
  ) {materialFunction=&MF_Primitive_TriplanarUVSimple}
  n17 = NamedRerouteUsage()
  n18 = Multiply(A=@n14.GravelColorMod, B=ComponentMask(None=LinearInterpolate(A=@n15.XYZ Texture, B=@n16.XYZ Texture, Alpha=@n17)) {b=true, g=true, r=true})
  NamedRerouteDeclaration(None=LinearInterpolate(
    A=Multiply(A=@n14.GravelEdgeColorMod, B=@n18)
    B=@n18
    Alpha=ScalarParameter() {defaultValue=0.5, group=Dirt, parameterName=CM_Dirt_GravelWeight}
   )) {name=diffuse, nodeColor=<rgba 1 0.9141 0 1>}
  NamedRerouteDeclaration(None=ComponentMask(None=@n13.XYZ Texture) {r=true}) {name=Specular, nodeColor=<rgba 0 0.6016 1 1>}
  n19 = NamedRerouteUsage()
  NamedRerouteDeclaration(None=QualitySwitch(Default=@n19, Low=@n19, High=@n19, Medium=@n19, Epic=@n19)) {name=FinalNormal, nodeColor=<rgba 0.8047 0 1 1>}
  n20 = MaterialFunctionCall() {materialFunction=&/MF_Primitive_CalculatedBounds}
  n21 = Divide(A=@n20.WorldPosition, B=ScalarParameter() {defaultValue=100, group=Dirt, parameterName=CM_Dirt_WorldScale})
  NamedRerouteDeclaration(None=@n21) {name=UVS, nodeColor=<rgba 0.9922 0 1 1>}
  NamedRerouteDeclaration(None=@n21) {name=CalculatedBounds, nodeColor=<rgba 1 0.1964 0 1>}
  NamedRerouteDeclaration(None=@n20.World Normals) {name=WorldNorm, nodeColor=<rgba 0 1 0.9531 1>}
  NamedRerouteDeclaration(None=@n11) {name=Base Color, nodeColor=<rgba 0 1 0.4141 1>}
  NamedRerouteDeclaration(None=LinearInterpolate(
    A=ScalarParameter() {group=Dirt, parameterName=CM_Dirt_Specular}
    B=ScalarParameter() {defaultValue=1, group=Dirt, parameterName=CM_Dirt_SpecularHi}
    Alpha=NamedRerouteUsage()
   )) {name=FinalSpec, nodeColor=<rgba 0 0.4072 0.3968 1>}
  NamedRerouteDeclaration(None=LinearInterpolate(
    A=ScalarParameter() {group=Dirt, parameterName=CM_Dirt_Roughness}
    B=ScalarParameter() {defaultValue=1, group=Dirt, parameterName=CM_Dirt_RoughnessHi}
    Alpha=NamedRerouteUsage()
   )) {name=FinalRough, nodeColor=<rgba 0.4072 0.2831 0 1>}
  n22 = ScalarParameter() {defaultValue=1, group=Dirt, parameterName=CM_Dirt_BaseUVTile}
  n23 = NamedRerouteUsage()
  n24 = NamedRerouteUsage()
  n25 = ScalarParameter() {defaultValue=10, group=Dirt, parameterName=CM_Dirt_TriplannarContrast}
  n26 = NamedRerouteUsage()
  n27 = TextureObjectParameter() {
   group=Dirt
   parameterName=CM_Dirt_Normal
   samplerType=SAMPLERTYPE_Normal
   texture=/Game/Environments/Apollo/Sets/Terrain/Textures/T_Apollo_Dirt_N_01
  }
  n28 = ScalarParameter() {defaultValue=1, group=Dirt, parameterName=CM_Dirt_NormalIntensity}
  n29 = ScalarParameter() {defaultValue=1, group=Dirt, parameterName=CM_Dirt_BaseUVTile}
  n30 = NamedRerouteUsage()
  n31 = NamedRerouteUsage()
  n32 = ScalarParameter() {defaultValue=10, group=Dirt, parameterName=CM_Dirt_TriplannarContrast}
  n33 = NamedRerouteUsage()
  n34 = TextureObjectParameter() {
   group=Dirt
   parameterName=CM_Dirt_GravelNormal
   samplerType=SAMPLERTYPE_Normal
   texture=/Game/Environments/Apollo/Sets/Terrain/Textures/T_Apollo_Gravel_N_01
  }
  n35 = ScalarParameter() {defaultValue=1, group=Dirt, parameterName=CM_Dirt_NormalIntensity}
  NamedRerouteDeclaration(None=LinearInterpolate(
    A=MaterialFunctionCall(
     TextureScale=@n22
     CalculatedBounds=@n23
     WorldNormals=@n24
     ProjectionTransitionContrast=@n25
     Material Normals=@n26
     NormalTexture=@n27
     Normal Texture Intensity=@n28
    ) {materialFunction=&MF_Primitive_TriplanarUVs}.Normal Texture
    B=MaterialFunctionCall(
     TextureScale=@n29
     CalculatedBounds=@n30
     WorldNormals=@n31
     ProjectionTransitionContrast=@n32
     Material Normals=@n33
     NormalTexture=@n34
     Normal Texture Intensity=@n35
    ) {materialFunction=&MF_Primitive_TriplanarUVs}.Normal Texture
    Alpha=NamedRerouteUsage()
   )) {name=Normals, nodeColor=<rgba 0 1 0.2969 1>}
  n36 = TextureObjectParameter() {
   group=Dirt
   parameterName=CM_Dirt_NoiseGravelTex
   samplerType=SAMPLERTYPE_LinearGrayscale
   texture=/Game/Creative/Sets/PropSets/Primitives/Customization/Textures/GalacticSwirl/T_CM_GalacticSwirl_Grad
  }
  n37 = ScalarParameter() {defaultValue=1, group=Dirt, parameterName=CM_Dirt_NoiseGravelUVTile}
  n38 = Divide(
   A=MaterialFunctionCall() {materialFunction=&MF_Primitive_CalculatedBounds}.WorldPosition
   B=ScalarParameter() {defaultValue=100, group=Dirt, parameterName=CM_Dirt_NoiseGravelWorldScale}
  )
  n39 = NamedRerouteUsage()
  n40 = ScalarParameter() {defaultValue=10, group=Dirt, parameterName=CM_Dirt_NoiseGravelTriplannerContrast}
  n41 = Multiply(
   A=ComponentMask(None=MaterialFunctionCall(TextureObject=@n36, TextureSize=@n37, WorldPosition=@n38, World Space Normal=@n39, ProjectionTransitionContrast=@n40) {materialFunction=&MF_Primitive_TriplanarUVSimple}.XYZ Texture) {r=true}
   B=Constant() {r=1.65}
  )
  NamedRerouteDeclaration(None=Multiply(A=ScalarParameter() {defaultValue=0.5, group=Dirt, parameterName=CM_Dirt_GravelWeight}, B=Saturate(None=@n41))) {name=Gravel Weight, nodeColor=<rgba 1 0 0.8203 1>}
  n42 = TextureSample(UVs=Multiply(B=ScalarParameter() {defaultValue=0.07, group=Rock Small, parameterName=EPIC_GravelMacro-Scale})) {
   samplerSource=SSM_Wrap_WorldGroupSettings
   samplerType=SAMPLERTYPE_LinearColor
   texture=/Game/Environments/Apollo/Sets/Terrain/Textures/T_Apollo_DirtHeight_M
  }
  n43 = ScalarParameter() {defaultValue=1.5, parameterName=EPIC_GravelMacroHeight-Pow}
  Saturate(None=Multiply(A=Power(Base=@n42.R, Exp=@n43), B=ScalarParameter() {defaultValue=0.75, parameterName=EPIC_GravelMacroHeight-Mod}) {desc="Can use an additional mask to affect the multiply amount for variation!"})
  TextureSample(UVs=@n3) {
   samplerSource=SSM_Wrap_WorldGroupSettings
   samplerType=SAMPLERTYPE_Normal
   texture=/Game/Environments/Apollo/Sets/Terrain/Textures/T_Apollo_DirtHeight_N
  }
 ]}
