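# MF_CM_DiscoLights is a material function that overlays an animated, music-reactive disco-light look on Creative primitive props, gated by the Primitive_DiscoLights static switch.
# Triplanar mask, normal and noise textures, scaled by calculated bounds, shape the base color, roughness, specular and normal.
# Time-driven flash and pulse waves feed masked emissive, tinted by hue-shift and panning rainbow-gradient fresnel reflections.
# Music amplitude and audio-setup values, clamped against spikes, drive hue and warp.
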MaterialFunction /Game/Creative/Sets/PropSets/Primitives/Customization/MaterialFunctions/MF_CM_DiscoLights
graph {nodes [
  n1 = GetMaterialAttributes(None=FunctionInput() {inputType=FunctionInput_MaterialAttributes}) {attributeGetTypes [#1]}
  n2 = SetMaterialAttributes!(
   MaterialAttributes=@n1.MaterialAttributes
   BaseColor=NamedRerouteUsage()
   EmissiveColor=NamedRerouteUsage()
   Specular=NamedRerouteUsage()
   Roughness=NamedRerouteUsage()
   Normal=NamedRerouteUsage()
   Metallic=ScalarParameter() {defaultValue=1, group=DiscoLights, parameterName=CM_DL_Metallic}
   AmbientOcclusion=NamedRerouteUsage()
  ) {attributeSetTypes [
    #2
    #3
    #4
    #5
    #1
    #6
    #7
   ]}
  FunctionOutput(None=StaticSwitchParameter(True=@n2, False=@n1.MaterialAttributes) {group=Customization, parameterName=Primitive_DiscoLights})
  n3 = VectorParameter() {group=DiscoLights, parameterName=CM_DL_BaseColorLo, defaultValue=<rgba 0.06 0.06 0.06 1>}
  n4 = VectorParameter() {group=DiscoLights, parameterName=CM_DL_BaseColor, defaultValue=<rgba 0.06 0.06 0.06 1>}
  n5 = NamedRerouteUsage()
  n6 = Desaturation(None=NamedRerouteUsage())
  NamedRerouteDeclaration(None=Subtract(A=LinearInterpolate(A=@n3.RGB, B=@n4.RGB, Alpha=@n5), B=Power(Base=@n6) {constExponent=40})) {name=Base Color, nodeColor=<rgba 1 0.3281 0 1>}
  NamedRerouteDeclaration(None=@n1.MaterialAttributes) {name=Material Attribute, nodeColor=<rgba 0.6641 0 1 1>}
  n7 = TextureObjectParameter() {
   group=DiscoLights
   parameterName=CM_DL_NoiseTex
   texture=/Game/Creative/Sets/PropSets/Primitives/Customization/Textures/NeonGasCloud/T_CM_NeonGasCloud_M2
  }
  n8 = MaterialFunctionCall!(
   TextureObject=@n7
   TextureSize=ScalarParameter() {defaultValue=1, group=DiscoLights, parameterName=CM_DL_Noise_UVTile}
   WorldPosition=NamedRerouteUsage()
   World Space Normal=NamedRerouteUsage()
   ProjectionTransitionContrast=ScalarParameter() {defaultValue=10, group=DiscoLights, parameterName=CM_DL_Noise_TriplannarContrast}
  ) {materialFunction=&/MF_Primitive_TriplanarUVSimple}
  NamedRerouteDeclaration(None=ComponentMask(None=@n8.XYZ Texture) {r=true}) {name=Roughness, nodeColor=<rgba 1 0.1172 0 1>}
  n9 = NamedRerouteUsage()
  n10 = Add(A=@n9) {constB=-0.75}
  n11 = Multiply(A=Time(), B=ScalarParameter() {defaultValue=0.5, group=DiscoLights, parameterName=CM_DL_LightSpeed})
  n12 = Subtract(A=Saturate(None=Multiply(A=@n10) {constB=4}), B=@n11)
  n13 = LinearInterpolate(A=Abs(None=Cosine(None=@n12)))
  n14 = ScalarParameter() {defaultValue=10, group=DiscoLights, parameterName=CM_DL_PulsePower}
  NamedRerouteDeclaration(None=Saturate(None=Power(Base=@n13, Exp=@n14))) {name=Flash1, nodeColor=<rgba 0 1 0.4844 1>}
  NamedRerouteDeclaration(None=ComponentMask(None=@n8.XYZ Texture) {a=true}) {name=Specular, nodeColor=<rgba 0 0.6016 1 1>}
  n15 = Add(A=@n9) {constB=-0.25}
  n16 = Subtract(A=Saturate(None=Multiply(A=@n15) {constB=4}), B=@n11)
  n17 = Abs(None=Cosine(None=Multiply(A=@n16) {constB=0.25}))
  n18 = Subtract(A=Clamp(None=Multiply(A=@n9) {constB=4}) {minDefault=0.001}, B=@n11)
  n19 = Abs(None=MaterialFunctionCall(Value=Multiply(A=@n18) {constB=0.5}) {materialFunction=&LinearSine}.Linear Sine)
  NamedRerouteDeclaration(None=LinearInterpolate(A=@n13, B=LinearInterpolate(A=@n17), Alpha=LinearInterpolate(A=@n19))) {name=FlashRandom, nodeColor=<rgba 0.9688 1 0 1>}
  n20 = MaterialFunctionCall() {materialFunction=&/MF_Primitive_CalculatedBounds}
  n21 = Divide(A=@n20.WorldPosition, B=ScalarParameter() {defaultValue=100, group=DiscoLights, parameterName=CM_DL_WorldScale})
  NamedRerouteDeclaration(None=@n21) {name=UVs, nodeColor=<rgba 0 1 0.4375 1>}
  n22 = Multiply(A=Add(A=@n9) {constB=-0.5}) {constB=4}
  n23 = Multiply(A=Subtract(A=Saturate(None=@n22), B=@n11)) {constB=0.125}
  n24 = LinearInterpolate(A=Abs(None=Frac(None=@n23)))
  NamedRerouteDeclaration(None=LinearInterpolate(Alpha=Clamp(None=@n24)) {constA=2, constB=4}) {name=Flash2, nodeColor=<rgba 1 0 0.2812 1>}
  n25 = ScalarParameter() {defaultValue=0.25, group=DiscoLights, parameterName=CM_DL_AlternateFreq}
  n26 = Abs(None=Cosine(None=Multiply(A=@n18, B=@n25) {constB=0.25}))
  NamedRerouteDeclaration(None=Saturate(None=Power(Base=@n26, Exp=@n14))) {name=FlashPulse, nodeColor=<rgba 0 0.1328 1 1>}
  n27 = TextureObject() {samplerType=SAMPLERTYPE_Masks, texture=/Game/Creative/Sets/PropSets/Primitives/Customization/Textures/DiscoLights/T_CM_DiscoLights_M}
  n28 = MaterialFunctionCall(
   TextureObject=@n27
   TextureSize=ScalarParameter() {defaultValue=1, group=DiscoLights, parameterName=CM_DL_BaseUVTile}
   WorldPosition=Frac(None=NamedRerouteUsage())
   World Space Normal=NamedRerouteUsage()
   ProjectionTransitionContrast=ScalarParameter() {defaultValue=100, group=DiscoLights, parameterName=CM_DL_BaseTriplannarContrast}
  ) {materialFunction=&MF_Primitive_TriplanarUVSimple}
  n29 = ComponentMask(None=@n28.XYZ Texture) {r=true}
  n30 = ScalarParameter() {desc=New, group=DiscoLights, parameterName=CM_DL_ThresholdOffset}
  n31 = ComponentMask(None=@n28.XYZ Texture) {b=true}
  n32 = LinearInterpolate(
   A=Add(A=NamedRerouteDeclaration(None=@n29) {name=Warp01, nodeColor=<rgba 1 0 0.7969 1>}, B=@n30) {constB=0.6, desc=New}
   B=Add(A=NamedRerouteDeclaration(None=@n31) {name=Warp02, nodeColor=<rgba 1 0 0.7969 1>}, B=@n30) {constB=0.6, desc=New}
   Alpha=NamedRerouteUsage()
  )
  n33 = NamedRerouteUsage()
  n34 = LinearInterpolate(A=@n32, B=NamedRerouteUsage())
  n35 = Multiply(A=NamedRerouteUsage()) {constB=2}
  n36 = Saturate(None=Multiply(A=NamedRerouteUsage(), B=Min(A=@n35)))
  n37 = NamedRerouteUsage()
  NamedRerouteDeclaration(None=LinearInterpolate(A=@n32, B=LinearInterpolate(A=@n33, B=@n34, Alpha=@n36), Alpha=Saturate(None=@n37))) {name=FlashFinal, nodeColor=<rgba 0.0312 1 0 1>}
  n38 = TextureObjectParameter() {
   group=DiscoLights
   parameterName=CM_DL_NormalTexture
   samplerType=SAMPLERTYPE_Normal
   texture=/Game/Creative/Sets/PropSets/Primitives/Customization/Textures/DiscoLights/T_CM_DiscoLights_N
  }
  NamedRerouteDeclaration(None=MaterialFunctionCall!(
    TextureObject_Diffuse
    TextureScale=ScalarParameter() {defaultValue=1, group=DiscoLights, parameterName=CM_DL_BaseUVTile}
    CalculatedBounds=NamedRerouteUsage()
    WorldNormals=NamedRerouteUsage()
    RotationX
    RotationY
    RotationZ
    ProjectionTransitionContrast=ScalarParameter() {defaultValue=100, group=DiscoLights, parameterName=CM_DL_BaseTriplannarContrast}
    Material Normals=NamedRerouteUsage()
    NormalTexture=@n38
    Normal Texture Intensity=ScalarParameter() {defaultValue=1, group=DiscoLights, parameterName=CM_DL_NormalIntensity}
   ) {materialFunction=&/MF_Primitive_TriplanarUVs}.Normal Texture) {name=FinalNormal, nodeColor=<rgba 0.8047 0 1 1>}
  n39 = NamedRerouteUsage()
  n40 = Multiply(A=ComponentMask(None=ReflectionVectorWS(CustomWorldNormal=@n39)) {g=true, r=true}, B=Fresnel() {baseReflectFraction=0, exponent=0.5})
  n41 = LinearInterpolate(B=TextureSample(UVs=Panner(Coordinate=@n40) {speedX=1, speedY=1}) {texture=/Game/Global/Textures/Default/Masks_and_Gradients/T_Grad_Rainbow}.RGB) {constA=1, constAlpha=0.75}
  n42 = Constant() {r=1}
  n43 = ScalarParameter() {defaultValue=8, group=DiscoLights, parameterName=CM_DL_FrenselMaskExp}
  NamedRerouteDeclaration(None=Multiply(A=QualitySwitch(Default=@n41, Low=@n42), B=Fresnel(ExponentIn=@n43) {baseReflectFraction=0})) {name=Rainbow, nodeColor=<rgba 0 0.3672 1 1>}
  n44 = Power(Base=NamedRerouteUsage())
  n45 = OneMinus(None=NamedRerouteUsage())
  n46 = NamedRerouteUsage()
  n47 = LinearInterpolate(
   A=Clamp(None=Power(Base=@n45, Exp=@n46))
   B=NamedRerouteUsage()
   Alpha=ScalarParameter() {group=DiscoLights, parameterName=CM_DL_DontInvertMask}
  )
  n48 = NamedRerouteUsage()
  n49 = If(
   A=Power(
    Base=LinearInterpolate(A=@n44, B=@n47, Alpha=@n48)
    Exp=ScalarParameter() {defaultValue=3, group=DiscoLights, parameterName=CM_DL_Light Frequency}
   ) {constExponent=4}
   B=ScalarParameter() {defaultValue=0.05, group=DiscoLights, parameterName=CM_DL_Threshold_Min}
   "A > B"=Constant() {r=1}
   "A < B"=Constant()
  )
  n50 = NamedRerouteUsage()
  NamedRerouteDeclaration(None=Multiply(A=Saturate(None=@n49), B=Power(Base=@n50) {constExponent=24})) {name=Mask, nodeColor=<rgba 0 1 0.7188 1>}
  n51 = CollectionParameter() {collection=/Game/Athena/Events/AudioAnalysis/MPC_Audio_Reactive_Parameters, parameterName=Amplitude_Music}
  n52 = MaterialFunctionCall() {materialFunction=&/MF_Primitive_AudioSetup}
  n53 = ScalarParameter() {defaultValue=0.1, group=DiscoLights, parameterName=CM_DL_EventAudioBlend}
  n54 = Min(A=Multiply(
    A=LinearInterpolate(A=@n51, B=@n52.Result, Alpha=@n53) {constAlpha=0.25}
    B=ScalarParameter() {defaultValue=1, group=DiscoLights, parameterName=CM_DL_EventAudioMultiplier}
   )) {constB=30, desc=prevents crazy values}
  NamedRerouteDeclaration(None=Saturate(None=Abs(None=@n54))) {name=Hueshift, nodeColor=<rgba 0.6172 0 1 1>}
  n55 = Min(A=Multiply(A=@n51, B=ScalarParameter() {defaultValue=1, group=DiscoLights, parameterName=CM_DL_EventWarpMultiplier})) {constB=30, desc=prevents crazy values}
  NamedRerouteDeclaration(None=MaterialFunctionCall!(In=@n55, Ramp Width) {materialFunction=/Engine/Functions/Engine_MaterialFunctions01/Math/SmoothCeil}.Result) {name=Warp, nodeColor=<rgba 0.2188 0 1 1>}
  NamedRerouteDeclaration(None=Ceil(None=Min(A=@n52.Result) {constB=30, desc=prevents crazy values})) {name=Switch, nodeColor=<rgba 0.7813 0 1 1>}
  NamedRerouteDeclaration(None=@n1.Normal) {name=InNormal, nodeColor=<rgba 0 0.2266 1 1>}
  n56 = VectorParameter() {group=DiscoLights, parameterName=CM_DL_EmissiveColor, defaultValue=<rgba 1 1 1 1>}
  n57 = Multiply(A=Time(), B=ScalarParameter() {defaultValue=0.125, group=DiscoLights, parameterName=CM_DL_EventHueshiftAmount})
  n58 = Divide(A=Add(A=Sine(None=@n57)))
  n59 = NamedRerouteUsage()
  n60 = NamedRerouteDeclaration(None=Add(A=LinearInterpolate(B=@n58, Alpha=@n59), B=Saturate(None=@n59))) {name=React, nodeColor=<rgba 0 1 0.6016 1>}
  n61 = NamedRerouteUsage()
  n62 = MaterialFunctionCall(Hue Shift Percentage=Add(A=Add(A=@n60, B=@n61), B=NamedRerouteUsage()), Texture=NamedRerouteUsage()) {materialFunction=&HueShift}
  n63 = Multiply(A=NamedRerouteUsage(), B=NamedRerouteUsage())
  n64 = Desaturation(
   None=Add(A=NamedRerouteUsage(), B=Multiply(A=@n62.Result, B=@n63))
   Fraction=ScalarParameter() {group=DiscoLights, parameterName=CM_DL_EmissiveDesaturation}
  )
  n65 = ScalarParameter() {defaultValue=8, group=DiscoLights, parameterName=CM_DL_EmissiveBrightness}
  n66 = Min(A=Multiply(A=NamedRerouteUsage()) {constB=2})
  NamedRerouteDeclaration(None=Multiply(A=Multiply(A=@n56.RGB, B=@n64), B=Multiply(A=@n65, B=@n66))) {name=EmissiveFinal, nodeColor=<rgba 1 0.375 0 1>}
  n67 = Constant2Vector() {g=1, r=1}
  n68 = ScalarParameter() {defaultValue=0.125, group=DiscoLights, parameterName=CM_DL_HueshiftReactiveAmount}
  n69 = Add(A=Multiply(A=Time(), B=@n67), B=LinearInterpolate(A=@n67, B=Add(A=@n67, B=@n68), Alpha=NamedRerouteUsage()))
  n70 = Multiply(A=ComponentMask(None=ReflectionVectorWS()) {g=true, r=true}, B=Fresnel() {baseReflectFraction=0, exponent=0.5})
  NamedRerouteDeclaration(None=TextureSample(UVs=Add(A=@n69, B=@n70)) {texture=&T_Grad_Rainbow}.RGB) {name=FresnelColor, nodeColor=<rgba 1 0.375 0 1>}
  NamedRerouteDeclaration(None=@n55) {name=Amplitude, nodeColor=<rgba 0 1 0.0859 1>}
  n71 = ScalarParameter() {group=DiscoLights, parameterName=CM_DL_Specular}
  n72 = ScalarParameter() {defaultValue=1, group=DiscoLights, parameterName=CM_DL_SpecularHi}
  n73 = NamedRerouteUsage()
  NamedRerouteDeclaration(None=Multiply(A=LinearInterpolate(A=@n71, B=@n72, Alpha=@n73), B=NamedRerouteUsage())) {name=FinalSpec, nodeColor=<rgba 0 0.4072 0.3968 1>}
  n74 = ScalarParameter() {group=DiscoLights, parameterName=CM_DL_Roughness}
  n75 = ScalarParameter() {defaultValue=0.15, group=DiscoLights, parameterName=CM_DL_RoughnessHi}
  n76 = NamedRerouteUsage()
  n77 = OneMinus(None=NamedRerouteUsage())
  n78 = ScalarParameter() {defaultValue=0.5, group=DiscoLights, parameterName=CM_DL_RoughnessHeight}
  NamedRerouteDeclaration(None=Add(A=LinearInterpolate(A=@n74, B=@n75, Alpha=@n76), B=Multiply(A=@n77, B=@n78))) {name=FinalRough, nodeColor=<rgba 0.4072 0.2831 0 1>}
  n79 = ScalarParameter() {defaultValue=4, group=DiscoLights, parameterName=CM_DL_Fresnel}
  NamedRerouteDeclaration(None=Saturate(None=Fresnel(ExponentIn=@n79) {baseReflectFraction=0})) {name=EdgeMask, nodeColor=<rgba 0.2422 1 0 1>}
  NamedRerouteDeclaration(None=ComponentMask(None=@n28.XYZ Texture) {g=true}) {name=Height, nodeColor=<rgba 0.8516 0 1 1>}
  NamedRerouteDeclaration(None=@n21) {name=CalculatedBounds, nodeColor=<rgba 1 0.1406 0 1>}
  NamedRerouteDeclaration(None=@n20.World Normals) {name=WorldNorm, nodeColor=<rgba 0.7344 0 1 1>}
  NamedRerouteUsage()
  NamedRerouteUsage()
  NamedRerouteUsage()
  NamedRerouteUsage()
  NamedRerouteUsage()
  NamedRerouteUsage()
  NamedRerouteUsage()
  ScalarParameter() {defaultValue=0.5, group=DiscoLights, parameterName=CM_DL_AlternateEmissiveMasks}
 ]}
