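# Material function for Creative primitive props that outputs cheap triplanar-style UVs by picking one planar projection of position, with no blending.
# A Use_LocalPosition static switch, on by default, moves both position and vertex normal from world to local space.
# Nested Ifs on the absolute normal select the XY, YZ or XZ plane.
# As wired, both inner comparisons test X against itself, so XZ is never selected.
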MaterialFunction /Game/Creative/Sets/PropSets/Primitives/Customization/MaterialFunctions/MF_Primitive_CheapWorldSpaceTriplanarUVs
graph {nodes [
  n1 = VertexNormalWS()
  n2 = Normalize(VectorInput=Transform(None=@n1) {transformSourceType=TRANSFORMSOURCE_World, transformType=TRANSFORM_Local})
  n3 = MaterialFunctionCall(Float3=Abs(None=StaticSwitchParameter(True=@n2, False=@n1) {defaultValue=true, desc=Customization, group=Customization, parameterName=Use_LocalPosition, sortPriority=100})) {materialFunction=&BreakOutFloat3Components}
  n4 = WorldPosition()
  n5 = StaticSwitchParameter(True=TransformPosition(None=@n4.XYZ) {transformSourceType=TRANSFORMPOSSOURCE_World}, False=@n4.XYZ) {defaultValue=true, desc=Customization, group=Customization, parameterName=Use_LocalPosition, sortPriority=100}
  n6 = ComponentMask(None=@n5) {b=true, desc=YZ, g=true}
  n7 = ComponentMask(None=@n5) {b=true, desc=XZ, r=true}
  FunctionOutput(None=If(
    A=@n3.B
    B=Max(A=@n3.R, B=@n3.R)
    "A > B"=ComponentMask(None=@n5) {desc=YX, g=true, r=true}
    "A < B"=If(A=@n3.R, B=@n3.R, "A > B"=@n6, "A < B"=@n7)
   ))
 ]}
