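# Material function for Creative primitive props that applies an animated galaxy look to material attributes, behind the Primitive_GalacticSwirl static switch.
# Camera-vector triplanar projections build a rotating nebula background with twinkling stars and a noise-tinted Fresnel rim.
# Triplanar gas clouds use two-phase flowmap distortion, four tint colors and a time-based hue shift.
# Music amplitude from a parameter collection and audio setup function supplies clamped hue-shift and warp signals.
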MaterialFunction /Game/Creative/Sets/PropSets/Primitives/Customization/MaterialFunctions/MF_CM_GalacticSwirl
graph {nodes [
  n1 = Multiply(A=Time(), B=VectorParameter() {group=GalacticSwirl, parameterName=CM_GS_StarWiggle, defaultValue=<rgba 0.5 0 0 1>}.RGB)
  n2 = Sine(None=@n1) {period=3.7}
  n3 = MaterialFunctionCall(X=@n2, Y=@n2, Z=Cosine(None=@n1) {period=2.3}) {materialFunction=&MakeFloat3}
  n4 = ScalarParameter() {defaultValue=0.15, group=GalacticSwirl, parameterName=CM_GS_StarPulseAmplitude, sortPriority=210}
  n5 = NamedRerouteUsage()
  Add(A=ComponentMask(None=Multiply(A=@n3.Result, B=@n4) {constB=0.1}) {b=true, r=true}, B=@n5)
  Divide(A=MaterialFunctionCall() {materialFunction=&/MF_Primitive_CheapWorldSpaceTriplanarUVs}.Result) {constB=100}
  n6 = GetMaterialAttributes(None=FunctionInput() {inputType=FunctionInput_MaterialAttributes}) {attributeGetTypes [#1]}
  n7 = SetMaterialAttributes!(
   MaterialAttributes=@n6.MaterialAttributes
   BaseColor=VectorParameter() {group=GalacticSwirl, parameterName=CM_GS_BaseColor, defaultValue=<rgba 0 0 0 1>}.RGB
   Metallic=ScalarParameter() {group=GalacticSwirl, parameterName=CM_GS_Metallic}
   Specular=NamedRerouteUsage()
   Roughness=NamedRerouteUsage()
   EmissiveColor=NamedRerouteUsage()
  ) {attributeSetTypes [
    #2
    #3
    #4
    #5
    #6
   ]}
  FunctionOutput(None=StaticSwitchParameter(True=@n7, False=@n6.MaterialAttributes) {group=Customization, parameterName=Primitive_GalacticSwirl, sortPriority=-100})
  n8 = ScalarParameter() {defaultValue=0.5, group=GalacticSwirl, parameterName=CM_GS_BGTiling}
  n9 = ScalarParameter() {defaultValue=6, group=GalacticSwirl, parameterName=CM_GS_BGAxisFade}
  n10 = Time()
  n11 = Constant()
  n12 = ScalarParameter() {defaultValue=0.02, group=GalacticSwirl, parameterName=CM_GS_BGRotationSpeed}
  n13 = ScalarParameter() {defaultValue=-0.02, group=GalacticSwirl, parameterName=CM_GS_BGRotationSpeed_1}
  n14 = Multiply(
   A=LinearInterpolate(A=@n10, B=Multiply(A=@n10, B=@n11), Alpha=ScalarParameter() {group=GalacticSwirl, parameterName=CM_GS_BGRotation})
   B=AppendVector(A=AppendVector(A=@n12, B=@n13), B=ScalarParameter() {defaultValue=0.1, group=GalacticSwirl, parameterName=CM_GS_BGRotationSpeed_2})
  )
  n15 = TextureObjectParameter() {
   group=GalacticSwirl
   parameterName=CM_GS_GalaxyTexture
   texture=/Game/Creative/Sets/PropSets/Primitives/Customization/Textures/GalacticSwirl/T_CM_GalacticSwirl_BC
  }
  n16 = Multiply(
   A=Power(
    Base=MaterialFunctionCall!(CameraVector, Rotation Speed=@n14, Texture=@n15, Tiling=@n8, Offset, AxisFadeContrast=@n9) {materialFunction=/Game/Weapons/WeaponSkins/MaterialFunction/TriplanarCameraVector_Rotated}.Result
    Exp=ScalarParameter() {defaultValue=1, group=GalacticSwirl, parameterName=CM_GS_BGPower}
   )
   B=ScalarParameter() {defaultValue=0.5, group=GalacticSwirl, parameterName=CM_GS_BGMultiplier}
  )
  n17 = TextureObjectParameter() {
   group=GalacticSwirl
   parameterName=CM_GS_TextureNoise
   samplerType=SAMPLERTYPE_LinearGrayscale
   texture=/Game/Creative/Sets/PropSets/Primitives/Customization/Textures/GalacticSwirl/T_CM_GalacticSwirl_Grad
  }
  n18 = Time()
  n19 = ScalarParameter() {defaultValue=0.1, group=GalacticSwirl, parameterName=CM_GS_GradientSpeed_Y}
  n20 = Constant()
  n21 = MaterialFunctionCall!(
   TextureObject_Diffuse=@n17
   TextureScale=ScalarParameter() {defaultValue=1, group=GalacticSwirl, parameterName=CM_GS_TextureScale}
   CalculatedBounds=Add(A=NamedRerouteUsage(), B=Multiply(A=@n18, B=@n19))
   WorldNormals=NamedRerouteUsage()
   RotationX=@n20
   RotationY=@n20
   RotationZ=@n20
   ProjectionTransitionContrast=ScalarParameter() {defaultValue=4, group=GalacticSwirl, parameterName=CM_GS_NoiseTransContrast}
   Material Normals
   NormalTexture
   Normal Texture Intensity
  ) {materialFunction=&/MF_Primitive_TriplanarUVs}
  n22 = LinearInterpolate(
   A=Multiply(A=ScalarParameter() {defaultValue=8, group=GalacticSwirl, parameterName=CM_GS_AccentRimBrightness}, B=@n16)
   B=VectorParameter() {group=GalacticSwirl, parameterName=CM_GS_EdgeFresnel, defaultValue=<rgba 5.0909 8.586 10 1>}.RGB
   Alpha=Multiply(
    A=ComponentMask(None=@n21.RGBA) {r=true}
    B=ScalarParameter() {defaultValue=0.2, group=GalacticSwirl, parameterName=CM_GS_GalaxyFlameFresnelIntensity}
   )
  )
  n23 = ScalarParameter() {defaultValue=12, group=GalacticSwirl, parameterName=CM_GS_Galaxy_FresEX}
  n24 = LinearInterpolate(A=NamedRerouteUsage(), B=@n22, Alpha=Saturate(None=Fresnel(ExponentIn=@n23) {baseReflectFraction=0}))
  LinearInterpolate(A=@n24, B=NamedRerouteUsage(), Alpha=NamedRerouteUsage())
  n25 = CollectionParameter() {collection=/Game/Athena/Events/AudioAnalysis/MPC_Audio_Reactive_Parameters, parameterName=Amplitude_Music}
  n26 = MaterialFunctionCall() {materialFunction=&/MF_Primitive_AudioSetup}
  n27 = ScalarParameter() {defaultValue=0.25, group=GalacticSwirl, parameterName=CM_GS_EventAudioBlend}
  n28 = Min(A=Multiply(
    A=LinearInterpolate(A=@n25, B=@n26.Result, Alpha=@n27) {constAlpha=0.25}
    B=ScalarParameter() {defaultValue=1, group=GalacticSwirl, parameterName=CM_GS_EventAudioMultiplier}
   )) {constB=30, desc=prevents crazy values}
  MaterialFunctionCall!(In=@n28, Ramp Width=ScalarParameter() {defaultValue=0.5, group=GalacticSwirl, parameterName=CM_GS_EventAudioRamp}) {materialFunction=/Engine/Functions/Engine_MaterialFunctions01/Math/SmoothCeil}
  n29 = TextureObjectParameter() {
   group=GalacticSwirl
   parameterName=CM_GS_Flowmap
   samplerType=SAMPLERTYPE_LinearColor
   texture=/Game/Creative/Sets/PropSets/Primitives/Customization/Textures/GalacticSwirl/T_CM_GalacticSwirl_Flowmap
  }
  n30 = ScalarParameter() {defaultValue=1, group=GalacticSwirl, parameterName=CM_GS_TextureScale}
  n31 = NamedRerouteUsage()
  n32 = NamedRerouteUsage()
  n33 = Constant()
  n34 = ScalarParameter() {defaultValue=4, group=GalacticSwirl, parameterName=CM_GS_TextureTransContrast}
  n35 = ConstantBiasScale(None=ComponentMask(None=MaterialFunctionCall(
     TextureObject_Diffuse=@n29
     TextureScale=@n30
     CalculatedBounds=@n31
     WorldNormals=@n32
     RotationX=@n33
     RotationY=@n33
     RotationZ=@n33
     ProjectionTransitionContrast=@n34
    ) {materialFunction=&MF_Primitive_TriplanarUVs}.RGBA) {g=true, r=true}) {bias=-0.5, scale=2}
  Multiply(A=@n35, B=Constant2Vector() {g=1, r=-1})
  n36 = Time()
  n37 = Constant() {r=2}
  n38 = Constant() {r=2}
  n39 = TextureObjectParameter() {
   group=GalacticSwirl
   parameterName=CM_GS_TextureStarCenter
   samplerType=SAMPLERTYPE_Masks
   texture=/Game/Creative/Sets/PropSets/Primitives/Customization/Textures/GalacticSwirl/T_CM_GalacticSwirl_Flipbook
  }
  n40 = Multiply(A=Time()) {constB=0.5}
  n41 = MaterialFunctionCall(
   UVs=@n5
   Texture Scale=LinearInterpolate(
    A=ScalarParameter() {defaultValue=0.025, group=GalacticSwirl, parameterName=CM_GS_StarPulseScaleMin, sortPriority=210}
    B=ScalarParameter() {defaultValue=0.05, group=GalacticSwirl, parameterName=CM_GS_StarPulseScaleMax, sortPriority=210}
    Alpha=Sine(None=@n40)
   )
  ) {materialFunction=&ScaleUVsByCenter}
  n42 = Constant() {r=0.5}
  Multiply(
   A=MaterialFunctionCall("Animation  Phase (0-1)"=Multiply(A=@n36, B=@n37), Number of Rows=@n38, Number of Columns=@n38, Texture=@n39, UVs=@n41.UVs) {materialFunction=&FlipBook}.Result
   B=SphereMask(A=@n41.UVs, B=@n42, Radius=@n42) {attenuationRadius=0.5, desc=Mask}
  )
  n43 = ScalarParameter() {defaultValue=0.02, group=GalacticSwirl, parameterName=CM_GS_StarPulseScaleMin, sortPriority=210}
  n44 = ScalarParameter() {defaultValue=0.04, group=GalacticSwirl, parameterName=CM_GS_StarPulseScaleMax, sortPriority=210}
  n45 = Sine(None=Multiply(A=Time()) {constB=0.5})
  n46 = OneMinus(None=Fresnel(ExponentIn=LinearInterpolate(A=@n43, B=@n44, Alpha=@n45)) {baseReflectFraction=0})
  n47 = ScalarParameter() {defaultValue=1, group=GalacticSwirl, parameterName=CM_GS_StarIntensity, sortPriority=210}
  NamedRerouteDeclaration(None=Power(Base=Multiply(A=@n46, B=@n47), Exp=ScalarParameter() {defaultValue=2, group=GalacticSwirl, parameterName=CM_GS_StarPower, sortPriority=210})) {name=Star, nodeColor=<rgba 1 0 0.9609 1>}
  n48 = Add(
   A=Multiply(A=Time(), B=ScalarParameter() {defaultValue=0.5, group=GalacticSwirl, parameterName=CM_GS_FlowMap_Speed}) {constB=0.5}
   B=NamedRerouteUsage()
  )
  n49 = Frac(None=@n48)
  NamedRerouteDeclaration(None=Abs(None=ConstantBiasScale(None=@n49) {bias=-0.5, scale=2})) {name=FlowmapPhaseControl, nodeColor=<rgba 0.5234 0 1 1>}
  NamedRerouteDeclaration(None=@n24) {name=Emissive, nodeColor=<rgba 1 0.6797 0 1>}
  n50 = NamedRerouteUsage()
  NamedRerouteDeclaration(None=LinearInterpolate(
    A=ScalarParameter() {group=GalacticSwirl, parameterName=CM_GS_Specular}
    B=ScalarParameter() {defaultValue=0.1, group=GalacticSwirl, parameterName=CM_GS_SpecularHi}
    Alpha=@n50
   )) {name=Spec, nodeColor=<rgba 0 1 0.3906 1>}
  NamedRerouteDeclaration(None=LinearInterpolate(
    A=ScalarParameter() {group=GalacticSwirl, parameterName=CM_GS_Roughness}
    B=ScalarParameter() {defaultValue=0.1, group=GalacticSwirl, parameterName=CM_GS_RoughnessHi}
    Alpha=@n50
   )) {name=Roughness, nodeColor=<rgba 1 0 0.5859 1>}
  n51 = ScalarParameter() {defaultValue=0.025, group=GalacticSwirl, parameterName=CM_GS_UVPan}
  n52 = Multiply(A=Time(), B=@n51)
  n53 = LinearInterpolate(A=@n51, B=ScalarParameter() {defaultValue=0.025, group=GalacticSwirl, parameterName=CM_GS_UVPanHi}, Alpha=NamedRerouteUsage())
  NamedRerouteDeclaration(None=Add(A=Add(A=@n52, B=@n53), B=NamedRerouteUsage())) {name=UVs, nodeColor=<rgba 0 1 0.4375 1>}
  n54 = Desaturation(None=@n16, Fraction=ScalarParameter() {defaultValue=0.5, group=GalacticSwirl, parameterName=CM_GS_StarSaturation})
  n55 = Multiply(A=@n54, B=VectorParameter() {group=GalacticSwirl, parameterName=CM_GS_StarBrightness, defaultValue=<rgba 300 300 300 1>}.RGB)
  n56 = Time()
  n57 = ScalarParameter() {defaultValue=-0.125, group=GalacticSwirl, parameterName=CM_GS_SmallStarRotationspeed}
  n58 = TextureObjectParameter() {
   group=GalacticSwirl
   parameterName=Stars
   texture=/Game/Creative/Sets/PropSets/Primitives/Customization/Textures/GalacticSwirl/T_CM_GalacticSwirl_M2
  }
  n59 = Time()
  n60 = VectorParameter() {group=GalacticSwirl, parameterName=CM_GS_SmallStarPanSpeed, defaultValue=<rgba 0 0 0 1>}
  n61 = ComponentMask(None=MaterialFunctionCall(
    Rotation Speed=Multiply(A=@n56, B=@n57)
    Texture=@n58
    Tiling=ScalarParameter() {defaultValue=5, group=GalacticSwirl, parameterName=CM_GS_SmallStarTiling}
    Offset=Multiply(A=@n59, B=@n60.RGB)
   ) {materialFunction=&TriplanarCameraVector_Rotated}.Result) {b=true}
  n62 = Multiply(A=@n61, B=@n61)
  n63 = VectorParameter() {group=GalacticSwirl, parameterName=CM_GS_StarBrightness_1, defaultValue=<rgba 950 950 950 1>}
  n64 = ComponentMask(None=MaterialFunctionCall!(
    CameraVector
    Texture=@n58
    Tiling=ScalarParameter() {defaultValue=2, group=GalacticSwirl, parameterName=CM_GS_BrightStarTiling}
    Offset
    AxisFadeContrast
   ) {materialFunction=/Engine/Functions/Engine_MaterialFunctions01/Texturing/TriplanarCameraVector}.Result) {r=true}
  n65 = Multiply(A=@n64, B=@n64)
  n66 = TextureSample(UVs=Panner(Coordinate=TextureCoordinate() {uTiling=2, vTiling=2}) {speedX=0.133, speedY=0.2}) {
   samplerSource=SSM_Wrap_WorldGroupSettings
   texture=&T_CM_GalacticSwirl_M2
  }
  NamedRerouteDeclaration(None=LinearInterpolate(A=LinearInterpolate(A=@n16, B=@n55, Alpha=@n62), B=Multiply(A=@n54, B=@n63.RGB), Alpha=Multiply(A=@n65, B=@n66.G))) {name=Background, nodeColor=<rgba 1 0 0.7266 1>}
  NamedRerouteDeclaration(None=@n28) {name=Hueshift, nodeColor=<rgba 0.6172 0 1 1>}
  n67 = Multiply(A=@n25, B=ScalarParameter() {defaultValue=1, group=GalacticSwirl, parameterName=CM_GS_EventWarpMultiplier})
  NamedRerouteDeclaration(None=MaterialFunctionCall(In=Min(A=@n67) {constB=30, desc=prevents crazy values}) {materialFunction=&SmoothCeil}.Result) {name=Warp, nodeColor=<rgba 0.2188 0 1 1>}
  n68 = Sine(None=Multiply(A=Time(), B=ScalarParameter() {defaultValue=1, group=GalacticSwirl, parameterName=CM_GS_EventHueShiftAmount})) {period=5}
  n69 = NamedRerouteUsage()
  n70 = Ceil(None=@n69)
  n71 = LinearInterpolate(B=Divide(A=Add(A=@n68)), Alpha=@n70)
  n72 = LinearInterpolate(A=@n71, B=Add(A=@n71, B=LinearInterpolate(Alpha=@n69) {constA=0.1, constB=0.14}), Alpha=@n70)
  n73 = TextureObjectParameter() {
   group=GalacticSwirl
   parameterName=CM_GS_CloudsTexture
   texture=/Game/Creative/Sets/PropSets/Primitives/Customization/Textures/GalacticSwirl/T_CM_GalacticSwirl_M1
  }
  n74 = ScalarParameter() {defaultValue=1, group=GalacticSwirl, parameterName=CM_GS_TextureScale}
  n75 = NamedRerouteUsage()
  n76 = Multiply(A=@n35, B=ScalarParameter() {defaultValue=0.2, group=GalacticSwirl, parameterName=CM_GS_FlowMapEffect})
  n77 = Multiply(A=@n76, B=@n49)
  n78 = Constant()
  n79 = Constant()
  n80 = ScalarParameter() {defaultValue=4, group=GalacticSwirl, parameterName=CM_GS_TextureTransContrast}
  n81 = MaterialFunctionCall(
   TextureObject_Diffuse=@n73
   TextureScale=@n74
   CalculatedBounds=Add(A=@n75, B=AppendVector(A=@n77, B=@n78))
   WorldNormals=NamedRerouteUsage()
   RotationX=@n79
   RotationY=@n79
   RotationZ=@n79
   ProjectionTransitionContrast=@n80
  ) {materialFunction=&MF_Primitive_TriplanarUVs}
  n82 = Multiply(A=@n76, B=Frac(None=Add(A=@n48) {constB=0.5}))
  n83 = MaterialFunctionCall(
   TextureObject_Diffuse=@n73
   TextureScale=@n74
   CalculatedBounds=Add(A=@n75, B=AppendVector(A=@n82, B=@n78))
   WorldNormals=NamedRerouteUsage()
   RotationX=@n79
   RotationY=@n79
   RotationZ=@n79
   ProjectionTransitionContrast=@n80
  ) {materialFunction=&MF_Primitive_TriplanarUVs}
  n84 = NamedRerouteUsage()
  n85 = LinearInterpolate(
   A=NamedRerouteUsage()
   B=VectorParameter() {group=GalacticSwirl, parameterName=CM_GS_CloudColor_01, defaultValue=<rgba 0.088 0.0221 0.5 1>}.RGB
   Alpha=LinearInterpolate(A=ComponentMask(None=@n81.RGBA) {r=true}, B=ComponentMask(None=@n83.RGBA) {r=true}, Alpha=@n84)
  )
  n86 = VectorParameter() {group=GalacticSwirl, parameterName=CM_GS_CloudColor_02, defaultValue=<rgba 0.0125 0 0.5 1>}
  n87 = LinearInterpolate(A=ComponentMask(None=@n81.RGBA) {g=true}, B=ComponentMask(None=@n83.RGBA) {g=true}, Alpha=@n84)
  n88 = ComponentMask(None=@n81.RGBA) {b=true}
  n89 = ComponentMask(None=@n83.RGBA) {b=true}
  n90 = LinearInterpolate(
   A=LinearInterpolate(
    A=LinearInterpolate(A=@n85, B=@n86.RGB, Alpha=@n87)
    B=VectorParameter() {group=GalacticSwirl, parameterName=CM_GS_CloudColor_03, defaultValue=<rgba 0.035 0.5 0.5 1>}.RGB
    Alpha=LinearInterpolate(A=@n88, B=@n89, Alpha=@n84)
   )
   B=VectorParameter() {group=GalacticSwirl, parameterName=CM_GS_CloudColor_04, defaultValue=<rgba 0.5 0.0704 0.4969 1>}.RGB
   Alpha=LinearInterpolate(A=ComponentMask(None=@n81.RGBA) {a=true}, B=ComponentMask(None=@n83.RGBA) {a=true}, Alpha=@n84)
  )
  n91 = DotProduct(A=VertexNormalWS(), B=CameraVectorWS())
  NamedRerouteDeclaration(None=Multiply(
    A=MaterialFunctionCall(Hue Shift Percentage=@n72, Texture=@n90) {materialFunction=&HueShift}.Result
    B=Power(Base=@n91) {constExponent=1.5}
   )) {name=GasClouds, nodeColor=<rgba 1 0 0.6797 1>}
  n92 = TextureObjectParameter() {
   group=GalacticSwirl
   parameterName=CM_GS_TextureNoise
   samplerType=SAMPLERTYPE_LinearGrayscale
   texture=&T_CM_GalacticSwirl_Grad
  }
  n93 = ScalarParameter() {defaultValue=1, group=GalacticSwirl, parameterName=CM_GS_TextureScale}
  n94 = NamedRerouteUsage()
  n95 = NamedRerouteUsage()
  n96 = Constant()
  n97 = ScalarParameter() {defaultValue=4, group=GalacticSwirl, parameterName=CM_GS_NoiseTransContrast}
  NamedRerouteDeclaration(None=ComponentMask(None=MaterialFunctionCall(
     TextureObject_Diffuse=@n92
     TextureScale=@n93
     CalculatedBounds=@n94
     WorldNormals=@n95
     RotationX=@n96
     RotationY=@n96
     RotationZ=@n96
     ProjectionTransitionContrast=@n97
    ) {materialFunction=&MF_Primitive_TriplanarUVs}.RGBA) {r=true}) {name=Mask, nodeColor=<rgba 0 0.5312 1 1>}
  NamedRerouteDeclaration(None=@n22) {name=StarColor, nodeColor=<rgba 1 0 0.9609 1>}
  n98 = MaterialFunctionCall() {materialFunction=&/MF_Primitive_CalculatedBounds}
  NamedRerouteDeclaration(None=Divide(A=@n98.WorldPosition, B=ScalarParameter() {defaultValue=175, group=GalacticSwirl, parameterName=CM_GS_WorldScale})) {name=CalculatedBounds, nodeColor=<rgba 0.9922 1 0 1>}
  NamedRerouteDeclaration(None=@n98.World Normals) {name=WorldNormals, nodeColor=<rgba 0.9453 0 1 1>}
 ]}
