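# Material function for Creative primitive props that adds an animated, music-reactive rainbow gas cloud look when the Primitive_GasCloud static switch is on, otherwise passing attributes through.
# A panning triplanar warp texture distorts a triplanar grayscale cloud texture, which indexes a color curve atlas.
# Music amplitude from the audio-reactive parameter collection alters panning and hue shift.
# The result drives base color and dim emissive, with flat or texture-based specular and roughness.
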MaterialFunction /Game/Creative/Sets/PropSets/Primitives/Customization/MaterialFunctions/MF_CM_GasCloudRainbow_01
graph {nodes [
  n1 = Sine(None=Time()) {period=5}
  n2 = CollectionParameter() {collection=/Game/Athena/Events/AudioAnalysis/MPC_Audio_Reactive_Parameters, parameterName=Amplitude_Music}
  n3 = Ceil(None=@n2)
  n4 = LinearInterpolate(B=Divide(A=Add(A=@n1)), Alpha=@n3)
  n5 = LinearInterpolate(B=LinearInterpolate(Alpha=Saturate(None=@n2)) {constA=0.1, constB=-0.1}, Alpha=@n3)
  n6 = CurveAtlasRowParameter(CurveTime=ComponentMask(None=NamedRerouteUsage()) {r=true}) {
   atlas=/Game/Creative/Sets/PropSets/Primitives/Customization/Textures/ColorCurve/CC_CM_Primitives_01
   curve=/Game/Creative/Sets/PropSets/Primitives/Customization/Textures/ColorCurve/C_CM_GasCloud_01
   defaultValue=3
   group=GasCloud
   parameterName=CM_GasCloud_ColorCurve
  }
  n7 = ComponentMask(None=MaterialFunctionCall(Hue Shift Percentage=Add(A=@n4, B=@n5), Texture=@n6) {materialFunction=&HueShift}.Result) {b=true, g=true, r=true}
  n8 = FunctionInput(Preview=SetMaterialAttributes!(
    MaterialAttributes
    BaseColor=Constant3Vector() {constant=<rgba 1 0 0 0>}
    Normal=TextureSample() {samplerType=SAMPLERTYPE_Normal, texture=/Game/Weapons/FORT_Rifles/T_SCAR_N}.RGB
   ) {attributeSetTypes [#1, #2]}) {inputType=FunctionInput_MaterialAttributes}
  n9 = SetMaterialAttributes!(
   MaterialAttributes=@n8
   BaseColor=@n7
   EmissiveColor=Multiply(A=Constant() {r=0.3}, B=@n7)
   Specular=NamedRerouteUsage()
   Roughness=NamedRerouteUsage()
  ) {attributeSetTypes [
    #1
    #3
    #4
    #5
   ]}
  FunctionOutput(None=StaticSwitchParameter(True=@n9, False=@n8) {group=Customization, parameterName=Primitive_GasCloud})
  NamedRerouteDeclaration(None=Divide(
    A=MaterialFunctionCall() {materialFunction=&/MF_Primitive_CalculatedBounds}.WorldPosition
    B=ScalarParameter() {defaultValue=256, group=GasCloud, parameterName=CM_GasCloud_WorldSpaceTiling}
   )) {name=WorldTiling, nodeColor=<rgba 0.4297 0 1 1>}
  n10 = ComponentMask(None=NamedRerouteUsage()) {r=true}
  n11 = ScalarParameter() {defaultValue=1, group=GasCloud, parameterName=CM_GasCloud_TextSpecInst, sliderMax=1}
  NamedRerouteDeclaration(None=LinearInterpolate(
    A=ScalarParameter() {group=GasCloud, parameterName=CM_GasCloud_FlatSpec}
    B=Multiply(A=@n10, B=@n11)
    Alpha=ScalarParameter() {group=GasCloud, parameterName=CM_GasCloudUse_TextureforSpec}
   )) {name=Spec, nodeColor=<rgba 0 0.4072 0.3968 1>}
  n12 = ComponentMask(None=NamedRerouteUsage()) {r=true}
  n13 = ScalarParameter() {defaultValue=1, group=GasCloud, parameterName=CM_GasCloud_TextRoughInst, sliderMax=1}
  NamedRerouteDeclaration(None=LinearInterpolate(
    A=ScalarParameter() {group=GasCloud, parameterName=CM_GasCloud_FlatRough}
    B=Multiply(A=@n12, B=@n13)
    Alpha=ScalarParameter() {group=GasCloud, parameterName=CM_GasCloudUse_TextureforRough}
   )) {name=Rough, nodeColor=<rgba 0.4072 0.2831 0 1>}
  n14 = TextureObjectParameter() {
   group=GasCloud
   parameterName=CM_GasCloud_WarpTex
   samplerType=SAMPLERTYPE_LinearGrayscale
   texture=/Game/Creative/Sets/PropSets/Primitives/Customization/Textures/GasCloud/T_CM_GasCloud_WarpTexture_01
  }
  n15 = Time()
  n16 = VectorParameter() {defaultValue=<rgba 0.02 -0.01 0.03 1>}
  n17 = VectorParameter() {group=GasCloud, parameterName=CM_GasCloud_Rotation, defaultValue=<rgba 0 0 0 1>}
  n18 = MaterialFunctionCall!(
   TextureObject_Diffuse=@n14
   TextureScale=ScalarParameter() {defaultValue=1, group=GasCloud, parameterName=CM_GasCloud_WarpUVTile}
   CalculatedBounds=Add(A=Multiply(A=@n15, B=@n16.RGB), B=NamedRerouteUsage())
   WorldNormals=MaterialFunctionCall() {materialFunction=&MF_Primitive_CalculatedBounds}.World Normals
   RotationX=@n17.R
   RotationY=@n17.R
   RotationZ=@n17.R
   ProjectionTransitionContrast=ScalarParameter() {defaultValue=1, parameterName=CM_GasCloud_Triplannar_Contrast}
   Material Normals
   NormalTexture
   Normal Texture Intensity
  ) {materialFunction=&/MF_Primitive_TriplanarUVs}
  NamedRerouteDeclaration(None=LinearInterpolate(A=Multiply(A=@n18.RGBA) {constB=0}, B=@n18.RGBA, Alpha=ScalarParameter() {defaultValue=1, parameterName=CM_GasCloud_WarpIntes})) {name=NewUVs, nodeColor=<rgba 1 0 0.375 1>}
  n19 = NamedRerouteUsage()
  n20 = NamedRerouteUsage()
  n21 = CollectionParameter() {collection=&MPC_Audio_Reactive_Parameters, parameterName=Amplitude_Music}
  n22 = Multiply(
   A=LinearInterpolate(B=Divide(A=@n21) {constB=300}, Alpha=NamedRerouteUsage())
   B=LinearInterpolate(Alpha=Saturate(None=@n21)) {constA=0.3, constB=-0.3}
  )
  n23 = Add(
   A=LinearInterpolate(
    A=ScalarParameter() {defaultValue=0.03, group=GasCloud, parameterName=CM_GasCloudDefualtPanX}
    B=ScalarParameter() {defaultValue=0.035, group=GasCloud, parameterName=CM_GasCloudMusicPanX}
    Alpha=@n20
   )
   B=@n22
  )
  n24 = Add(
   A=LinearInterpolate(
    A=ScalarParameter() {defaultValue=0.006, group=GasCloud, parameterName=CM_GasCloudDefualtPanY}
    B=ScalarParameter() {defaultValue=0.009, group=GasCloud, parameterName=CM_GasCloudMusicPanY}
    Alpha=@n20
   )
   B=@n22
  )
  n25 = LinearInterpolate(
   A=Constant3Vector()
   B=AppendVector(A=AppendVector(A=@n23, B=@n24), B=ScalarParameter() {defaultValue=0.03, parameterName=CM_GasCloud_ZPanning})
   Alpha=ScalarParameter() {defaultValue=1, group=GasCloud, parameterName=CM_GasCloud_UsePanning}
  )
  n26 = Time()
  NamedRerouteDeclaration(None=Add(A=ComponentMask(None=@n19) {b=true, g=true, r=true}, B=Multiply(A=@n25, B=@n26))) {name=AudioPanner, nodeColor=<rgba 0 0.473 1 1>}
  NamedRerouteDeclaration(None=Ceil(None=Saturate(None=@n21))) {name=CheckIfMusicIsOn, nodeColor=<rgba 1 0.2109 0 1>}
  n27 = TextureObject() {samplerType=SAMPLERTYPE_Grayscale, texture=/Game/Creative/Sets/PropSets/Primitives/Customization/Textures/GasCloud/T_CM_GasCloud_BC_01}
  n28 = NamedRerouteUsage()
  n29 = NamedRerouteUsage()
  n30 = VectorParameter() {group=GasCloud, parameterName=CM_GasCloud_Rotation, defaultValue=<rgba 0 0 0 1>}
  NamedRerouteDeclaration(None=MaterialFunctionCall(
    TextureObject_Diffuse=@n27
    TextureScale=ScalarParameter() {defaultValue=1, group=GasCloud, parameterName=CM_GasCloud_UVTile}
    CalculatedBounds=Add(A=@n28, B=@n29)
    WorldNormals=MaterialFunctionCall() {materialFunction=&MF_Primitive_CalculatedBounds}.World Normals
    RotationX=@n30.R
    RotationY=@n30.R
    RotationZ=@n30.R
    ProjectionTransitionContrast=ScalarParameter() {defaultValue=1, parameterName=CM_GasCloud_Triplannar_Contrast}
   ) {materialFunction=&MF_Primitive_TriplanarUVs}.RGBA) {name=Warped Greyscale, nodeColor=<rgba 0.9768 1 0 1>}
 ]}
