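# Material function for Creative primitive props that applies a grass look to incoming material attributes when the Primitive_Grass static switch is on.
# A triplanar-projected packed grass texture drives base color, specular and roughness.
# Color-mask samples at several world scales build micro and macro tint variation, with camera-distance sphere masks fading detail toward a flat far color.
# Detail and background triplanar normals combine through angle-corrected blending.
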MaterialFunction /Game/Creative/Sets/PropSets/Primitives/Customization/MaterialFunctions/MF_CM_Grass
graph {nodes [
  n1 = TextureObjectParameter() {
   group=Grass
   parameterName=CM_Grass_NoiseTex
   texture=/Game/Creative/Sets/PropSets/Primitives/Customization/Textures/NeonGasCloud/T_CM_NeonGasCloud_M2
  }
  n2 = MaterialFunctionCall!(
   TextureObject=@n1
   TextureSize=ScalarParameter() {defaultValue=1, group=Grass, parameterName=CM_Grass_Noise_UVTile}
   WorldPosition=NamedRerouteUsage()
   World Space Normal=NamedRerouteUsage()
   ProjectionTransitionContrast=ScalarParameter() {defaultValue=10, group=Grass, parameterName=CM_Grass_Noise_TriplannarContrast}
  ) {materialFunction=&/MF_Primitive_TriplanarUVSimple}
  ComponentMask(None=@n2.XYZ Texture) {r=true}
  ComponentMask(None=@n2.XYZ Texture) {a=true}
  Constant3Vector() {constant=<rgba 0.1157 0.15 0.0332 0>}
  n3 = GetMaterialAttributes(None=FunctionInput() {inputType=FunctionInput_MaterialAttributes}) {attributeGetTypes [#1]}
  n4 = SetMaterialAttributes!(
   MaterialAttributes=@n3.MaterialAttributes
   BaseColor=NamedRerouteUsage()
   Specular=NamedRerouteUsage()
   Roughness=NamedRerouteUsage()
   Normal=NamedRerouteUsage()
   Metallic=ScalarParameter() {group=Grass, parameterName=CM_Grass_Metallic}
   EmissiveColor=Constant()
  ) {attributeSetTypes [
    #2
    #3
    #4
    #1
    #5
    #6
   ]}
  FunctionOutput(None=StaticSwitchParameter(True=@n4, False=@n3.MaterialAttributes) {group=Customization, parameterName=Primitive_Grass})
  n5 = Constant() {r=0.01}
  n6 = TextureObjectParameter() {
   group=Grass
   parameterName=CM_Grass_DiffuseTex
   samplerType=SAMPLERTYPE_Masks
   texture=/Game/Environments/Apollo/Sets/Terrain/Textures/T_Apollo_Grass_D_02
  }
  n7 = MaterialFunctionCall(
   TextureObject=@n6
   TextureSize=ScalarParameter() {defaultValue=1, group=Grass, parameterName=CM_Grass_BaseUVTile}
   WorldPosition=NamedRerouteUsage()
   World Space Normal=NamedRerouteUsage()
   ProjectionTransitionContrast=ScalarParameter() {defaultValue=100, group=Grass, parameterName=CM_Grass_BaseTriplannarContrast}
  ) {materialFunction=&MF_Primitive_TriplanarUVSimple}
  n8 = ComponentMask(None=@n7.XYZ Texture) {r=true}
  n9 = Saturate(None=Multiply(A=NamedRerouteUsage(), B=Constant3Vector() {constant=<rgba 0.0932 0.15 0.03 1>}))
  n10 = Saturate(None=Add(A=Multiply(A=@n5, B=@n8), B=Multiply(A=@n9, B=@n8)))
  n11 = Constant3Vector() {constant=<rgba 0.6 0.7258 0.8 1>}
  LinearInterpolate(A=@n10, B=Saturate(None=Divide(A=@n10, B=@n11)), Alpha=Power(Base=ComponentMask(None=@n7.XYZ Texture) {a=true}, Exp=Constant() {r=3}))
  NamedRerouteDeclaration(None=@n3.Normal) {name=InNormal, nodeColor=<rgba 0 0.171 1 1>}
  NamedRerouteDeclaration(None=@n3.MaterialAttributes) {name=Material Attribute, nodeColor=<rgba 0.7323 0 1 1>}
  n12 = NamedRerouteUsage()
  n13 = NamedRerouteUsage()
  n14 = SphereMask(A=WorldPosition().XYZ, B=CameraPositionWS(), Radius=Constant() {r=20000}, Hardness=Constant())
  n15 = SphereMask(A=CameraPositionWS(), B=WorldPosition().XYZ, Radius=Constant() {r=100000}, Hardness=Constant())
  NamedRerouteDeclaration(None=LinearInterpolate(A=LinearInterpolate(A=@n12, B=@n13, Alpha=@n14), B=Constant3Vector() {constant=<rgba 0.7925 1.0999 0.1493 0>}, Alpha=OneMinus(None=@n15))) {name=DistanceBasedMasking, nodeColor=<rgba 0 1 0.6484 1>}
  NamedRerouteDeclaration(None=ComponentMask(None=@n7.XYZ Texture) {g=true}) {name=Specular, nodeColor=<rgba 0 0.6016 1 1>}
  NamedRerouteDeclaration(None=ComponentMask(None=@n7.XYZ Texture) {b=true}) {name=Roughness, nodeColor=<rgba 1 0.1172 0 1>}
  n16 = ScalarParameter() {defaultValue=1, group=Grass, parameterName=CM_Grass_BaseUVTile}
  n17 = NamedRerouteUsage()
  n18 = NamedRerouteUsage()
  n19 = ScalarParameter() {defaultValue=100, group=Grass, parameterName=CM_Grass_BaseTriplannarContrast}
  n20 = NamedRerouteUsage()
  n21 = TextureObjectParameter() {
   group=Grass
   parameterName=CM_Grass_NormalTexture
   samplerType=SAMPLERTYPE_Normal
   texture=/Game/Athena/Apollo/Environments/Landscape/Materials/Biome/CommonMaterials/T_CommonGrass_N
  }
  n22 = ScalarParameter() {defaultValue=1, group=Grass, parameterName=CM_Grass_NormalIntensity}
  n23 = ScalarParameter() {defaultValue=1, group=Grass, parameterName=CM_Grass_BackgrNormalUVTile}
  n24 = Divide(
   A=MaterialFunctionCall() {materialFunction=&/MF_Primitive_CalculatedBounds}.WorldPosition
   B=ScalarParameter() {defaultValue=4560, group=Grass, parameterName=CM_Grass_Backgr_Normal}
  )
  n25 = NamedRerouteUsage()
  n26 = ScalarParameter() {defaultValue=100, group=Grass, parameterName=CM_Grass_BaseTriplannarContrast}
  n27 = NamedRerouteUsage()
  n28 = TextureObjectParameter() {
   group=Grass
   parameterName=CM_GrassBackgr_NormalTexture
   samplerType=SAMPLERTYPE_Normal
   texture=&T_CommonGrass_N
  }
  n29 = ScalarParameter() {defaultValue=1, group=Grass, parameterName=CM_Grass_BackgrNormalIntensity}
  NamedRerouteDeclaration(None=MaterialFunctionCall(
    BaseNormal=MaterialFunctionCall!(
     TextureObject_Diffuse
     TextureScale=@n16
     CalculatedBounds=@n17
     WorldNormals=@n18
     RotationX
     RotationY
     RotationZ
     ProjectionTransitionContrast=@n19
     Material Normals=@n20
     NormalTexture=@n21
     Normal Texture Intensity=@n22
    ) {materialFunction=&/MF_Primitive_TriplanarUVs}.Normal Texture
    AdditionalNormal=MaterialFunctionCall(
     TextureScale=@n23
     CalculatedBounds=@n24
     WorldNormals=@n25
     ProjectionTransitionContrast=@n26
     Material Normals=@n27
     NormalTexture=@n28
     Normal Texture Intensity=@n29
    ) {materialFunction=&MF_Primitive_TriplanarUVs}.Normal Texture
   ) {materialFunction=&BlendAngleCorrectedNormals}.Result) {name=FinalNormal, nodeColor=<rgba 0.8047 0 1 1>}
  n30 = MaterialFunctionCall() {materialFunction=&MF_Primitive_CalculatedBounds}
  n31 = Divide(A=@n30.WorldPosition, B=ScalarParameter() {defaultValue=100, group=Grass, parameterName=CM_Grass_WorldScale})
  NamedRerouteDeclaration(None=@n31) {name=UVS, nodeColor=<rgba 0.9922 0 1 1>}
  NamedRerouteDeclaration(None=@n31) {name=CalculatedBounds, nodeColor=<rgba 1 0.1964 0 1>}
  NamedRerouteDeclaration(None=@n30.World Normals) {name=WorldNorm, nodeColor=<rgba 0 1 0.9531 1>}
  NamedRerouteDeclaration(None=@n10) {name=Base Color, nodeColor=<rgba 0 0.4141 1 1>}
  NamedRerouteDeclaration(None=LinearInterpolate(
    A=ScalarParameter() {group=Grass, parameterName=CM_Grass_Specular}
    B=ScalarParameter() {defaultValue=1, group=Grass, parameterName=CM_Grass_SpecularHi}
    Alpha=NamedRerouteUsage()
   )) {name=FinalSpec, nodeColor=<rgba 0 0.4072 0.3968 1>}
  NamedRerouteDeclaration(None=LinearInterpolate(
    A=ScalarParameter() {group=Grass, parameterName=CM_Grass_Roughness}
    B=ScalarParameter() {defaultValue=1, group=Grass, parameterName=CM_Grass_RoughnessHi}
    Alpha=NamedRerouteUsage()
   )) {name=FinalRough, nodeColor=<rgba 0.4072 0.2831 0 1>}
  n32 = Constant3Vector() {constant=<rgba 0.3797 0.7333 0.0733 1>}
  n33 = TextureObject() {
   samplerType=SAMPLERTYPE_Grayscale
   texture=/Game/Athena/Apollo/Environments/Landscape/Materials/Biome/CommonMaterials/T_ApolloGrass_ColorMask
  }
  n34 = ScalarParameter() {defaultValue=1, group=Grass, parameterName=CM_Grass_BaseUVTile}
  n35 = Divide(
   A=MaterialFunctionCall() {materialFunction=&MF_Primitive_CalculatedBounds}.WorldPosition
   B=ScalarParameter() {defaultValue=4560, group=Grass, parameterName=CM_Grass_Color_Macro_Small}
  )
  n36 = NamedRerouteUsage()
  n37 = ScalarParameter() {defaultValue=100, group=Grass, parameterName=CM_Grass_BaseTriplannarContrast}
  n38 = LinearInterpolate(
   A=Constant()
   B=Constant() {r=1.5}
   Alpha=ComponentMask(None=MaterialFunctionCall(TextureObject=@n33, TextureSize=@n34, WorldPosition=@n35, World Space Normal=@n36, ProjectionTransitionContrast=@n37) {materialFunction=&MF_Primitive_TriplanarUVSimple}.XYZ Texture) {a=true}
  )
  n39 = Constant() {r=1.5}
  n40 = Constant() {r=-9}
  n41 = TextureObject() {
   samplerType=SAMPLERTYPE_Grayscale
   texture=&T_ApolloGrass_ColorMask
  }
  n42 = MaterialFunctionCall() {materialFunction=&MF_Primitive_CalculatedBounds}
  n43 = ScalarParameter() {defaultValue=-3600, group=Grass, parameterName=CM_Grass_Color_Macro_Mid}
  n44 = ComponentMask(None=MaterialFunctionCall(
    TextureObject=@n41
    TextureSize=ScalarParameter() {defaultValue=1, group=Grass, parameterName=CM_Grass_BaseUVTile}
    WorldPosition=Divide(A=@n42.WorldPosition, B=@n43)
    World Space Normal=NamedRerouteUsage()
    ProjectionTransitionContrast=ScalarParameter() {defaultValue=100, group=Grass, parameterName=CM_Grass_BaseTriplannarContrast}
   ) {materialFunction=&MF_Primitive_TriplanarUVSimple}.XYZ Texture) {a=true}
  n45 = LinearInterpolate(
   A=LinearInterpolate(
    A=Constant3Vector() {constant=<rgba 0.8038 1.0666 0.1066 1>}
    B=Constant3Vector() {constant=<rgba 0.2396 0.6666 0 1>}
    Alpha=Saturate(None=@n38)
   )
   B=Constant3Vector() {constant=<rgba 0.711 1.0666 0 1>}
   Alpha=Saturate(None=LinearInterpolate(A=@n39, B=@n40, Alpha=@n44))
  )
  n46 = Constant3Vector() {constant=<rgba 0.9853 1.2 0.3 1>}
  n47 = TextureObject() {
   samplerType=SAMPLERTYPE_Grayscale
   texture=&T_ApolloGrass_ColorMask
  }
  n48 = MaterialFunctionCall() {materialFunction=&MF_Primitive_CalculatedBounds}
  n49 = Constant() {r=500}
  n50 = MaterialFunctionCall(
   TextureObject=@n47
   TextureSize=ScalarParameter() {defaultValue=1, group=Grass, parameterName=CM_Grass_BaseUVTile}
   WorldPosition=Divide(A=Add(A=@n48.WorldPosition, B=@n49), B=ScalarParameter() {defaultValue=5599, group=Grass, parameterName=CM_Grass_Color_Macro_Large})
   World Space Normal=NamedRerouteUsage()
   ProjectionTransitionContrast=ScalarParameter() {defaultValue=100, group=Grass, parameterName=CM_Grass_BaseTriplannarContrast}
  ) {materialFunction=&MF_Primitive_TriplanarUVSimple}
  n51 = Saturate(None=LinearInterpolate(A=Constant() {r=-0.6}, B=Constant() {r=2}, Alpha=ComponentMask(None=@n50.XYZ Texture) {a=true}))
  n52 = TextureObject() {
   samplerType=SAMPLERTYPE_Grayscale
   texture=&T_ApolloGrass_ColorMask
  }
  n53 = MaterialFunctionCall() {materialFunction=&MF_Primitive_CalculatedBounds}
  n54 = Constant() {r=-13000}
  n55 = ComponentMask(None=MaterialFunctionCall(
    TextureObject=@n52
    TextureSize=ScalarParameter() {defaultValue=1, group=Grass, parameterName=CM_Grass_BaseUVTile}
    WorldPosition=Divide(A=@n53.WorldPosition, B=@n54)
    World Space Normal=NamedRerouteUsage()
    ProjectionTransitionContrast=ScalarParameter() {defaultValue=100, group=Grass, parameterName=CM_Grass_BaseTriplannarContrast}
   ) {materialFunction=&MF_Primitive_TriplanarUVSimple}.XYZ Texture) {r=true}
  n56 = Multiply(A=@n55) {constB=4}
  NamedRerouteDeclaration(None=LinearInterpolate(A=@n32, B=LinearInterpolate(A=@n45, B=@n46, Alpha=@n51), Alpha=Saturate(None=@n56))) {name=MicroGrassColors, nodeColor=<rgba 1 0 0.9844 1>}
  NamedRerouteDeclaration(None=LinearInterpolate(A=@n32, B=Constant3Vector() {constant=<rgba 1.2054 1.4666 0.2254 1>}, Alpha=@n55)) {name=MacroGrassColorLerp, nodeColor=<rgba 0.0781 0 1 1>}
  ScalarParameter() {defaultValue=100000, parameterName=DistanceFarEffect}
 ]}
