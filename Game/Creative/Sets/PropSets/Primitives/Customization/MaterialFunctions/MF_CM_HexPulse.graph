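# Material function that adds an optional animated, glowing HexPulse look to Creative primitive props, gated by the Primitive_HexPulse static switch.
# Triplanar hex textures at several bump-offset depths, height-lerped masks and a panning gradient build layered emissive patterns over background hexes.
# Glow color comes from a curve atlas warped by screen-space noise and hue-shifted, then lighten-blended with a Fresnel core glow.
# It reads music amplitude from an audio-reactive parameter collection and overrides the incoming material attributes.
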MaterialFunction /Game/Creative/Sets/PropSets/Primitives/Customization/MaterialFunctions/MF_CM_HexPulse
graph {nodes [
  n1 = NamedRerouteUsage()
  Divide(A=@n1) {constB=50}
  n2 = LinearInterpolate(Alpha=@n1) {constA=1, constB=0}
  n3 = Ceil(None=@n1)
  n4 = Constant3Vector() {constant=<rgba 0.0753 0.0285 0.1844 1>}
  n5 = ComponentMask(None=NamedRerouteUsage()) {r=true}
  n6 = Subtract(A=@n5, B=Fresnel() {baseReflectFraction=0.02, exponent=2})
  n7 = OneMinus(None=Fresnel(ExponentIn=Constant() {r=0.08}, BaseReflectFractionIn=Constant() {r=0.01}))
  n8 = ScalarParameter() {defaultValue=0.08, group=HexPulse, parameterName=CM_HexPulse_CoreGlow}
  n9 = Sine(None=Time()) {period=5}
  n10 = NamedRerouteUsage()
  n11 = Ceil(None=@n10)
  n12 = LinearInterpolate(B=Divide(A=Add(A=@n9)), Alpha=@n11)
  n13 = LinearInterpolate(Alpha=@n10) {constA=0.1, constB=0.14}
  n14 = TextureObjectParameter() {
   group=HexPulse
   parameterName=CM_HexPulse_UVWarpTexture
   samplerType=SAMPLERTYPE_Masks
   texture=/Game/Global/Textures/Default/Noise/T_WrapEdgeWear_Noise
  }
  n15 = TextureSample(UVs=ScreenPosition().ViewportUV, Tex=@n14) {samplerSource=SSM_Wrap_WorldGroupSettings, samplerType=SAMPLERTYPE_LinearGrayscale}
  n16 = ScalarParameter() {defaultValue=0.25, group=HexPulse, parameterName=CM_HexPulse_ColorWarpAmount}
  n17 = CurveAtlasRowParameter(CurveTime=Add(A=NamedRerouteUsage(), B=Multiply(A=@n15.R, B=@n16))) {
   atlas=/Game/Creative/Sets/PropSets/Primitives/Customization/Textures/ColorCurve/CC_CM_Primitives_01
   curve=/Game/Creative/Sets/PropSets/Primitives/Customization/Textures/ColorCurve/C_CM_SynthWave_01
   defaultValue=2
   group=HexPulse
   parameterName=CM_HexPulse_ColorCurve
  }
  n18 = MaterialFunctionCall(Hue Shift Percentage=LinearInterpolate(A=@n12, B=Add(A=@n12, B=@n13), Alpha=@n11), Texture=@n17) {materialFunction=&HueShift}
  n19 = Multiply(A=LinearInterpolate(A=Multiply(A=@n7, B=@n8), Alpha=@n5) {constB=0}, B=@n18.Result)
  n20 = LinearInterpolate(B=@n18.Result, Alpha=ComponentMask(None=NamedRerouteUsage()) {r=true})
  n21 = FunctionInput() {inputType=FunctionInput_MaterialAttributes}
  n22 = SetMaterialAttributes!(
   MaterialAttributes=@n21
   BaseColor=MaterialFunctionCall(Hue Shift Percentage=LinearInterpolate(B=@n2, Alpha=@n3), Texture=LinearInterpolate(B=@n4, Alpha=@n6)) {materialFunction=&HueShift}.Result
   EmissiveColor=Multiply(
    A=MaterialFunctionCall!(Blend=@n19, Base=@n20) {materialFunction=/Engine/Functions/Engine_MaterialFunctions03/Blends/Blend_Lighten}.Result
    B=NamedRerouteUsage()
   )
   Specular=NamedRerouteUsage()
   Roughness=NamedRerouteUsage()
   Normal
  ) {attributeSetTypes [
    #1
    #2
    #3
    #4
    #5
   ]}
  FunctionOutput(None=StaticSwitchParameter(True=@n22, False=@n21) {group=Customization, parameterName=Primitive_HexPulse})
  n23 = NamedRerouteUsage()
  LinearInterpolate(Alpha=@n23) {constA=0.2, constB=6}
  n24 = MaterialFunctionCall!(
   HeightTexture=LinearInterpolate(Alpha=NamedRerouteUsage()) {constA=1, constB=6}
   TextureCoordinates
   ReferencePlane
   HeightRatioInput=Constant() {r=-0.05}
   CameraVector
  ) {materialFunction=&/MF_Primitive_TriplannarBumpOffset}
  n25 = NamedRerouteUsage()
  Multiply(A=MaterialFunctionCall!(
    TextureObject_Diffuse=TextureObject() {samplerType=SAMPLERTYPE_Masks, texture=/Game/Creative/Sets/PropSets/Primitives/Customization/Textures/Hex/T_CM_HexPulse_BC_01}
    TextureScale=NamedRerouteUsage()
    CalculatedBounds=Add(A=@n24.Distortion, B=@n25)
    WorldNormals=NamedRerouteUsage()
    RotationX
    RotationY
    RotationZ
    ProjectionTransitionContrast=ScalarParameter() {defaultValue=10, group=HexPulse, parameterName=CM_HexPulse_ EmissiveContrast}
    Material Normals
    NormalTexture
    Normal Texture Intensity
   ) {materialFunction=&/MF_Primitive_TriplanarUVs}.RGBA)
  n26 = MaterialFunctionCall() {materialFunction=&/MF_Primitive_CalculatedBounds}
  NamedRerouteDeclaration(None=@n26.Calculated Bounds) {name=CalculatedBounds, nodeColor=<rgba 1 0.5156 0 1>}
  n27 = LinearInterpolate(Alpha=NamedRerouteUsage()) {constA=0.2, constB=6}
  n28 = Constant() {r=-0.05}
  n29 = MaterialFunctionCall(
   TextureObject_Diffuse=TextureObject() {samplerType=SAMPLERTYPE_LinearColor, texture=/Game/Creative/Sets/PropSets/Primitives/Customization/Textures/Hex/T_CM_HexPulse_Mask}
   TextureScale=NamedRerouteUsage()
   CalculatedBounds=Add(
    A=MaterialFunctionCall(HeightTexture=@n27, HeightRatioInput=@n28) {materialFunction=&MF_Primitive_TriplannarBumpOffset}.Distortion
    B=NamedRerouteUsage()
   )
   WorldNormals=NamedRerouteUsage()
   ProjectionTransitionContrast=ScalarParameter() {defaultValue=10, group=HexPulse, parameterName=CM_HexPulse_ EmissiveContrast}
  ) {materialFunction=&MF_Primitive_TriplanarUVs}
  n30 = LinearInterpolate(Alpha=@n23) {constA=0.2, constB=5}
  n31 = Constant() {r=-0.05}
  n32 = MaterialFunctionCall(
   TextureObject_Diffuse=TextureObject() {samplerType=SAMPLERTYPE_LinearColor, texture=&T_CM_HexPulse_Mask}
   TextureScale=NamedRerouteUsage()
   CalculatedBounds=Add(
    A=MaterialFunctionCall(HeightTexture=@n30, HeightRatioInput=@n31) {materialFunction=&MF_Primitive_TriplannarBumpOffset}.Distortion
    B=NamedRerouteUsage()
   )
   WorldNormals=NamedRerouteUsage()
   ProjectionTransitionContrast=ScalarParameter() {defaultValue=10, group=HexPulse, parameterName=CM_HexPulse_ EmissiveContrast}
  ) {materialFunction=&MF_Primitive_TriplanarUVs}
  n33 = MaterialFunctionCall(Transition Phase=NamedRerouteUsage(), Height Texture=ComponentMask(None=@n32.RGBA) {a=true}) {materialFunction=&HeightLerp}
  NamedRerouteDeclaration(None=Multiply(A=ComponentMask(None=@n29.RGBA) {r=true}, B=@n33.Alpha)) {name=EmissiveMask_01, nodeColor=<rgba 0.7109 1 0 1>}
  n34 = NamedRerouteUsage()
  n35 = ScalarParameter() {defaultValue=1, group=HexPulse, parameterName=CM_HexPulse_TextRoughInst, sliderMax=1}
  NamedRerouteDeclaration(None=LinearInterpolate(
    A=ScalarParameter() {group=HexPulse, parameterName=CM_HexPulse_FlatRough}
    B=Multiply(A=@n34, B=@n35)
    Alpha=ScalarParameter() {group=HexPulse, parameterName=CM_HexPulse_Use_TextureforRough}
   )) {name=Rough, nodeColor=<rgba 0.4072 0.2831 0 1>}
  n36 = NamedRerouteUsage()
  n37 = Saturate(None=@n36)
  NamedRerouteDeclaration(None=LinearInterpolate(B=Multiply(A=@n37) {constB=4}, Alpha=Ceil(None=@n36)) {constA=1}) {name=Music Intensity Mult, nodeColor=<rgba 0 1 0.9063 1>}
  NamedRerouteDeclaration(None=ComponentMask(None=VectorParameter() {group=HexPulse, parameterName=CM_HexPulse_EmissiveTexScale, defaultValue=<rgba 1 1 1 1>}.RGB) {b=true, g=true, r=true}) {name=EmissiveTexScale, nodeColor=<rgba 0.0547 0 1 1>}
  n38 = LinearInterpolate(Alpha=NamedRerouteUsage()) {constA=0.2, constB=2}
  n39 = Constant() {r=-0.05}
  n40 = MaterialFunctionCall(
   TextureObject_Diffuse=TextureObject() {samplerType=SAMPLERTYPE_LinearColor, texture=&T_CM_HexPulse_Mask}
   TextureScale=NamedRerouteUsage()
   CalculatedBounds=Add(
    A=MaterialFunctionCall(HeightTexture=@n38, HeightRatioInput=@n39) {materialFunction=&MF_Primitive_TriplannarBumpOffset}.Distortion
    B=NamedRerouteUsage()
   )
   WorldNormals=NamedRerouteUsage()
   ProjectionTransitionContrast=ScalarParameter() {defaultValue=10, group=HexPulse, parameterName=CM_HexPulse_ EmissiveContrast}
  ) {materialFunction=&MF_Primitive_TriplanarUVs}
  NamedRerouteDeclaration(None=Multiply(A=@n33.Alpha, B=ComponentMask(None=@n40.RGBA) {r=true})) {name=EmissiveMask_03, nodeColor=<rgba 0.7109 1 0 1>}
  NamedRerouteDeclaration(None=@n26.LocalPosition) {name=LocalPosition, nodeColor=<rgba 1 0 0.3281 1>}
  NamedRerouteDeclaration(None=Fresnel() {baseReflectFraction=0, exponent=3.5}) {name=Fresnel_Mask, nodeColor=<rgba 0 0.625 1 1>}
  NamedRerouteDeclaration(None=Divide(A=@n26.WorldPosition, B=ScalarParameter() {defaultValue=100, group=HexPulse, parameterName=CM_HexPulse_Size})) {name=WorldPositionScale, nodeColor=<rgba 0 1 0.5078 1>}
  n41 = Divide(A=MaterialFunctionCall() {materialFunction=&MF_Primitive_CalculatedBounds}.WorldPosition) {constB=400}
  n42 = Add(A=NamedRerouteUsage(), B=Multiply(A=Constant3Vector() {constant=<rgba 0 0 0.25 0>}, B=Time()))
  NamedRerouteDeclaration(None=MaterialFunctionCall(
    TextureObject_Diffuse=TextureObject() {samplerType=SAMPLERTYPE_Masks, texture=/Game/Creative/Sets/PropSets/Primitives/Customization/Textures/Hex/T_CM_HexPulse_Gradient_01}
    CalculatedBounds=Add(A=@n41, B=@n42)
    WorldNormals=NamedRerouteUsage()
   ) {materialFunction=&MF_Primitive_TriplanarUVs}.RGBA) {name=Panning Emissive Mask, nodeColor=<rgba 1 0 0.3047 1>}
  NamedRerouteDeclaration(None=@n26.World Normals) {name=World Normals, nodeColor=<rgba 0 1 0.25 1>}
  NamedRerouteDeclaration(None=CollectionParameter() {collection=/Game/Athena/Events/AudioAnalysis/MPC_Audio_Reactive_Parameters, parameterName=Amplitude_Music}) {name=Music Amp, nodeColor=<rgba 0 1 0.3438 1>}
  NamedRerouteDeclaration(None=Multiply(A=@n33.Alpha, B=ComponentMask(None=@n32.RGBA) {r=true})) {name=EmissiveMask_02, nodeColor=<rgba 0.7109 1 0 1>}
  n43 = TextureObject() {samplerType=SAMPLERTYPE_Masks, texture=/Game/Creative/Sets/PropSets/Primitives/Customization/Textures/Hex/T_CM_HexPulse_BC_02}
  n44 = NamedRerouteUsage()
  n45 = NamedRerouteUsage()
  n46 = Add(
   A=MaterialFunctionCall(HeightTexture=LinearInterpolate(Alpha=@n45) {constA=0.2, constB=5}, HeightRatioInput=Constant() {r=-0.05}) {materialFunction=&MF_Primitive_TriplannarBumpOffset}.Distortion
   B=NamedRerouteUsage()
  )
  n47 = NamedRerouteUsage()
  n48 = ScalarParameter() {defaultValue=10, group=HexPulse, parameterName=CM_HexPulse_ EmissiveContrast}
  n49 = Multiply(
   A=ComponentMask(None=MaterialFunctionCall(TextureObject_Diffuse=@n43, TextureScale=@n44, CalculatedBounds=@n46, WorldNormals=@n47, ProjectionTransitionContrast=@n48) {materialFunction=&MF_Primitive_TriplanarUVs}.RGBA) {r=true}
   B=NamedRerouteUsage()
  )
  n50 = TextureObject() {samplerType=SAMPLERTYPE_Masks, texture=/Game/Creative/Sets/PropSets/Primitives/Customization/Textures/Hex/T_CM_HexPulse_BC_03}
  n51 = NamedRerouteUsage()
  n52 = NamedRerouteUsage()
  n53 = Add(
   A=MaterialFunctionCall(HeightTexture=LinearInterpolate(Alpha=@n52) {constA=0.2, constB=2}, HeightRatioInput=Constant() {r=-0.05}) {materialFunction=&MF_Primitive_TriplannarBumpOffset}.Distortion
   B=NamedRerouteUsage()
  )
  n54 = NamedRerouteUsage()
  n55 = ScalarParameter() {defaultValue=10, group=HexPulse, parameterName=CM_HexPulse_ EmissiveContrast}
  n56 = Multiply(
   A=ComponentMask(None=MaterialFunctionCall(TextureObject_Diffuse=@n50, TextureScale=@n51, CalculatedBounds=@n53, WorldNormals=@n54, ProjectionTransitionContrast=@n55) {materialFunction=&MF_Primitive_TriplanarUVs}.RGBA) {r=true}
   B=NamedRerouteUsage()
  )
  NamedRerouteDeclaration(None=Add(B=Add(A=@n49, B=@n56))) {name=Hexes for Emissive, nodeColor=<rgba 1 0 0.6094 1>}
  n57 = TextureObject() {samplerType=SAMPLERTYPE_Masks, texture=/Game/Creative/Sets/PropSets/Primitives/Customization/Textures/Hex/T_CM_HexPulse_Background_BC}
  n58 = NamedRerouteUsage()
  n59 = VectorParameter() {group=HexPulse, parameterName=CM_HexPulse_BackgroundRotation, defaultValue=<rgba 0 0 0 1>}
  NamedRerouteDeclaration(None=MaterialFunctionCall(
    TextureObject_Diffuse=@n57
    TextureScale=ScalarParameter() {defaultValue=1, group=HexPulse, parameterName=CM_HexPulse_BackgroundScale}
    CalculatedBounds=Multiply(A=@n58) {constB=1.3}
    WorldNormals=NamedRerouteUsage()
    RotationX=@n59.R
    RotationY=@n59.R
    RotationZ=@n59.R
    ProjectionTransitionContrast=ScalarParameter() {defaultValue=3, group=HexPulse, parameterName=CM_HexPulse_Background_Contrast}
   ) {materialFunction=&MF_Primitive_TriplanarUVs}.RGBA) {name=Background Hexes, nodeColor=<rgba 0 0.104 1 1>}
  n60 = MaterialFunctionCall(Color=NamedRerouteUsage()) {materialFunction=&SplitComponents}
  n61 = ScalarParameter() {group=HexPulse, parameterName=CM_HexPulse_UseYAxis}
  n62 = ScalarParameter() {group=HexPulse, parameterName=CM_HexPulse_UseZAxis}
  NamedRerouteDeclaration(None=LinearInterpolate(A=LinearInterpolate(A=@n60.R, B=@n60.R, Alpha=@n61), B=LinearInterpolate(A=@n60.G, B=@n60.G, Alpha=@n62))) {name=BoundsDirection, nodeColor=<rgba 0.1719 0 1 1>}
  n63 = NamedRerouteUsage()
  n64 = ScalarParameter() {defaultValue=1, group=HexPulse, parameterName=CM_HexPulse_TextSpecInst, sliderMax=1}
  NamedRerouteDeclaration(None=LinearInterpolate(
    A=ScalarParameter() {group=HexPulse, parameterName=CM_HexPulse_FlatSpec}
    B=Multiply(A=@n63, B=@n64)
    Alpha=ScalarParameter() {group=HexPulse, parameterName=CM_HexPulse_Use_TextureforSpec}
   )) {name=Spec, nodeColor=<rgba 0 0.4072 0.3968 1>}
  n65 = TextureObject() {samplerType=SAMPLERTYPE_Masks, texture=/Game/Creative/Sets/PropSets/Primitives/Customization/Textures/Hex/T_CM_HexPulse_RandomColor}
  n66 = NamedRerouteUsage()
  n67 = Add(
   A=NamedRerouteUsage()
   B=MaterialFunctionCall(HeightTexture=LinearInterpolate(Alpha=@n66) {constA=0.2, constB=2}, HeightRatioInput=Constant() {r=-0.05}) {materialFunction=&MF_Primitive_TriplannarBumpOffset}.Distortion
  )
  n68 = NamedRerouteUsage()
  n69 = ScalarParameter() {defaultValue=10, group=HexPulse, parameterName=CM_HexPulse_ EmissiveContrast}
  NamedRerouteDeclaration(None=ComponentMask(None=MaterialFunctionCall(TextureObject_Diffuse=@n65, CalculatedBounds=@n67, WorldNormals=@n68, ProjectionTransitionContrast=@n69) {materialFunction=&MF_Primitive_TriplanarUVs}.RGBA) {b=true, g=true, r=true}) {name=HexPopping, nodeColor=<rgba 0.4297 1 0 1>}
  NamedRerouteUsage()
  SetMaterialAttributes!(
   MaterialAttributes
   BaseColor=Constant3Vector() {constant=<rgba 1 0 0 0>}
   Normal=TextureSample() {samplerType=SAMPLERTYPE_Normal, texture=/Game/Weapons/FORT_Rifles/T_SCAR_N}.RGB
  ) {attributeSetTypes [#1, #5]}
 ]}
