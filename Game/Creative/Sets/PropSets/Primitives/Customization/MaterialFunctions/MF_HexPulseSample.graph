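# Material function for Primitives prop customization that renders a music-reactive, pulsing hex surface with fake depth, outputting base color, emissive and normal.
# A Custom HLSL loop layers parallax-offset hex texture samples from the Position and Camera Info inputs, with time and MPC music amplitude culling layers.
# The depth result masks base color and normal, and drives emissive tinted by a noise-scrolled thin-film spectrum.
# Quality switches reduce depth and iterations on Low and Medium.
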MaterialFunction /Game/Creative/Sets/PropSets/Primitives/Customization/MaterialFunctions/MF_HexPulseSample
graph {nodes [
  n1 = VectorParameter() {group=Hex Pulse, parameterName=Hex Pulse BG Color, defaultValue=<rgba 0.0278 0.0207 0.0312 0>}
  n2 = Divide(A=ScalarParameter() {defaultValue=20, group=Hex Pulse, parameterName=Hex Pulse Depth}) {constB=100}
  MaterialExpressionQualitySwitch_0 = QualitySwitch(Default=@n2, Low=Multiply(A=@n2) {constB=0.25}, High=@n2, Medium=Multiply(A=@n2) {constB=0.5}, Epic=@n2)
  MaterialExpressionFunctionInput_4 = FunctionInput() {inputName=Camera Info, inputType=FunctionInput_Vector2, sortPriority=3}
  MaterialExpressionFunctionInput_5 = FunctionInput() {inputName=Position, inputType=FunctionInput_Vector2, sortPriority=4}
  n3 = Add(
   A=Multiply(A=Multiply(A=@MaterialExpressionQualitySwitch_0, B=@MaterialExpressionFunctionInput_4)) {constB=-1}
   B=@MaterialExpressionFunctionInput_5
  )
  n4 = TextureSample(UVs=@n3) {texture=/Game/Creative/Sets/PropSets/Primitives/Customization/Textures/Hex/T_CM_HexPulse_M}
  MaterialExpressionTextureObject_0 = TextureObject() {texture=&T_CM_HexPulse_M}
  n5 = ScalarParameter() {defaultValue=12, group=Hex Pulse, parameterName=Hex Pulse Iterations}
  MaterialExpressionQualitySwitch_1 = QualitySwitch(Default=@n5, Low=Multiply(A=@n5) {constB=0.25}, High=@n5, Medium=Multiply(A=@n5) {constB=0.5}, Epic=@n5)
  n6 = CollectionParameter() {collection=/Game/Athena/Events/AudioAnalysis/MPC_Audio_Reactive_Parameters, parameterName=Amplitude_Music}
  n7 = LinearInterpolate(
   B=Multiply(A=ScalarParameter() {defaultValue=3, group=Hex Pulse, parameterName=Hex Pulse Music Amplitude Scaler}, B=@n6) {constB=2}
   Alpha=Ceil(None=@n6)
  ) {constA=1}
  MaterialExpressionReroute_0 = Reroute(None=LinearInterpolate(Alpha=Saturate(None=@n7)) {constA=0.25})
  MaterialExpressionMultiply_19 = Multiply(A=Time(), B=ScalarParameter() {defaultValue=0.5, group=Hex Pulse, parameterName=Hex Pulse Animation Speed})
  n8 = Custom!(
   Tex=@MaterialExpressionTextureObject_0
   MaxDepth=@MaterialExpressionQualitySwitch_0
   Iterations=@MaterialExpressionQualitySwitch_1
   UV=@MaterialExpressionFunctionInput_5
   CameraInfo=@MaterialExpressionFunctionInput_4
   Cutoff=@MaterialExpressionReroute_0
   Time=@MaterialExpressionMultiply_19
  ) {
   code="float3 color = float3(0,0,0);\r\n\r\nfloat depth = 0;\r\n\r\n//Texture2DSample(Tex, TexSampler, UV);\r\n\r\n\r\nfor (int i = 0; i < Iterations; i++)\r\n{\r\n\tfloat increasingIntensity = i/Iterations; \t\t\t\t\t// 0 --> 1\r\n\tfloat decreasingIntensity = (Iterations - i)/Iterations;\t// 1 --> 0\r\n\tfloat2 sampleUV = UV  + CameraInfo * decreasingIntensity * -MaxDepth; // UV + displacement por normal map + offset por camera ray\t\r\n\t\r\n\t// Sample depth                                      \r\n\tfloat4 sampledTexture = Texture2DSample(Tex, TexSampler, sampleUV);\r\n\tfloat sampledValue = sampledTexture.g;\r\n\tfloat sampleddepth = sampledTexture.r;\t\r\n\t\r\n\t// increasingly removes the bottom part of every iteration to keep the highest parts\t\t \r\n\tfloat extractedDepth = sampleddepth * ceil(sampleddepth - increasingIntensity) * increasingIntensity; \r\n\r\n\t// Kill iterations based on Cutoff value\t\t\t\r\n\tsampledValue = lerp(0.3 , 1 , sin(sampledValue*4 + Time) * 0.5 + 0.5);\t\r\n\textractedDepth *= saturate(ceil(Cutoff * Iterations * sampledValue  - i));\r\n\t\r\n\t// Compare\r\n\tfloat compareDepth = saturate(saturate(extractedDepth - depth) * 1000);\t\r\n\t\r\n\t// return\r\n\tdepth = max(depth,extractedDepth);\t\t\r\n\tcolor = depth;\r\n\t\r\n}\r\n\r\nreturn color;\r\n"
   inputs [
    {
     inputName=Tex
     input {expression=&self:MaterialExpressionTextureObject_0}
    }
    {
     inputName=MaxDepth
     input {expression=&self:MaterialExpressionQualitySwitch_0}
    }
    {
     inputName=Iterations
     input {expression=&self:MaterialExpressionQualitySwitch_1}
    }
    {
     inputName=UV
     input {expression=&self:MaterialExpressionFunctionInput_5}
    }
    {
     inputName=CameraInfo
     input {expression=&self:MaterialExpressionFunctionInput_4}
    }
    {
     inputName=Cutoff
     input {expression=&self:MaterialExpressionReroute_0}
    }
    {
     inputName=Time
     input {expression=&self:MaterialExpressionMultiply_19}
    }
   ]
  }
  n9 = OneMinus(None=Ceil(None=@n8))
  FunctionOutput(None=LinearInterpolate(A=Multiply(A=@n1.RGB, B=@n4.G), Alpha=@n9) {constB=0}) {outputName=Base Color}
  n10 = WorldPosition()
  n11 = Multiply(A=Time(), B=ScalarParameter() {defaultValue=200, group=Hex Pulse, parameterName=Hex Pulse Color Speed})
  n12 = TextureSample(UVs=Noise!(World Position=Add(A=@n10.XYZ, B=@n11), FilterWidth) {levels=2, noiseFunction=NOISEFUNCTION_GradientTex3D, outputMin=0, scale=0.001}) {texture=/Game/Global/Textures/Default/Thinfilm/T_ThinFilm_Spectrum_COLOR}
  n13 = ScalarParameter() {defaultValue=5, group=Hex Pulse, parameterName=Hex Pulse Emissive}
  FunctionOutput(None=Multiply(A=Multiply(A=@n12.RGB, B=@n13), B=Power(Base=@n8))) {outputName=Emissive, sortPriority=1}
  FunctionOutput(None=LinearInterpolate(
    A=Constant3Vector() {constant=<rgba 0 0 1 0>}
    B=TextureSample(UVs=@n3) {samplerType=SAMPLERTYPE_Normal, texture=/Game/Creative/Sets/PropSets/Primitives/Customization/Textures/Hex/T_CM_HexPulse_N}.RGB
    Alpha=@n9
   )) {outputName=Normal, sortPriority=2}
 ]}
