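# Material function that applies a world-aligned Hex Pulse look to Creative primitive props when the Primitive_HexPulse static switch is on, otherwise passing input attributes through.
# MF_HexPulseSample runs on three planes of tiled world position, each fed normal-relative camera vector data, likely for parallax.
# Contrast-adjusted vertex-normal masks blend base color, emissive and normals triplanar-style.
# Each plane's normal is rebuilt in world space from a derived basis, then transformed to tangent space.
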MaterialFunction /Game/Creative/Sets/PropSets/Primitives/Customization/MaterialFunctions/MF_CM_HexPulse2
graph {nodes [
  n1 = Multiply(A=Divide(A=WorldPosition().XYZ) {constB=1000}, B=ScalarParameter() {defaultValue=5, group=Hex Pulse, parameterName=Hex Pulse Tiling})
  n2 = MaterialFunctionCall!(Camera Info=NamedRerouteUsage(), Position=ComponentMask(None=@n1) {b=true, g=true}) {materialFunction=&/MF_HexPulseSample}
  n3 = MaterialFunctionCall(Camera Info=NamedRerouteUsage(), Position=ComponentMask(None=@n1) {b=true, r=true}) {materialFunction=&MF_HexPulseSample}
  n4 = MaterialFunctionCall(Camera Info=NamedRerouteUsage(), Position=ComponentMask(None=@n1) {g=true, r=true}) {materialFunction=&MF_HexPulseSample}
  n5 = LinearInterpolate(A=LinearInterpolate(A=@n2.Base Color, B=@n3.Base Color, Alpha=NamedRerouteUsage()), B=@n4.Base Color, Alpha=NamedRerouteUsage())
  n6 = LinearInterpolate(A=LinearInterpolate(A=@n2.Emissive, B=@n3.Emissive, Alpha=NamedRerouteUsage()), B=@n4.Emissive, Alpha=NamedRerouteUsage())
  n7 = MaterialFunctionCall(Float3=@n2.Normal) {materialFunction=&BreakOutFloat3Components}
  n8 = VertexNormalWS()
  n9 = MaterialFunctionCall(Vector=CrossProduct(A=@n8, B=Constant3Vector() {constant=<rgba 0 0 1 0>})) {materialFunction=&SafeNormalize}
  n10 = Sign(None=DotProduct(A=@n8, B=Constant3Vector() {constant=<rgba 1 0 0 0>}))
  n11 = Multiply(A=@n7.R, B=Multiply(A=@n9.Result, B=Multiply(A=@n10) {constB=-1}) {constB=-1})
  n12 = Multiply(A=@n7.G, B=CrossProduct(A=@n9.Result, B=@n8))
  n13 = Normalize(VectorInput=Add(A=Add(A=@n11, B=@n12), B=Multiply(A=@n7.B, B=@n8)))
  n14 = MaterialFunctionCall(Float3=@n3.Normal) {materialFunction=&BreakOutFloat3Components}
  n15 = VertexNormalWS()
  n16 = MaterialFunctionCall(Vector=CrossProduct(A=@n15, B=Constant3Vector() {constant=<rgba 1 0 0 0>})) {materialFunction=&SafeNormalize}
  n17 = Sign(None=DotProduct(A=@n15, B=Constant3Vector() {constant=<rgba 0 1 0 0>}))
  n18 = Multiply(A=@n14.G, B=Multiply(A=@n16.Result, B=Multiply(A=@n17) {constB=-1}) {constB=-1})
  n19 = Multiply(A=@n14.R, B=CrossProduct(A=@n16.Result, B=@n15))
  n20 = Normalize(VectorInput=Add(A=Add(A=@n18, B=@n19), B=Multiply(A=@n14.B, B=@n15)))
  n21 = MaterialFunctionCall(Float3=@n4.Normal) {materialFunction=&BreakOutFloat3Components}
  n22 = VertexNormalWS()
  n23 = MaterialFunctionCall(Vector=CrossProduct(A=@n22, B=Constant3Vector() {constant=<rgba 0 1 0 0>})) {materialFunction=&SafeNormalize}
  n24 = DotProduct(A=@n22, B=Constant3Vector() {constant=<rgba 0 0 1 0>})
  n25 = Multiply(A=@n23.Result, B=Multiply(A=Sign(None=@n24)) {constB=-1}) {constB=-1}
  n26 = CrossProduct(A=@n23.Result, B=@n22)
  n27 = Add(A=Add(A=Multiply(A=@n21.R, B=@n25), B=Multiply(A=@n21.G, B=@n26)), B=Multiply(A=@n21.B, B=@n22))
  n28 = LinearInterpolate(
   A=LinearInterpolate(
    A=Transform(None=@n13) {transformSourceType=TRANSFORMSOURCE_World, transformType=TRANSFORM_Tangent}
    B=Transform(None=@n20) {transformSourceType=TRANSFORMSOURCE_World, transformType=TRANSFORM_Tangent}
    Alpha=NamedRerouteUsage()
   )
   B=Transform(None=Normalize(VectorInput=@n27)) {transformSourceType=TRANSFORMSOURCE_World, transformType=TRANSFORM_Tangent}
   Alpha=NamedRerouteUsage()
  )
  FunctionOutput(None=StaticSwitchParameter(
    True=MakeMaterialAttributes!(
     BaseColor=@n5
     Metallic
     Specular
     Roughness
     Anisotropy
     EmissiveColor=@n6
     Opacity
     OpacityMask
     Normal=@n28
     Tangent
     WorldPositionOffset
     SubsurfaceColor
     ClearCoat
     ClearCoatRoughness
     AmbientOcclusion
     Refraction
     CustomizedUVs_0
     CustomizedUVs_1
     CustomizedUVs_2
     CustomizedUVs_3
     CustomizedUVs_4
     CustomizedUVs_5
     CustomizedUVs_6
     CustomizedUVs_7
     PixelDepthOffset
     ShadingModel
     Displacement
    )
    False=FunctionInput() {inputType=FunctionInput_MaterialAttributes}
   ) {group=Customization, parameterName=Primitive_HexPulse})
  n29 = CameraVectorWS()
  n30 = VertexNormalWS()
  n31 = DotProduct(A=@n29, B=@n30)
  n32 = Divide(A=Subtract(A=@n29, B=Multiply(A=@n31, B=@n30)), B=@n31)
  NamedRerouteDeclaration(None=ComponentMask(None=@n32) {g=true, r=true}) {name=Camera Info XY, nodeColor=<rgba 1 1 0 1>}
  NamedRerouteDeclaration(None=ComponentMask(None=@n32) {b=true, g=true}) {name=Camera Info YZ, nodeColor=<rgba 0 1 1 1>}
  NamedRerouteDeclaration(None=ComponentMask(None=@n32) {b=true, r=true}) {name=Camera Info XZ, nodeColor=<rgba 1 0 1 1>}
  n33 = DotProduct(A=Constant3Vector() {constant=<rgba 0 1 0 0>}, B=VertexNormalWS())
  n34 = ScalarParameter() {defaultValue=3, group=Hex Pulse, parameterName=Hex Pulse Triplanar Contrast}
  NamedRerouteDeclaration(None=MaterialFunctionCall(In=Abs(None=@n33), Contrast=@n34) {materialFunction=&CheapContrast}.Result) {name=Triplanar Y, nodeColor=<rgba 0 1 0 1>}
  n35 = DotProduct(A=Constant3Vector() {constant=<rgba 0 0 1 0>}, B=VertexNormalWS())
  NamedRerouteDeclaration(None=MaterialFunctionCall(In=Abs(None=@n35), Contrast=@n34) {materialFunction=&CheapContrast}.Result) {name=Triplanar Z, nodeColor=<rgba 0 0 1 1>}
 ]}
