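# Material function that shades layered ice for Creative primitive props, outputting color, roughness, normal and combined material attributes.
# A custom HLSL loop marches the crack texture along the view vector, scaled by a depth map that also blends shallow and deep colors, with iterations varying by quality.
# Direction-weighted world-space surface normals distort the crack UVs.
# Surface mask adds to color and roughness; camera-distance fades are computed per object and per vertex.
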MaterialFunction /Game/Creative/Sets/PropSets/Primitives/Customization/MaterialFunctions/MF_IceSample
graph {nodes [
  n1 = TextureObjectParameter() {
   group=Ice
   parameterName=Ice Tex 2
   samplerType=SAMPLERTYPE_LinearGrayscale
   texture=/Game/Creative/Sets/PropSets/Primitives/Customization/Textures/Ice/T_IceMask_02_M
  }
  n2 = NamedRerouteUsage()
  n3 = TextureSample(UVs=@n2, Tex=FunctionInput(Preview=@n1) {bUsePreviewValueAsDefault=true, inputName=Surface Base Color, inputType=FunctionInput_Texture2D}) {samplerType=SAMPLERTYPE_LinearGrayscale}
  n4 = Multiply(A=@n3.R, B=ScalarParameter() {defaultValue=1, group=Ice, parameterName=Ice Surface Intensity})
  n5 = NamedRerouteUsage()
  n6 = NamedRerouteUsage()
  n7 = ScalarParameter() {defaultValue=0.3, parameterName=Ice Distance factor A}
  n8 = OneMinus(None=@n6)
  n9 = ScalarParameter() {defaultValue=1, parameterName=Ice Distance factor B}
  n10 = ScalarParameter() {defaultValue=0.235, group=Ice, parameterName=Ice Depth Map Tiling}
  n11 = NamedRerouteUsage()
  n12 = ScalarParameter() {group=Ice, parameterName=Ice Depth Map Offset X}
  n13 = ScalarParameter() {group=Ice, parameterName=Ice Depth Map Offset Y}
  n14 = TextureSampleParameter2D(UVs=Add(A=Multiply(A=@n11, B=@n10), B=AppendVector(A=@n12, B=@n13))) {
   group=Ice
   parameterName=Ice Depth Map
   samplerType=SAMPLERTYPE_LinearColor
   texture=/Game/Effects/Fort_Effects/Textures/Plasma/T_Turbulence_AdditiveWarp_Linear
  }
  MaterialExpressionOneMinus_0 = OneMinus(None=Power(Base=@n14.R, Exp=ScalarParameter() {defaultValue=2, group=Ice, parameterName=Ice Depth Map Power}))
  n15 = LinearInterpolate(
   A=VectorParameter() {group=Ice, parameterName=Ice Color Shallow, defaultValue=<rgba 0.0976 0.4621 0.8963 1>}.RGB
   B=VectorParameter() {group=Ice, parameterName=Ice Color Deep, defaultValue=<rgba 0.0018 0.0496 0.1146 1>}.RGB
   Alpha=@MaterialExpressionOneMinus_0
  )
  n16 = Add(A=Multiply(A=Add(A=@n4, B=@n5), B=Max(A=@n6, B=@n7) {constB=0.3}), B=Multiply(A=Multiply(A=@n8, B=@n9), B=@n15))
  n17 = FunctionInput() {inputName=Direction}
  n18 = Constant3Vector() {constant=<rgba 0 0 1 0>}
  n19 = DotProduct(A=@n17, B=@n18)
  n20 = TextureObjectParameter() {
   group=Ice
   parameterName=Ice Surface Normal
   samplerType=SAMPLERTYPE_Normal
   texture=/Game/Creative/Sets/PropSets/Primitives/Customization/Textures/Ice/T_IceMask_01_N
  }
  n21 = TextureSample(UVs=@n2, Tex=FunctionInput(Preview=@n20) {bUsePreviewValueAsDefault=true, inputName=Surface Normal, inputType=FunctionInput_Texture2D}) {samplerType=SAMPLERTYPE_Normal}
  n22 = MaterialFunctionCall(Float3=@n21.RGB) {materialFunction=&BreakOutFloat3Components}
  n23 = VertexNormalWS()
  n24 = MaterialFunctionCall(Vector=CrossProduct(A=@n23, B=Constant3Vector() {constant=<rgba 0 1 0 0>})) {materialFunction=&SafeNormalize}
  n25 = Multiply(A=Sign(None=DotProduct(A=@n23, B=@n18))) {constB=-1}
  n26 = Add(A=Multiply(A=@n22.R, B=Multiply(A=@n24.Result, B=@n25) {constB=-1}), B=Multiply(A=@n22.G, B=CrossProduct(A=@n24.Result, B=@n23)))
  n27 = Multiply(A=@n22.B, B=@n23)
  n28 = Transform(None=Normalize(VectorInput=Add(A=@n26, B=@n27))) {transformSourceType=TRANSFORMSOURCE_World, transformType=TRANSFORM_Tangent}
  n29 = Constant3Vector() {constant=<rgba 1 0 0 0>}
  n30 = DotProduct(A=@n17, B=@n29)
  n31 = MaterialFunctionCall(Float3=@n21.RGB) {materialFunction=&BreakOutFloat3Components}
  n32 = MaterialFunctionCall(Vector=CrossProduct(A=@n23, B=Constant3Vector() {constant=<rgba 0 0 1 0>})) {materialFunction=&SafeNormalize}
  n33 = Multiply(A=Sign(None=DotProduct(A=@n23, B=@n29))) {constB=-1}
  n34 = Add(A=Multiply(A=@n31.R, B=Multiply(A=@n32.Result, B=@n33) {constB=-1}), B=Multiply(A=@n31.G, B=CrossProduct(A=@n32.Result, B=@n23)))
  n35 = Multiply(A=@n31.B, B=@n23)
  n36 = Transform(None=Normalize(VectorInput=Add(A=@n34, B=@n35))) {transformSourceType=TRANSFORMSOURCE_World, transformType=TRANSFORM_Tangent}
  n37 = Constant3Vector() {constant=<rgba 0 1 0 0>}
  n38 = MaterialFunctionCall(Float3=@n21.RGB) {materialFunction=&BreakOutFloat3Components}
  n39 = MaterialFunctionCall(Vector=CrossProduct(A=@n23, B=Constant3Vector() {constant=<rgba 1 0 0 0>})) {materialFunction=&SafeNormalize}
  n40 = Sign(None=DotProduct(A=@n23, B=@n37))
  n41 = Multiply(A=@n38.G, B=Multiply(A=@n39.Result, B=Multiply(A=@n40) {constB=-1}) {constB=-1})
  n42 = Multiply(A=@n38.R, B=CrossProduct(A=@n39.Result, B=@n23))
  n43 = Normalize(VectorInput=Add(A=Add(A=@n41, B=@n42), B=Multiply(A=@n38.B, B=@n23)))
  n44 = Add(
   A=Add(A=Multiply(A=@n19, B=@n28), B=Multiply(A=@n30, B=@n36))
   B=Multiply(A=DotProduct(A=@n17, B=@n37), B=Transform(None=@n43) {transformSourceType=TRANSFORMSOURCE_World, transformType=TRANSFORM_Tangent})
  )
  n45 = LinearInterpolate(
   A=Constant3Vector() {constant=<rgba 0 0 1 0>}
   B=@n44
   Alpha=ScalarParameter() {defaultValue=0.5, group=Ice, parameterName=Ice Normal Intensity}
  )
  n46 = Multiply(A=@n2, B=ScalarParameter() {defaultValue=1, group=Ice, parameterName=Ice Surface Tiling})
  n47 = TextureObjectParameter() {
   group=Ice
   parameterName=Ice Roughness Texture
   samplerType=SAMPLERTYPE_LinearColor
   texture=/Game/Creative/Sets/PropSets/Primitives/Customization/Textures/Ice/T_IceMasks_M
  }
  n48 = FunctionInput(Preview=@n47) {bUsePreviewValueAsDefault=true, inputName=Surface Roughness, inputType=FunctionInput_Texture2D}
  n49 = TextureSample(UVs=Multiply(A=@n46) {constB=0.25}, Tex=@n48) {samplerType=SAMPLERTYPE_LinearGrayscale}
  n50 = TextureSample(UVs=@n46, Tex=@n48) {samplerType=SAMPLERTYPE_LinearGrayscale}
  n51 = NamedRerouteUsage()
  n52 = Add(
   A=LinearInterpolate(
    A=ScalarParameter() {group=Ice, parameterName=Ice Roughness Low}
    B=ScalarParameter() {defaultValue=1, group=Ice, parameterName=Ice Roughness High}
    Alpha=LinearInterpolate(A=@n49.B, B=@n50.B, Alpha=@n51)
   )
   B=Multiply(A=@n3.R, B=ScalarParameter() {defaultValue=0.3, group=Ice, parameterName=Ice Roughness Add})
  )
  FunctionOutput(None=SetMaterialAttributes!(
    MaterialAttributes
    BaseColor=FunctionOutput(None=@n16) {outputName=Color}
    Normal=FunctionOutput(None=@n45) {outputName=Normal, sortPriority=2}
    Roughness=FunctionOutput(None=@n52) {outputName=Roughness, sortPriority=1}
   ) {attributeSetTypes [#1, #2, #3]}) {outputName=MA, sortPriority=-1}
  n53 = TextureObjectParameter() {
   group=Ice
   parameterName=Ice Cracks Depth Texture
   samplerType=SAMPLERTYPE_LinearColor
   texture=&T_IceMasks_M
  }
  MaterialExpressionFunctionInput_4 = FunctionInput(Preview=@n53) {bUsePreviewValueAsDefault=true, inputName=Ice Cracks, inputType=FunctionInput_Texture2D}
  MaterialExpressionMultiply_4 = Multiply(A=Divide(A=ScalarParameter() {defaultValue=-10, group=Ice, parameterName=Ice Sample Depth}) {constB=100}, B=NamedRerouteUsage())
  n54 = Multiply(A=ScalarParameter() {defaultValue=30, group=Ice, parameterName=Ice Sample Iterations}, B=NamedRerouteUsage())
  n55 = Max(A=@n54)
  MaterialExpressionQualitySwitch_0 = QualitySwitch(Default=@n55, Low=Constant() {r=1}, High=@n55, Medium=Max(A=Multiply(A=@n54) {constB=0.5}), Epic=@n55)
  n56 = ScalarParameter() {defaultValue=2, group=Ice, parameterName=Ice Sample Distortion}
  MaterialExpressionAdd_3 = Add(A=Multiply(A=ComponentMask(None=@n44) {g=true, r=true}, B=Divide(A=@n56) {constB=100}), B=@n11)
  n57 = CameraVectorWS()
  MaterialExpressionFunctionInput_2 = FunctionInput(Preview=ComponentMask(None=Transform(None=@n57) {transformSourceType=TRANSFORMSOURCE_World, transformType=TRANSFORM_Tangent}) {g=true, r=true}) {bUsePreviewValueAsDefault=true, inputName=Camera Info, inputType=FunctionInput_Vector2}
  MaterialExpressionScalarParameter_5 = ScalarParameter() {defaultValue=2, group=Ice, parameterName=Ice Sample Falloff}
  n58 = Custom!(
   Tex=@MaterialExpressionFunctionInput_4
   Height=@MaterialExpressionMultiply_4
   Iterations=@MaterialExpressionQualitySwitch_0
   UV=@MaterialExpressionAdd_3
   CameraInfo=@MaterialExpressionFunctionInput_2
   Falloff=@MaterialExpressionScalarParameter_5
   DepthTex=@MaterialExpressionOneMinus_0
  ) {
   code="// Tex <- Crack Texture\r\n// DepthTex <- Depth multiplier\r\n// DistortTex <- Distortion at depth\r\n\r\nfloat3 color = float3(0,0,0);\r\n//Texture2DSample(Tex, TexSampler, UV);\r\n\r\nfloat modulatedHeight = Height * DepthTex;\r\n\r\nfor (int i = 0; i < Iterations; i++)\r\n{\r\n  float2 uvs = UV + CameraInfo * i * modulatedHeight/Iterations;\r\n  \r\n  float intensity = pow( (Iterations-i)/Iterations , Falloff ) ;\r\n  \r\n  color = max(color, Texture2DSample(Tex, TexSampler, uvs).xyz * intensity );\r\n\r\n}\r\n\r\nreturn color;"
   inputs [
    {
     inputName=Tex
     input {expression=&self:MaterialExpressionFunctionInput_4}
    }
    {
     inputName=Height
     input {expression=&self:MaterialExpressionMultiply_4}
    }
    {
     inputName=Iterations
     input {expression=&self:MaterialExpressionQualitySwitch_0}
    }
    {
     inputName=UV
     input {expression=&self:MaterialExpressionAdd_3}
    }
    {
     inputName=CameraInfo
     input {expression=&self:MaterialExpressionFunctionInput_2}
    }
    {
     inputName=Falloff
     input {expression=&self:MaterialExpressionScalarParameter_5}
    }
    {
     inputName=DepthTex
     input {expression=&self:MaterialExpressionOneMinus_0}
    }
   ]
  }
  n59 = Multiply(
   A=ComponentMask(None=@n58) {g=true}
   B=VectorParameter() {group=Ice, parameterName=Ice Crack Color, defaultValue=<rgba 0.1606 0.4368 0.6562 0>}.RGB
  )
  n60 = Constant()
  NamedRerouteDeclaration(None=Add(A=QualitySwitch(Default=@n59, Low=@n60, High=@n59, Medium=@n59, Epic=@n59), B=@n15)) {name=Depth Output, nodeColor=<rgba 0 0.445 1 1>}
  n61 = WorldPosition()
  n62 = FunctionInput(Preview=ComponentMask(None=Divide(A=@n61.XYZ) {constB=1000}) {g=true, r=true}) {bUsePreviewValueAsDefault=true, inputName=UV, inputType=FunctionInput_Vector2}
  NamedRerouteDeclaration(None=@n62) {name=UVs, nodeColor=<rgba 1 0 0.9844 1>}
  n63 = Subtract(A=ObjectPositionWS(), B=CameraPositionWS())
  n64 = ScalarParameter() {defaultValue=3000, parameterName="Ice Fade Length (S)"}
  n65 = ScalarParameter() {defaultValue=3000, parameterName="Ice Fade Offset (S)"}
  n66 = Divide(
   A=Subtract(A=MaterialFunctionCall(Vector 3=@n63) {materialFunction=/Engine/Functions/Engine_MaterialFunctions02/Utility/VectorLength}.V3 Length, B=@n64)
   B=@n65
  )
  NamedRerouteDeclaration(None=OneMinus(None=Saturate(None=@n66))) {name=Per Object Camera Depth Fade, nodeColor=<rgba 0 1 0.5547 1>}
  NamedRerouteDeclaration(None=OneMinus(None=MaterialFunctionCall(Fade Length=@n64, Fade Offset=@n65) {materialFunction=&CameraDepthFade}.Result)) {name=Per Vertex Camera Depth Fade, nodeColor=<rgba 0 0.8871 1 1>}
  Reroute(None=@n62)
 ]}
