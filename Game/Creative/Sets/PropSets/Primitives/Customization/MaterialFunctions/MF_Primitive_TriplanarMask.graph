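# Material function for Creative primitive props that blends three material attribute sets by surface orientation, as a triplanar mask.
# Absolute dot products of the world-space vertex normal with the Y and Z axes pass through CheapContrast with a shared Contrast input.
# YZ is the base, the Y mask blends in XZ, then the Z mask layers XY over that result.
# Input previews default to red, green and blue base colors per plane.
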MaterialFunction /Game/Creative/Sets/PropSets/Primitives/Customization/MaterialFunctions/MF_Primitive_TriplanarMask
graph {nodes [
  n1 = FunctionInput(Preview=SetMaterialAttributes!(MaterialAttributes, BaseColor=Constant3Vector() {constant=<rgba 1 0 0 0>}) {attributeSetTypes [#1]}) {bUsePreviewValueAsDefault=true, inputName=Material Attribute YZ, inputType=FunctionInput_MaterialAttributes, sortPriority=11}
  n2 = FunctionInput(Preview=SetMaterialAttributes(BaseColor=Constant3Vector() {constant=<rgba 0 1 0 0>}) {attributeSetTypes [#1]}) {bUsePreviewValueAsDefault=true, inputName=Material Attribute XZ, inputType=FunctionInput_MaterialAttributes, sortPriority=12}
  n3 = Constant3Vector() {constant=<rgba 0 1 0 0>}
  n4 = VertexNormalWS()
  n5 = FunctionInput(Preview=Constant() {r=3}) {bUsePreviewValueAsDefault=true, inputName=Contrast, inputType=FunctionInput_Scalar}
  n6 = MaterialFunctionCall(In=Abs(None=DotProduct(A=@n3, B=@n4)), Contrast=@n5) {materialFunction=&CheapContrast}
  n7 = FunctionInput(Preview=SetMaterialAttributes(BaseColor=Constant3Vector() {constant=<rgba 0 0 1 0>}) {attributeSetTypes [#1]}) {bUsePreviewValueAsDefault=true, inputName=Material Attribute XY, inputType=FunctionInput_MaterialAttributes, sortPriority=10}
  n8 = Abs(None=DotProduct(A=Constant3Vector() {constant=<rgba 0 0 1 0>}, B=@n4))
  FunctionOutput(None=BlendMaterialAttributes(
    A=BlendMaterialAttributes(A=@n1, B=@n2, Alpha=@n6.Result)
    B=@n7
    Alpha=MaterialFunctionCall(In=@n8, Contrast=@n5) {materialFunction=&CheapContrast}.Result
   )) {outputName=Blender MA}
 ]}
