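# Material function that adds an optional ice look to customizable Creative primitive props, toggled by the Primitive_Ice static switch.
# It samples MF_IceSample on three world-space planes with tiled position UVs, blended by MF_Primitive_TriplanarMask.
# Each sample receives a view vector projected onto the surface and divided by N·V, a parallax-style offset.
# The Ice Subsurface Color drives both subsurface color and opacity; switched off, input attributes pass through.
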MaterialFunction /Game/Creative/Sets/PropSets/Primitives/Customization/MaterialFunctions/MF_CM_Ice
graph {nodes [
  n1 = CameraVectorWS()
  n2 = VertexNormalWS()
  n3 = DotProduct(A=@n1, B=@n2)
  n4 = Divide(A=Subtract(A=@n1, B=Multiply(A=@n3, B=@n2)), B=@n3)
  n5 = Multiply(A=Divide(A=WorldPosition().XYZ) {constB=1000}, B=ScalarParameter() {defaultValue=3, group=Ice, parameterName=Ice Tiling})
  n6 = MaterialFunctionCall!(
   Contrast=ScalarParameter() {defaultValue=3, group=Ice, parameterName=Ice Triplanar Contrast}
   Material Attribute XY=MaterialFunctionCall!(
    Surface Normal
    Camera Info=ComponentMask(None=@n4) {g=true, r=true}
    Surface Base Color
    Ice Cracks
    Surface Roughness
    Direction=Constant3Vector() {constant=<rgba 0 0 1 0>}
    UV=ComponentMask(None=@n5) {g=true, r=true}
   ) {materialFunction=&/MF_IceSample}.MA
   Material Attribute YZ=MaterialFunctionCall(
    Camera Info=ComponentMask(None=@n4) {b=true, g=true}
    Direction=Constant3Vector() {constant=<rgba 1 0 0 0>}
    UV=ComponentMask(None=@n5) {b=true, g=true}
   ) {materialFunction=&MF_IceSample}.MA
   Material Attribute XZ=MaterialFunctionCall(
    Camera Info=ComponentMask(None=@n4) {b=true, r=true}
    Direction=Constant3Vector() {constant=<rgba 0 1 0 0>}
    UV=ComponentMask(None=@n5) {b=true, r=true}
   ) {materialFunction=&MF_IceSample}.MA
  ) {materialFunction=&/MF_Primitive_TriplanarMask}
  n7 = VectorParameter() {group=Ice, parameterName=Ice Subsurface Color, defaultValue=<rgba 0.1285 0.391 0.7812 0.516>}
  FunctionOutput(None=StaticSwitchParameter(
    True=SetMaterialAttributes!(MaterialAttributes=@n6.Blender MA, SubsurfaceColor=@n7.RGB, Opacity=@n7.RGB) {attributeSetTypes [#1, #2]}
    False=FunctionInput() {inputType=FunctionInput_MaterialAttributes}
   ) {group=Customization, parameterName=Primitive_Ice})
 ]}
