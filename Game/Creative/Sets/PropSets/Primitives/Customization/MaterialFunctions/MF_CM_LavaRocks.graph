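# Material function that adds an animated lava rock look to Creative primitive props, toggled by the Primitive_LavaRocks static switch with passthrough when off.
# Triplanar-projected rock textures supply color, normal, roughness, height and masks, overriding the input material attributes.
# Time-panned warp and wave textures plus gas-cloud noise build lava masks, with a curve atlas coloring the emissive.
# It also reads music amplitude from an audio-reactive parameter collection and derives an object-height Z gradient.
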MaterialFunction /Game/Creative/Sets/PropSets/Primitives/Customization/MaterialFunctions/MF_CM_LavaRocks
graph {nodes [
  Constant3Vector() {constant=<rgba 0.5 0.5 0.5 0>}
  n1 = NamedRerouteUsage()
  n2 = NamedRerouteUsage()
  n3 = NamedRerouteUsage()
  n4 = Subtract(A=Add(A=ComponentMask(None=@n3) {r=true}, B=NamedRerouteUsage()), B=NamedRerouteUsage())
  n5 = CurveAtlasRowParameter(CurveTime=Saturate(None=OneMinus(None=@n4))) {
   atlas=/Game/Creative/Sets/PropSets/Primitives/Customization/Textures/ColorCurve/CC_CM_Primitives_01
   curve=/Game/Creative/Sets/PropSets/Primitives/Customization/Textures/ColorCurve/C_CM_LavaRocks_01
   defaultValue=8
   group=LavaRocks
   parameterName=CM_LavaRocks_ColorCurve
  }
  n6 = FunctionInput(Preview=SetMaterialAttributes!(
    MaterialAttributes
    BaseColor=Constant3Vector() {constant=<rgba 1 0 0 0>}
    Normal=TextureSample() {samplerType=SAMPLERTYPE_Normal, texture=/Game/Weapons/FORT_Rifles/T_SCAR_N}.RGB
   ) {attributeSetTypes [#1, #2]}) {inputType=FunctionInput_MaterialAttributes}
  n7 = SetMaterialAttributes!(
   MaterialAttributes=@n6
   BaseColor=LinearInterpolate(B=NamedRerouteUsage(), Alpha=@n1)
   EmissiveColor=LinearInterpolate(A=Saturate(None=@n5), Alpha=Multiply(A=@n2)) {constB=0}
   Specular=NamedRerouteUsage()
   Roughness=NamedRerouteUsage()
   Normal=LinearInterpolate(A=Constant3Vector() {constant=<rgba 0 0 1 0>}, B=NamedRerouteUsage(), Alpha=@n1)
  ) {attributeSetTypes [
    #1
    #3
    #4
    #5
    #2
   ]}
  FunctionOutput(None=StaticSwitchParameter(True=@n7, False=@n6) {group=Customization, parameterName=Primitive_LavaRocks})
  n8 = ScalarParameter() {defaultValue=0.2, group=LavaRocks, parameterName=CM_LavaRocks_FlatSpec}
  n9 = Multiply(
   A=OneMinus(None=NamedRerouteUsage())
   B=ScalarParameter() {defaultValue=1, group=LavaRocks, parameterName=CM_LavaRocks_TextSpecInst, sliderMax=1}
  )
  n10 = ScalarParameter() {defaultValue=1, group=LavaRocks, parameterName=CM_LavaRocksUse_TextureforSpec}
  NamedRerouteDeclaration(None=Min(A=LinearInterpolate(A=@n8, B=@n9, Alpha=@n10)) {constB=0.3}) {name=Spec, nodeColor=<rgba 0 0.4072 0.3968 1>}
  n11 = MaterialFunctionCall() {materialFunction=&/MF_Primitive_CalculatedBounds}
  NamedRerouteDeclaration(None=@n11.World Normals) {name=WorldNormals, nodeColor=<rgba 1 0.6797 0 1>}
  n12 = NamedRerouteUsage()
  n13 = ScalarParameter() {defaultValue=2, group=LavaRocks, parameterName=CM_LavaRocks_ColorMask Contrast}
  n14 = NamedRerouteUsage()
  n15 = NamedRerouteUsage()
  NamedRerouteDeclaration(None=Multiply(A=Power(Base=@n12, Exp=@n13), B=LinearInterpolate(B=@n14, Alpha=@n15) {constA=1})) {name=ColorGradient, nodeColor=<rgba 0.9922 1 0 1>}
  n16 = OneMinus(None=NamedRerouteUsage())
  n17 = TextureObject() {samplerType=SAMPLERTYPE_Grayscale, texture=/Game/Creative/Sets/PropSets/Primitives/Customization/Textures/GasCloud/T_CM_GasCloud_BC_01}
  n18 = NamedRerouteUsage()
  n19 = NamedRerouteUsage()
  n20 = VectorParameter() {group=LavaRocks, parameterName=CM_LavaRocks_Rotation, defaultValue=<rgba 0 0 0 1>}
  n21 = ComponentMask(None=MaterialFunctionCall!(
    TextureObject_Diffuse=@n17
    TextureScale=Constant() {r=2}
    CalculatedBounds=Add(A=@n18, B=@n19)
    WorldNormals=NamedRerouteUsage()
    RotationX=@n20.R
    RotationY=@n20.R
    RotationZ=@n20.R
    ProjectionTransitionContrast=ScalarParameter() {defaultValue=1, group=LavaRocks, parameterName=CM_LavaRocks_Triplannar_Contrast}
    Material Normals
    NormalTexture
    Normal Texture Intensity
   ) {materialFunction=&/MF_Primitive_TriplanarUVs}.RGBA) {r=true}
  NamedRerouteDeclaration(None=Clamp(None=Multiply(A=@n16, B=@n21)) {maxDefault=0.041}) {name=LavaRocks_LavaNoise, nodeColor=<rgba 0.9768 1 0 1>}
  n22 = TextureObject() {
   samplerType=SAMPLERTYPE_LinearGrayscale
   texture=/Game/Creative/Sets/PropSets/Primitives/Customization/Textures/GasCloud/T_CM_GasCloud_WarpTexture_01
  }
  n23 = Multiply(A=Time(), B=Constant3Vector() {constant=<rgba -0.1 0.04 0.02 0>})
  n24 = NamedRerouteUsage()
  n25 = VectorParameter() {group=LavaRocks, parameterName=CM_LavaRocks_Rotation, defaultValue=<rgba 0 0 0 1>}
  n26 = ComponentMask(None=MaterialFunctionCall(
    TextureObject_Diffuse=@n22
    TextureScale=Constant() {r=5}
    CalculatedBounds=Add(A=@n23, B=@n24)
    WorldNormals=NamedRerouteUsage()
    RotationX=@n25.R
    RotationY=@n25.R
    RotationZ=@n25.R
    ProjectionTransitionContrast=ScalarParameter() {defaultValue=1, group=LavaRocks, parameterName=CM_LavaRocks_Triplannar_Contrast}
   ) {materialFunction=&MF_Primitive_TriplanarUVs}.RGBA) {b=true, g=true, r=true}
  n27 = ScalarParameter() {defaultValue=1, group=LavaRocks, parameterName=CM_LavaNoise_WarpInt}
  NamedRerouteDeclaration(None=Multiply(A=LinearInterpolate(B=@n26, Alpha=@n27)) {constB=1.4}) {name=LavaNoiseWarp, nodeColor=<rgba 1 0.0234 0 1>}
  n28 = NamedRerouteUsage()
  n29 = NamedRerouteUsage()
  n30 = TextureObject() {
   samplerType=SAMPLERTYPE_LinearGrayscale
   texture=/Game/Creative/Sets/PropSets/Primitives/Customization/Textures/LavaRocks/T_CM_LavaRock_HeightPanner
  }
  n31 = Constant() {r=0.75}
  n32 = Add(A=NamedRerouteUsage(), B=Multiply(A=Time(), B=Constant3Vector() {constant=<rgba 0.1 -0.05 0.2 0>}))
  n33 = Multiply(
   A=ComponentMask(None=MaterialFunctionCall(TextureObject_Diffuse=@n30, TextureScale=@n31, CalculatedBounds=@n32) {materialFunction=&MF_Primitive_TriplanarUVs}.RGBA) {r=true}
   B=ScalarParameter() {defaultValue=0.2, group=LavaRocks, parameterName=CM_LavaRocks_LavaWave_Height}
  )
  n34 = ComponentMask(None=Multiply(A=Multiply(A=@n28, B=@n29), B=@n33)) {r=true}
  n35 = NamedRerouteUsage()
  n36 = ScalarParameter() {defaultValue=0.5, group=LavaRocks, parameterName=CM_LavaRocks_RockHeigh}
  NamedRerouteDeclaration(None=Add(A=Multiply(A=@n34) {constB=100}, B=Multiply(A=@n35, B=@n36))) {name=Lava_Wave_Mask, nodeColor=<rgba 1 0.9375 0 1>}
  NamedRerouteDeclaration(None=@n33) {name=LavaWave, nodeColor=<rgba 0 0.6719 1 1>}
  n37 = VectorParameter() {group=LavaRocks, parameterName=CM_LavaRocks_Rotation, defaultValue=<rgba 0 0 0 1>}
  n38 = Multiply(
   A=NamedRerouteUsage()
   B=MaterialFunctionCall(
    TextureObject_Diffuse=TextureObject() {samplerType=SAMPLERTYPE_Masks, texture=/Game/Creative/Sets/PropSets/Primitives/Customization/Textures/LavaRocks/T_CM_LavaRock_Mask}
    TextureScale=ScalarParameter() {defaultValue=1.2, group=LavaRocks, parameterName=CM_LavaRocks_UVTile}
    CalculatedBounds=NamedRerouteUsage()
    WorldNormals=NamedRerouteUsage()
    RotationX=@n37.R
    RotationY=@n37.R
    RotationZ=@n37.R
    ProjectionTransitionContrast=ScalarParameter() {defaultValue=1, group=LavaRocks, parameterName=CM_LavaRocks_Triplannar_Contrast}
   ) {materialFunction=&MF_Primitive_TriplanarUVs}.RGBA
  )
  n39 = MaterialFunctionCall(Color=ComponentMask(None=@n38) {b=true, g=true, r=true}) {materialFunction=&SplitComponents}
  NamedRerouteDeclaration(None=@n39.B) {name=LavaRocks_Height, nodeColor=<rgba 0.8984 0 1 1>}
  NamedRerouteDeclaration(None=ComponentMask(None=@n38) {a=true}) {name=LavaRock_RandomColor, nodeColor=<rgba 1 0.0216 0.6085 1>}
  n40 = ScalarParameter() {defaultValue=0.8, group=LavaRocks, parameterName=CM_LavaRocks_FlatRough}
  n41 = Multiply(A=NamedRerouteUsage(), B=ScalarParameter() {defaultValue=1, group=LavaRocks, parameterName=CM_LavaRocks_TextRoughInst, sliderMax=1})
  n42 = ScalarParameter() {defaultValue=1, group=LavaRocks, parameterName=CM_LavaRocksUse_TextureforRough}
  NamedRerouteDeclaration(None=Max(A=LinearInterpolate(A=@n40, B=@n41, Alpha=@n42)) {constB=0.45}) {name=Rough, nodeColor=<rgba 0.4072 0.2831 0 1>}
  n43 = CollectionParameter() {collection=/Game/Athena/Events/AudioAnalysis/MPC_Audio_Reactive_Parameters, parameterName=Amplitude_Music}
  NamedRerouteDeclaration(None=Saturate(None=@n43)) {name=MusicAmp, nodeColor=<rgba 0.2891 0 1 1>}
  NamedRerouteDeclaration(None=Ceil(None=Saturate(None=@n43))) {name=IsMusicOn, nodeColor=<rgba 0.5537 0 1 1>}
  NamedRerouteDeclaration(None=Divide(A=@n11.WorldPosition, B=ScalarParameter() {defaultValue=80, group=LavaRocks, parameterName=CM_LavaRocks_Scale})) {name=Scaled WorldScape, nodeColor=<rgba 1 0.6797 0 1>}
  n44 = VectorParameter() {group=LavaRocks, parameterName=CM_LavaRocks_Rotation, defaultValue=<rgba 0 0 0 1>}
  n45 = MaterialFunctionCall(
   TextureObject_Diffuse=TextureObject() {samplerType=SAMPLERTYPE_Normal, texture=/Game/Creative/Sets/PropSets/Primitives/Customization/Textures/LavaRocks/T_CM_LavaRock_N}
   TextureScale=ScalarParameter() {defaultValue=1.2, group=LavaRocks, parameterName=CM_LavaRocks_UVTile}
   CalculatedBounds=NamedRerouteUsage()
   WorldNormals=NamedRerouteUsage()
   RotationX=@n44.R
   RotationY=@n44.R
   RotationZ=@n44.R
   ProjectionTransitionContrast=ScalarParameter() {defaultValue=1, group=LavaRocks, parameterName=CM_LavaRocks_Triplannar_Contrast}
  ) {materialFunction=&MF_Primitive_TriplanarUVs}
  NamedRerouteDeclaration(None=Multiply(A=NamedRerouteUsage(), B=ComponentMask(None=@n45.RGBA) {b=true, g=true, r=true})) {name=LavaRocks_Normal, nodeColor=<rgba 0.9768 1 0 1>}
  n46 = ComponentMask(None=NamedRerouteUsage()) {a=true}
  n47 = Frac(None=NamedRerouteUsage())
  n48 = LinearInterpolate(B=Max(A=Multiply(A=@n47) {constB=10}) {constB=3.5}, Alpha=NamedRerouteUsage()) {constA=5}
  NamedRerouteDeclaration(None=Saturate(None=Multiply(A=@n46, B=@n48) {constB=4})) {name=Lava Rock Mask, nodeColor=<rgba 0.2254 0.357 1 1>}
  NamedRerouteDeclaration(None=Clamp(None=Multiply(A=@n39.R) {constB=3}) {maxDefault=0.4}) {name=LavaRocks_EdgeMask, nodeColor=<rgba 0.3564 0.0232 0.3613 1>}
  NamedRerouteDeclaration(None=@n11.Calculated Bounds) {name=CalculatedBounds, nodeColor=<rgba 1 0.6797 0 1>}
  NamedRerouteDeclaration(None=@n39.G) {name=LavaRocks_Roughness, nodeColor=<rgba 0.9922 0 1 1>}
  n49 = TextureObject() {texture=/Game/Creative/Sets/PropSets/Primitives/Customization/Textures/LavaRocks/T_CM_LavaRock_BC}
  n50 = ScalarParameter() {defaultValue=1.2, group=LavaRocks, parameterName=CM_LavaRocks_UVTile}
  n51 = NamedRerouteUsage()
  n52 = NamedRerouteUsage()
  n53 = VectorParameter() {group=LavaRocks, parameterName=CM_LavaRocks_Rotation, defaultValue=<rgba 0 0 0 1>}
  n54 = ScalarParameter() {defaultValue=1, group=LavaRocks, parameterName=CM_LavaRocks_Triplannar_Contrast}
  NamedRerouteDeclaration(None=Multiply(
    A=NamedRerouteUsage()
    B=MaterialFunctionCall(
     TextureObject_Diffuse=@n49
     TextureScale=@n50
     CalculatedBounds=@n51
     WorldNormals=@n52
     RotationX=@n53.R
     RotationY=@n53.R
     RotationZ=@n53.R
     ProjectionTransitionContrast=@n54
    ) {materialFunction=&MF_Primitive_TriplanarUVs}.RGBA
   )) {name=LavaRocks, nodeColor=<rgba 0.9768 1 0 1>}
  NamedRerouteUsage()
  NamedRerouteUsage()
  n55 = ObjectPositionWS()
  n56 = ObjectRadius()
  n57 = Subtract(A=ComponentMask(None=WorldPosition().XYZ) {b=true}, B=Subtract(A=ComponentMask(None=@n55) {b=true}, B=Max(A=@n56) {constB=50}))
  n58 = Multiply(A=@n56) {constB=1.8}
  n59 = NamedRerouteDeclaration(None=Saturate(None=Divide(A=@n57, B=@n58) {constB=100})) {name=ZGradient, nodeColor=<rgba 0.125 0 1 1>}
  n60 = ScalarParameter() {defaultValue=6, group=LavaRocks, parameterName=CM_LavaRocks_ZmaskContrast}
  OneMinus(None=NamedRerouteDeclaration(None=Power(Base=@n59, Exp=@n60)) {name=ZMask, nodeColor=<rgba 0 1 0.0156 1>})
  ScalarParameter() {defaultValue=1.2, group=LavaRocks, parameterName=CM_LavaRocks_UVTile}
 ]}
