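# Material function adding a switchable marble finish, with optional gold veining, to customizable Creative primitive props.
# Triplanar projection over scaled calculated-bounds positions samples color, normal, specular/roughness and noise textures.
# Vein and flake masks at separate tilings combine into a gold mask, gated by the UseGold static switch, which also routes output through MF_MetalsLUT.
# Tints, small-vein color and global blend shape base color; an unconnected view-angle Power node remains.
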MaterialFunction /Game/Creative/Sets/PropSets/Primitives/Customization/MaterialFunctions/MF_CM_Marble
graph {nodes [
  n1 = NamedRerouteUsage()
  n2 = GetMaterialAttributes(None=FunctionInput() {inputType=FunctionInput_MaterialAttributes}) {attributeGetTypes [#1]}
  n3 = SetMaterialAttributes!(
   MaterialAttributes=@n2.MaterialAttributes
   BaseColor=NamedRerouteUsage()
   Specular=NamedRerouteUsage()
   Roughness=NamedRerouteUsage()
   Normal=NamedRerouteUsage()
   Metallic=LinearInterpolate(
    A=ScalarParameter() {group=Marble, parameterName=CM_Metallic}
    B=ScalarParameter() {defaultValue=1, group=Marble, parameterName=CM_MetallicGold}
    Alpha=@n1
   )
   EmissiveColor=ScalarParameter() {group=Marble, parameterName=CM_Emissive}
  ) {attributeSetTypes [
    #2
    #3
    #4
    #1
    #5
    #6
   ]}
  n4 = MaterialFunctionCall(Mask=@n1, Materials=@n3) {materialFunction=&MF_MetalsLUT}
  FunctionOutput(None=StaticSwitchParameter(
    True=StaticSwitchParameter(True=@n4.Result, False=@n3) {defaultValue=true, group=Marble, parameterName=UseGold, sortPriority=-10}
    False=NamedRerouteUsage()
   ) {group=Customization, parameterName=Primitive_Marble})
  n5 = TextureObjectParameter() {
   group=Marble
   parameterName=CM_Marble_S
   samplerType=SAMPLERTYPE_Masks
   sortPriority=0
   texture=/Game/Creative/Sets/PropSets/Primitives/Customization/Textures/Marble/T_CP_Marble_S
  }
  n6 = ScalarParameter() {defaultValue=3, group=Marble, parameterName=CM_Marble_UVTile, sortPriority=100}
  n7 = NamedRerouteUsage()
  n8 = NamedRerouteUsage()
  n9 = ScalarParameter() {defaultValue=4, group=Marble, parameterName=CM_Marble_TriplannarContrast, sortPriority=101}
  n10 = MaterialFunctionCall!(TextureObject=@n5, TextureSize=@n6, WorldPosition=@n7, World Space Normal=@n8, ProjectionTransitionContrast=@n9) {materialFunction=&/MF_Primitive_TriplanarUVSimple}
  NamedRerouteDeclaration(None=ComponentMask(None=@n10.XYZ Texture) {b=true}) {name=Roughness, nodeColor=<rgba 0.8984 0 1 1>}
  NamedRerouteDeclaration(None=@n2.MaterialAttributes) {name=Material Attribute, nodeColor=<rgba 0.6641 0 1 1>}
  NamedRerouteDeclaration(None=ComponentMask(None=@n10.XYZ Texture) {r=true}) {name=Specular, nodeColor=<rgba 0 0.6016 1 1>}
  NamedRerouteDeclaration(None=NamedRerouteUsage()) {name=UVs, nodeColor=<rgba 0 1 0.4375 1>}
  n11 = TextureObjectParameter() {
   group=Marble
   parameterName=CM_Marble_BaseColor
   sortPriority=0
   texture=/Game/Creative/Sets/PropSets/Primitives/Customization/Textures/Marble/T_CP_Marble_BC2
  }
  n12 = Constant()
  n13 = TextureObjectParameter() {
   group=Marble
   parameterName=CM_Marble_NormalTex
   samplerType=SAMPLERTYPE_Normal
   sortPriority=0
   texture=/Game/Creative/Sets/PropSets/Primitives/Customization/Textures/Marble/T_CP_Marble_N
  }
  n14 = MaterialFunctionCall!(
   TextureObject_Diffuse=@n11
   TextureScale=@n6
   CalculatedBounds=@n7
   WorldNormals=@n8
   RotationX=@n12
   RotationY=@n12
   RotationZ=@n12
   ProjectionTransitionContrast=@n9
   Material Normals=NamedRerouteUsage()
   NormalTexture=@n13
   Normal Texture Intensity=ScalarParameter() {defaultValue=1, group=Marble, parameterName=CM_Marble_NormalIntensity, sortPriority=99}
  ) {materialFunction=&/MF_Primitive_TriplanarUVs}
  n15 = LinearInterpolate(
   A=Multiply(
    A=VectorParameter() {group=Marble, parameterName=CM_Marble_BaseColorTint, defaultValue=<rgba 1 1 1 1>}.RGB
    B=ComponentMask(None=@n14.RGBA) {b=true, g=true, r=true}
   )
   B=VectorParameter() {group=Marble, parameterName=CM_Marble_SmallVeinBaseColorColor, defaultValue=<rgba 1 1 1 1>}.RGB
   Alpha=Multiply(A=NamedRerouteUsage(), B=ScalarParameter() {group=Marble, parameterName=CM_Marble_SmallVeinColor})
  )
  n16 = VectorParameter() {group=Marble, parameterName=CM_Marble_GoldColor, defaultValue=<rgba 1 0.5683 0.2174 1>}
  n17 = NamedRerouteUsage()
  NamedRerouteDeclaration(None=LinearInterpolate(
    A=LinearInterpolate(A=@n15, B=@n16.RGB, Alpha=@n17)
    B=VectorParameter() {group=Marble, parameterName=CM_Marble_GlobalColorBlend, defaultValue=<rgba 1 1 1 1>}.RGB
    Alpha=ScalarParameter() {group=Marble, parameterName=CM_Marble_GlobalColorBlend}
   )) {name=BaseColor, nodeColor=<rgba 0 0.8594 1 1>}
  NamedRerouteDeclaration(None=@n14.Normal Texture) {name=Normal, nodeColor=<rgba 0 0.8594 1 1>}
  NamedRerouteDeclaration(None=@n2.Normal) {name=InNormal, nodeColor=<rgba 0 0.2266 1 1>}
  n18 = ScalarParameter() {group=Marble, parameterName=CM_Marble_SpecularMin}
  n19 = ScalarParameter() {defaultValue=1, group=Marble, parameterName=CM_Marble_SpecularMax}
  n20 = NamedRerouteUsage()
  NamedRerouteDeclaration(None=LinearInterpolate(
    A=LinearInterpolate(A=@n18, B=@n19, Alpha=@n20)
    B=ScalarParameter() {defaultValue=0.5, group=Marble, parameterName=CM_Marble_NoiseSpecular}
    Alpha=NamedRerouteUsage()
   )) {name=FinalSpec, nodeColor=<rgba 0 0.4072 0.3968 1>}
  n21 = ScalarParameter() {group=Marble, parameterName=CM_Marble_RoughnessMin}
  n22 = ScalarParameter() {defaultValue=1, group=Marble, parameterName=CM_Marble_RoughnessMax}
  n23 = NamedRerouteUsage()
  NamedRerouteDeclaration(None=LinearInterpolate(
    A=LinearInterpolate(A=@n21, B=@n22, Alpha=@n23)
    B=ScalarParameter() {defaultValue=0.28, group=Marble, parameterName=CM_Marble_NoiseRoughness}
    Alpha=NamedRerouteUsage()
   )) {name=FinalRough, nodeColor=<rgba 0.4072 0.2831 0 1>}
  n24 = TextureObjectParameter() {
   group=Marble
   parameterName=CM_Marble_NoiseMatTex
   samplerType=SAMPLERTYPE_Masks
   sortPriority=0
   texture=/Game/Creative/Sets/PropSets/Primitives/Customization/Textures/Marble/T_CP_Marble_M
  }
  n25 = ScalarParameter() {defaultValue=3, group=Marble, parameterName=CM_Marble_NoiseUVTile, sortPriority=100}
  n26 = MaterialFunctionCall(TextureObject=@n24, TextureSize=@n25, WorldPosition=@n7, World Space Normal=@n8, ProjectionTransitionContrast=@n9) {materialFunction=&MF_Primitive_TriplanarUVSimple}
  n27 = ComponentMask(None=@n26.XYZ Texture) {r=true}
  n28 = TextureObjectParameter() {
   group=Marble
   parameterName=CM_Marble_VeinTex
   samplerType=SAMPLERTYPE_Masks
   sortPriority=0
   texture=/Game/Creative/Sets/PropSets/Primitives/Customization/Textures/Marble/T_CP_Marble_M2
  }
  n29 = MaterialFunctionCall(TextureObject=@n28, TextureSize=@n25, WorldPosition=@n7, World Space Normal=@n8, ProjectionTransitionContrast=@n9) {materialFunction=&MF_Primitive_TriplanarUVSimple}
  n30 = ComponentMask(None=@n29.XYZ Texture) {g=true}
  n31 = ComponentMask(None=@n29.XYZ Texture) {b=true}
  n32 = Add(
   A=LinearInterpolate(
    A=NamedRerouteDeclaration(None=@n27) {name=Veins, nodeColor=<rgba 0 0.6016 1 1>}
    B=Add(A=@n30, B=@n31)
    Alpha=ScalarParameter() {group=Marble, parameterName=CM_Marble_AltGoldVein}
   )
   B=NamedRerouteDeclaration(None=ComponentMask(None=@n26.XYZ Texture) {g=true}) {name=Flakes1, nodeColor=<rgba 0 0.6016 1 1>}
  )
  n33 = ScalarParameter() {defaultValue=6, group=Marble, parameterName=CM_Marble_Noise2UVTile, sortPriority=100}
  n34 = NamedRerouteDeclaration(None=ComponentMask(None=MaterialFunctionCall(TextureObject=@n24, TextureSize=@n33, WorldPosition=@n7, World Space Normal=@n8, ProjectionTransitionContrast=@n9) {materialFunction=&MF_Primitive_TriplanarUVSimple}.XYZ Texture) {b=true}) {name=Flakes2, nodeColor=<rgba 0 0.6016 1 1>}
  NamedRerouteDeclaration(None=StaticSwitchParameter(True=Add(A=@n32, B=@n34), False=Constant()) {defaultValue=true, group=Marble, parameterName=UseGold, sortPriority=-10}) {name=AllMask, nodeColor=<rgba 1 0 0.5156 1>}
  NamedRerouteDeclaration(None=ComponentMask(None=@n29.XYZ Texture) {r=true}) {name=VeinsSmall, nodeColor=<rgba 0 0.6016 1 1>}
  n35 = MaterialFunctionCall() {materialFunction=&/MF_Primitive_CalculatedBounds}
  NamedRerouteDeclaration(None=Divide(A=@n35.WorldPosition, B=ScalarParameter() {defaultValue=100, group=Marble, parameterName=CM_Marble_WorldScale, sortPriority=100})) {name=CalculatedBounds, nodeColor=<rgba 1 0.1406 0 1>}
  NamedRerouteDeclaration(None=@n35.World Normals) {name=WorldNorm, nodeColor=<rgba 0.7344 0 1 1>}
  Power(Base=DotProduct(A=VertexNormalWS(), B=CameraVectorWS()))
 ]}
