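# Material function that adds an optional animated neon gas cloud look to Creative primitive props, behind the Primitive_NeonGasCloud static switch.
# Triplanar-projected panning noise textures provide UV warp offsets for foreground and background cloud layers.
# A color curve atlas colors the layers and Fresnel, feeding hue-shifted emissive, while another noise texture supplies roughness and specular masks.
# Music amplitude from an audio-reactive parameter collection, clamped and smooth-stepped, provides hue shift and warp signals.
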MaterialFunction /Game/Creative/Sets/PropSets/Primitives/Customization/MaterialFunctions/MF_CM_NeonGasCloud
graph {nodes [
  n1 = FunctionInput() {inputType=FunctionInput_MaterialAttributes}
  n2 = SetMaterialAttributes!(
   MaterialAttributes=@n1
   BaseColor=VectorParameter() {group=NeonGasCloud, parameterName=CM_NGC_BaseColor, defaultValue=<rgba 0 0.0006 0.0003 1>}.RGB
   EmissiveColor=NamedRerouteUsage()
   Specular=NamedRerouteUsage()
   Roughness=NamedRerouteUsage()
  ) {attributeSetTypes [
    #1
    #2
    #3
    #4
   ]}
  FunctionOutput(None=StaticSwitchParameter(True=@n2, False=@n1) {group=Customization, parameterName=Primitive_NeonGasCloud})
  n3 = CollectionParameter() {collection=/Game/Athena/Events/AudioAnalysis/MPC_Audio_Reactive_Parameters, parameterName=Amplitude_Music}
  n4 = MaterialFunctionCall() {materialFunction=&/MF_Primitive_AudioSetup}
  n5 = ScalarParameter() {defaultValue=0.25, group=NeonGasCloud, parameterName=CM_NGC_EventAudioBlend}
  n6 = Min(A=Multiply(
    A=LinearInterpolate(A=@n3, B=@n4.Result, Alpha=@n5) {constAlpha=0.25}
    B=ScalarParameter() {defaultValue=1, group=NeonGasCloud, parameterName=CM_NGC_EventAudioMultiplier}
   )) {constB=30, desc=prevents crazy values}
  MaterialFunctionCall!(In=@n6, Ramp Width=ScalarParameter() {defaultValue=0.5, group=NeonGasCloud, parameterName=CM_NGC_EventAudioRamp}) {materialFunction=/Engine/Functions/Engine_MaterialFunctions01/Math/SmoothCeil}
  Multiply(A=ComponentMask(None=ReflectionVectorWS()) {b=true, g=true, r=true}, B=NamedRerouteUsage())
  n7 = Sine(None=Time()) {period=5}
  n8 = NamedRerouteUsage()
  n9 = Ceil(None=@n8)
  n10 = LinearInterpolate(B=Divide(A=Add(A=@n7)), Alpha=@n9)
  n11 = LinearInterpolate(A=@n10, B=Add(A=@n10, B=LinearInterpolate(Alpha=@n8) {constA=0.1, constB=0.14}), Alpha=@n9)
  n12 = CurveAtlasRowParameter(CurveTime=NamedRerouteUsage()) {
   atlas=/Game/Creative/Sets/PropSets/Primitives/Customization/Textures/ColorCurve/CC_CM_Primitives_01
   curve=/Game/Creative/Sets/PropSets/Primitives/Customization/Textures/ColorCurve/C_CM_NeonGasCloud_01
   defaultValue=4
   group=NeonGasCloud
   parameterName=CM_NGC_ColorCurve
  }
  n13 = CurveAtlasRowParameter(CurveTime=NamedRerouteUsage()) {
   atlas=&CC_CM_Primitives_01
   curve=&C_CM_NeonGasCloud_01
   defaultValue=4
   group=NeonGasCloud
   parameterName=CM_NGC_ColorCurve
  }
  n14 = ScalarParameter() {defaultValue=0.2, group=NeonGasCloud, parameterName=CM_NGC_BGBrightness}
  n15 = Add(A=Add(A=@n12, B=Multiply(A=@n13, B=@n14)), B=NamedRerouteUsage())
  n16 = DotProduct(A=CameraVectorWS(), B=VertexNormalWS())
  NamedRerouteDeclaration(None=Multiply(
    A=MaterialFunctionCall(Hue Shift Percentage=@n11, Texture=@n15) {materialFunction=&HueShift}.Result
    B=Power(Base=@n16)
   )) {name=Emissive, nodeColor=<rgba 0 0.9297 1 1>}
  NamedRerouteDeclaration(None=LinearInterpolate(
    A=ScalarParameter() {group=NeonGasCloud, parameterName=CM_NGC_Specular}
    B=ScalarParameter() {defaultValue=0.6, group=NeonGasCloud, parameterName=CM_NGC_SpecularHi}
    Alpha=NamedRerouteUsage()
   )) {name=Spec, nodeColor=<rgba 0 0.4072 0.3968 1>}
  n17 = TextureObjectParameter() {
   group=NeonGasCloud
   parameterName=CM_NGC_NoiseTex
   texture=/Game/Creative/Sets/PropSets/Primitives/Customization/Textures/NeonGasCloud/T_CM_NeonGasCloud_M2
  }
  n18 = Constant()
  n19 = MaterialFunctionCall!(
   TextureObject_Diffuse=@n17
   TextureScale=ScalarParameter() {defaultValue=1, group=NeonGasCloud, parameterName=CM_NGC_Noise_UVTile}
   CalculatedBounds=NamedRerouteUsage()
   WorldNormals=NamedRerouteUsage()
   RotationX=@n18
   RotationY=@n18
   RotationZ=@n18
   ProjectionTransitionContrast=ScalarParameter() {defaultValue=10, group=NeonGasCloud, parameterName=CM_NGC_Noise_TriplannarContrast}
   Material Normals
   NormalTexture
   Normal Texture Intensity
  ) {materialFunction=&/MF_Primitive_TriplanarUVs}
  NamedRerouteDeclaration(None=ComponentMask(None=@n19.RGBA) {r=true}) {name=Roughness, nodeColor=<rgba 1 0.1172 0 1>}
  n20 = TextureObjectParameter() {
   group=NeonGasCloud
   parameterName=CM_NGC_Warp01_Texture
   texture=/Game/Creative/Sets/PropSets/Primitives/Customization/Textures/NeonGasCloud/T_CM_NeonGasCloud_N2
  }
  n21 = ScalarParameter() {defaultValue=0.0025, group=NeonGasCloud, parameterName=CM_NGC_WarpBGUVPan}
  n22 = Multiply(A=Time(), B=@n21)
  n23 = LinearInterpolate(A=@n21, B=ScalarParameter() {defaultValue=0.0025, group=NeonGasCloud, parameterName=CM_NGC_Warp01BGPanHi}, Alpha=NamedRerouteUsage())
  n24 = Constant()
  n25 = MaterialFunctionCall(
   TextureObject_Diffuse=@n20
   TextureScale=ScalarParameter() {defaultValue=1, group=NeonGasCloud, parameterName=CM_NGC_WarpBGUVScale}
   CalculatedBounds=Add(A=Add(A=@n22, B=@n23), B=NamedRerouteUsage())
   WorldNormals=NamedRerouteUsage()
   RotationX=@n24
   RotationY=@n24
   RotationZ=@n24
   ProjectionTransitionContrast=ScalarParameter() {defaultValue=1, group=NeonGasCloud, parameterName=CM_NGC_TriplannarContrast}
  ) {materialFunction=&MF_Primitive_TriplanarUVs}
  NamedRerouteDeclaration(None=Multiply(A=ComponentMask(None=@n25.RGBA) {r=true}, B=ScalarParameter() {defaultValue=0.1, group=NeonGasCloud, parameterName=CM_NGC_WarpBGRIntensity})) {name=BGWarp01, nodeColor=<rgba 0.1484 1 0 1>}
  NamedRerouteDeclaration(None=Multiply(A=ComponentMask(None=@n25.RGBA) {g=true}, B=ScalarParameter() {defaultValue=0.15, group=NeonGasCloud, parameterName=CM_NGC_WarpBGGIntensity})) {name=BGWarp02, nodeColor=<rgba 0 1 0.7188 1>}
  NamedRerouteDeclaration(None=ComponentMask(None=@n19.RGBA) {a=true}) {name=Specular, nodeColor=<rgba 0 0.6016 1 1>}
  n26 = ScalarParameter() {defaultValue=12, group=NeonGasCloud, parameterName=CM_NGC_Fresnel}
  n27 = CurveAtlasRowParameter(CurveTime=Saturate(None=Fresnel(ExponentIn=@n26) {baseReflectFraction=0})) {
   atlas=&CC_CM_Primitives_01
   curve=&C_CM_NeonGasCloud_01
   defaultValue=4
   group=NeonGasCloud
   parameterName=CM_NGC_ColorCurve
  }
  NamedRerouteDeclaration(None=@n27) {name=Fresnel, nodeColor=<rgba 0 0.2969 1 1>}
  n28 = VectorParameter() {group=NeonGasCloud, parameterName=CM_NGC_BaseAnimSpeed, defaultValue=<rgba 0 0.1 0.1 1>}
  n29 = Multiply(A=Time(), B=@n28.RGB)
  n30 = LinearInterpolate(
   A=@n28.RGB
   B=VectorParameter() {group=NeonGasCloud, parameterName=CM_NGC_BaseAnimSpeedWarp, defaultValue=<rgba 0.01 -0.1 -0.1 1>}.RGB
   Alpha=NamedRerouteUsage()
  )
  NamedRerouteDeclaration(None=Add(A=Add(A=@n29, B=@n30), B=NamedRerouteUsage())) {name=UVs, nodeColor=<rgba 0 1 0.4375 1>}
  n31 = TextureObjectParameter() {
   group=NeonGasCloud
   parameterName=CM_NGC_Warp01_Texture
   texture=&T_CM_NeonGasCloud_N2
  }
  n32 = ScalarParameter() {defaultValue=0.0025, group=NeonGasCloud, parameterName=CM_NGC_Warp01UVPan}
  n33 = Multiply(A=Time(), B=@n32)
  n34 = LinearInterpolate(A=@n32, B=ScalarParameter() {defaultValue=0.0025, group=NeonGasCloud, parameterName=CM_NGC_Warp01UVPanHi}, Alpha=NamedRerouteUsage())
  n35 = Constant()
  n36 = MaterialFunctionCall(
   TextureObject_Diffuse=@n31
   TextureScale=ScalarParameter() {defaultValue=1, group=NeonGasCloud, parameterName=CM_NGC_Warp01UVScale}
   CalculatedBounds=Add(A=Add(A=@n33, B=@n34), B=NamedRerouteUsage())
   WorldNormals=NamedRerouteUsage()
   RotationX=@n35
   RotationY=@n35
   RotationZ=@n35
   ProjectionTransitionContrast=ScalarParameter() {defaultValue=1, group=NeonGasCloud, parameterName=CM_NGC_TriplannarContrast}
  ) {materialFunction=&MF_Primitive_TriplanarUVs}
  NamedRerouteDeclaration(None=Multiply(A=ComponentMask(None=@n36.RGBA) {g=true}, B=ScalarParameter() {defaultValue=0.15, group=NeonGasCloud, parameterName=CM_NGC_Warp01GIntensity})) {name=Warp 01 G, nodeColor=<rgba 1 0 0.375 1>}
  NamedRerouteDeclaration(None=@n6) {name=Hueshift, nodeColor=<rgba 0.6172 0 1 1>}
  n37 = Multiply(A=@n3, B=ScalarParameter() {defaultValue=1, group=NeonGasCloud, parameterName=CM_NGC_EventWarpMultiplier})
  NamedRerouteDeclaration(None=MaterialFunctionCall(In=Min(A=@n37) {constB=30, desc=prevents crazy values}) {materialFunction=&SmoothCeil}.Result) {name=Warp, nodeColor=<rgba 0.2188 0 1 1>}
  n38 = TextureObjectParameter() {
   group=NeonGasCloud
   parameterName=CM_NGC_BaseTexture
   texture=/Game/Creative/Sets/PropSets/Primitives/Customization/Textures/NeonGasCloud/T_CM_NeonGasCloud_M1
  }
  n39 = ScalarParameter() {defaultValue=1, group=NeonGasCloud, parameterName=CM_NGC_UVTile}
  n40 = NamedRerouteUsage()
  n41 = NamedRerouteUsage()
  n42 = NamedRerouteUsage()
  n43 = NamedRerouteUsage()
  n44 = Add(A=NamedRerouteUsage(), B=Add(A=Add(A=@n40, B=@n41), B=Add(A=@n42, B=@n43)))
  n45 = NamedRerouteUsage()
  n46 = Constant()
  n47 = ScalarParameter() {defaultValue=1, group=NeonGasCloud, parameterName=CM_NGC_TriplannarContrast}
  NamedRerouteDeclaration(None=ComponentMask(None=MaterialFunctionCall(
     TextureObject_Diffuse=@n38
     TextureScale=@n39
     CalculatedBounds=@n44
     WorldNormals=@n45
     RotationX=@n46
     RotationY=@n46
     RotationZ=@n46
     ProjectionTransitionContrast=@n47
    ) {materialFunction=&MF_Primitive_TriplanarUVs}.RGBA) {r=true}) {name=Base, nodeColor=<rgba 1 0 0.2109 1>}
  NamedRerouteDeclaration(None=LinearInterpolate(
    A=ScalarParameter() {group=NeonGasCloud, parameterName=CM_NGC_Roughness}
    B=ScalarParameter() {defaultValue=1, group=NeonGasCloud, parameterName=CM_NGC_RoughnessHi}
    Alpha=NamedRerouteUsage()
   )) {name=Rough, nodeColor=<rgba 0.4072 0.2831 0 1>}
  n48 = TextureObjectParameter() {
   group=NeonGasCloud
   parameterName=CM_NGC_Warp02_Texture
   texture=/Game/Creative/Sets/PropSets/Primitives/Customization/Textures/NeonGasCloud/T_CM_NeonGasCloud_N1
  }
  n49 = ScalarParameter() {defaultValue=-0.0008, group=NeonGasCloud, parameterName=CM_NGC_Warp02UVPan}
  n50 = Multiply(A=Time(), B=@n49)
  n51 = LinearInterpolate(A=@n49, B=ScalarParameter() {defaultValue=-0.0006, group=NeonGasCloud, parameterName=CM_NGC_Warp02UVPanHi}, Alpha=NamedRerouteUsage())
  n52 = Constant()
  n53 = MaterialFunctionCall(
   TextureObject_Diffuse=@n48
   TextureScale=ScalarParameter() {defaultValue=1, group=NeonGasCloud, parameterName=CM_NGC_Warp02UVScale}
   CalculatedBounds=Add(A=Add(A=@n50, B=@n51), B=NamedRerouteUsage())
   WorldNormals=NamedRerouteUsage()
   RotationX=@n52
   RotationY=@n52
   RotationZ=@n52
   ProjectionTransitionContrast=ScalarParameter() {defaultValue=1, group=NeonGasCloud, parameterName=CM_NGC_TriplannarContrast}
  ) {materialFunction=&MF_Primitive_TriplanarUVs}
  NamedRerouteDeclaration(None=Multiply(
    A=ComponentMask(None=@n53.RGBA) {r=true}
    B=ScalarParameter() {defaultValue=0.0025, group=NeonGasCloud, parameterName=CM_NGC_Warp02RIntensity}
   )) {name=Warp 02 R, nodeColor=<rgba 1 0 0.375 1>}
  NamedRerouteDeclaration(None=Multiply(
    A=ComponentMask(None=@n53.RGBA) {g=true}
    B=ScalarParameter() {defaultValue=0.033, group=NeonGasCloud, parameterName=CM_NGC_Warp02GIntensity}
   )) {name=Warp 02 G, nodeColor=<rgba 1 0 0.375 1>}
  n54 = VectorParameter() {group=NeonGasCloud, parameterName=CM_NGC_BGAnimSpeed, defaultValue=<rgba 0 0.01 0.01 1>}
  n55 = Multiply(A=Time(), B=@n54.RGB)
  n56 = LinearInterpolate(
   A=@n54.RGB
   B=VectorParameter() {group=NeonGasCloud, parameterName=CM_NGC_BGAnimSpeedWarp, defaultValue=<rgba 0.001 -0.01 -0.01 1>}.RGB
   Alpha=NamedRerouteUsage()
  )
  NamedRerouteDeclaration(None=Add(A=Add(A=@n55, B=@n56), B=NamedRerouteUsage())) {name=UVsBG, nodeColor=<rgba 0 1 0.4375 1>}
  NamedRerouteDeclaration(None=Multiply(A=ComponentMask(None=@n36.RGBA) {r=true}, B=ScalarParameter() {defaultValue=0.1, group=NeonGasCloud, parameterName=CM_NGC_Warp01RIntensity})) {name=Warp 01 R, nodeColor=<rgba 1 0 0.375 1>}
  n57 = TextureObjectParameter() {
   group=NeonGasCloud
   parameterName=CM_NGC_BaseTexture
   texture=&T_CM_NeonGasCloud_M1
  }
  n58 = NamedRerouteUsage()
  n59 = NamedRerouteUsage()
  n60 = Constant()
  n61 = MaterialFunctionCall(
   TextureObject_Diffuse=@n57
   TextureScale=ScalarParameter() {defaultValue=2.5, group=NeonGasCloud, parameterName=CM_NGC_UVBGTile}
   CalculatedBounds=Add(A=NamedRerouteUsage(), B=Add(A=@n58, B=@n59))
   WorldNormals=NamedRerouteUsage()
   RotationX=@n60
   RotationY=@n60
   RotationZ=@n60
   ProjectionTransitionContrast=ScalarParameter() {defaultValue=1, group=NeonGasCloud, parameterName=CM_NGC_TriplannarContrast}
  ) {materialFunction=&MF_Primitive_TriplanarUVs}
  NamedRerouteDeclaration(None=Power(
    Base=ChannelMaskParameter(None=@n61.RGBA) {group=NeonGasCloud, parameterName=CM_NGC_BGTextureChannel}.R
    Exp=ScalarParameter() {defaultValue=1, group=NeonGasCloud, parameterName=CM_NGC_BGContrast}
   ) {constExponent=3}) {name=BG, nodeColor=<rgba 0.9453 1 0 1>}
  n62 = MaterialFunctionCall() {materialFunction=&/MF_Primitive_CalculatedBounds}
  NamedRerouteDeclaration(None=Divide(A=@n62.WorldPosition, B=ScalarParameter() {defaultValue=100, group=NeonGasCloud, parameterName=CM_NGC_WorldScale})) {name=CalculatedBounds, nodeColor=<rgba 1 0.1406 0 1>}
  NamedRerouteDeclaration(None=@n62.World Normals) {name=WorldNorm, nodeColor=<rgba 0.7344 0 1 1>}
 ]}
