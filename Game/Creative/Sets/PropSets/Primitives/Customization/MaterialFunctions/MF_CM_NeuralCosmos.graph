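# Customization material function for Creative primitive props, adding an animated, audio-reactive Neural Cosmos look behind a static switch that otherwise passes attributes through.
# Two rotating camera-vector triplanar galaxy layers form the background.
# Triplanar textures in scaled calculated-bounds space supply warp, stars, glow and masks, colored by curve atlases, with Fresnel rainbow and inner glow.
# Music amplitude, capped against spikes, drives warp and hue-shift signals; BaseColor, Specular, Roughness and Emissive are overridden.
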MaterialFunction /Game/Creative/Sets/PropSets/Primitives/Customization/MaterialFunctions/MF_CM_NeuralCosmos
graph {nodes [
  n1 = ScalarParameter() {defaultValue=6, group=NeuralCosmos, parameterName=CM_NC_BGAxisFade}
  n2 = Time()
  n3 = LinearInterpolate(A=@n2, B=Multiply(A=@n2, B=Constant()), Alpha=ScalarParameter() {group=NeuralCosmos, parameterName=CM_NC_BG01Rotation})
  n4 = AppendVector(
   A=AppendVector(
    A=ScalarParameter() {defaultValue=0.02, group=NeuralCosmos, parameterName=CM_NC_BG01RotationSpeed}
    B=ScalarParameter() {defaultValue=-0.02, group=NeuralCosmos, parameterName=CM_NC_BG01RotationSpeed_1}
   )
   B=ScalarParameter() {defaultValue=0.1, group=NeuralCosmos, parameterName=CM_NC_BG01RotationSpeed_2}
  )
  n5 = TextureObjectParameter() {
   group=NeuralCosmos
   parameterName=CM_NC_GalaxyTexture
   texture=/Game/Creative/Sets/PropSets/Primitives/Customization/Textures/GalacticSwirl/T_CM_GalacticSwirl_BC
  }
  n6 = Time()
  n7 = LinearInterpolate(A=@n6, B=Multiply(A=@n6, B=Constant()), Alpha=ScalarParameter() {defaultValue=0.25, group=NeuralCosmos, parameterName=CM_NC_BG02Rotation})
  n8 = AppendVector(
   A=AppendVector(
    A=ScalarParameter() {defaultValue=0.033, group=NeuralCosmos, parameterName=CM_NC_BG02RotationSpeed}
    B=ScalarParameter() {defaultValue=-0.033, group=NeuralCosmos, parameterName=CM_NC_BG02RotationSpeed_1}
   )
   B=ScalarParameter() {defaultValue=0.15, group=NeuralCosmos, parameterName=CM_NC_BG02RotationSpeed_2}
  )
  n9 = Add(
   A=MaterialFunctionCall!(
    CameraVector
    Rotation Speed=Multiply(A=@n3, B=@n4)
    Texture=@n5
    Tiling=ScalarParameter() {defaultValue=0.5, group=NeuralCosmos, parameterName=CM_NC_BG01Tiling}
    Offset
    AxisFadeContrast=@n1
   ) {materialFunction=/Game/Weapons/WeaponSkins/MaterialFunction/TriplanarCameraVector_Rotated}.Result
   B=MaterialFunctionCall(
    Rotation Speed=Multiply(A=@n7, B=@n8)
    Texture=@n5
    Tiling=ScalarParameter() {defaultValue=1, group=NeuralCosmos, parameterName=CM_NC_BG02Tiling}
    AxisFadeContrast=@n1
   ) {materialFunction=&TriplanarCameraVector_Rotated}.Result
  )
  n10 = ScalarParameter() {defaultValue=1, group=NeuralCosmos, parameterName=CM_NC_BGPower}
  n11 = ScalarParameter() {defaultValue=0.5, group=NeuralCosmos, parameterName=CM_NC_BGMultiplier}
  n12 = NamedRerouteUsage()
  n13 = Add(A=NamedRerouteUsage(), B=Multiply(A=Power(Base=@n9, Exp=@n10), B=LinearInterpolate(A=@n11, Alpha=@n12) {constB=1.5}))
  Add(A=NamedRerouteUsage(), B=@n13)
  Add()
  n14 = FunctionInput() {inputType=FunctionInput_MaterialAttributes}
  n15 = SetMaterialAttributes!(
   MaterialAttributes=@n14
   BaseColor=VectorParameter() {group=NeuralCosmos, parameterName=CM_NC_BaseColor, defaultValue=<rgba 0 0.0006 0.0003 1>}.RGB
   EmissiveColor=NamedRerouteUsage()
   Specular=ScalarParameter() {group=NeuralCosmos, parameterName=CM_NC_Specular}
   Roughness=ScalarParameter() {defaultValue=0.2, group=NeuralCosmos, parameterName=CM_NC_Roughness}
  ) {attributeSetTypes [
    #1
    #2
    #3
    #4
   ]}
  FunctionOutput(None=StaticSwitchParameter(True=@n15, False=@n14) {group=Customization, parameterName=Primitive_NeuralCosmos})
  n16 = CollectionParameter() {collection=/Game/Athena/Events/AudioAnalysis/MPC_Audio_Reactive_Parameters, parameterName=Amplitude_Music}
  n17 = MaterialFunctionCall() {materialFunction=&/MF_Primitive_AudioSetup}
  n18 = ScalarParameter() {defaultValue=0.25, group=NeuralCosmos, parameterName=CM_NC_EventAudioBlend}
  n19 = Min(A=Multiply(
    A=LinearInterpolate(A=@n16, B=@n17.Result, Alpha=@n18) {constAlpha=0.25}
    B=ScalarParameter() {defaultValue=1, group=NeuralCosmos, parameterName=CM_NC_EventAudioMultiplier}
   )) {constB=30, desc=prevents crazy values}
  MaterialFunctionCall!(In=@n19, Ramp Width=ScalarParameter() {defaultValue=0.5, group=NeuralCosmos, parameterName=CM_NC_EventAudioRamp}) {materialFunction=/Engine/Functions/Engine_MaterialFunctions01/Math/SmoothCeil}
  n20 = NamedRerouteUsage()
  n21 = Multiply(A=@n20) {constB=2}
  Multiply(A=Power(Base=@n21) {constExponent=3.5}) {constB=50}
  n22 = NamedRerouteUsage()
  Multiply(A=Power(Base=Multiply(A=@n22) {constB=0.1}) {constExponent=3.5}) {constB=50}
  n23 = MaterialFunctionCall() {materialFunction=&/MF_Primitive_CalculatedBounds}
  NamedRerouteDeclaration(None=@n23.LocalPosition) {name=Local Position, nodeColor=<rgba 1 0.75 0 1>}
  n24 = TextureObjectParameter() {
   group=NeuralCosmos
   parameterName=CM_NC_UVWarpTexture
   texture=/Game/Creative/Sets/PropSets/Primitives/Customization/Textures/NeuralCosmos/T_CM_NeuralCosmos_M2
  }
  n25 = NamedRerouteUsage()
  n26 = ScalarParameter() {defaultValue=0.1, group=NeuralCosmos, parameterName=CM_NC_UVwarpSpeed}
  n27 = Add(
   A=Multiply(A=Time(), B=@n26)
   B=LinearInterpolate(A=@n26, B=ScalarParameter() {defaultValue=0.25, group=NeuralCosmos, parameterName=CM_NC_UVwarpSpeedReactive}, Alpha=NamedRerouteUsage())
  )
  n28 = ComponentMask(None=MaterialFunctionCall!(
    TextureObject=@n24
    TextureSize=ScalarParameter() {defaultValue=3, group=NeuralCosmos, parameterName=CM_NC_WarpUVTile}
    WorldPosition=Add(A=@n25, B=@n27)
    World Space Normal=NamedRerouteUsage()
    ProjectionTransitionContrast=NamedRerouteUsage()
   ) {materialFunction=&/MF_Primitive_TriplanarUVSimple}.XYZ Texture) {a=true}
  NamedRerouteDeclaration(None=@n28) {name=WarpTex, nodeColor=<rgba 0.5234 1 0 1>}
  n29 = TextureObjectParameter() {
   group=NeuralCosmos
   parameterName=CM_NC_BaseTex
   texture=/Game/Creative/Sets/PropSets/Primitives/Customization/Textures/NeuralCosmos/T_CM_NeuralCosmos_M1
  }
  n30 = NamedRerouteUsage()
  n31 = Multiply(A=Time(), B=VectorParameter() {group=NeuralCosmos, parameterName=CM_NC_BaseSpeed, defaultValue=<rgba 0.1 0.1 -0.1 1>}.RGB)
  n32 = ScalarParameter() {defaultValue=1, group=NeuralCosmos, parameterName=CM_NC_TriplannarContrast}
  n33 = MaterialFunctionCall(
   TextureObject=@n29
   TextureSize=ScalarParameter() {defaultValue=2, group=NeuralCosmos, parameterName=CM_NC_BaseUVTile}
   WorldPosition=Add(A=NamedRerouteUsage(), B=Add(A=@n30, B=@n31))
   World Space Normal=NamedRerouteUsage()
   ProjectionTransitionContrast=@n32
  ) {materialFunction=&MF_Primitive_TriplanarUVSimple}
  NamedRerouteDeclaration(None=ComponentMask(None=@n33.XYZ Texture) {g=true}) {name=Stars, nodeColor=<rgba 0.875 1 0 1>}
  n34 = ReflectionVectorWS()
  n35 = Panner(Coordinate=Multiply(A=ComponentMask(None=@n34) {g=true, r=true}, B=Fresnel() {baseReflectFraction=0, exponent=0.5})) {speedX=0.25, speedY=0.25}
  n36 = TextureObjectParameter() {
   group=NeuralCosmos
   parameterName=CM_NC_GradTexture01
   texture=/Game/Weapons/WeaponSkins/Wraps/Textures/RainbowGlitter/T_WRap_RainbowGlitter_Grad
  }
  n37 = TextureObjectParameter() {
   group=NeuralCosmos
   parameterName=CM_NC_GradTexture02
   texture=&T_WRap_RainbowGlitter_Grad
  }
  n38 = LinearInterpolate(B=LinearInterpolate(
    A=TextureSample(UVs=@n35, Tex=@n36) {texture=&T_WRap_RainbowGlitter_Grad}.RGB
    B=TextureSample(UVs=@n35, Tex=@n37) {texture=&T_WRap_RainbowGlitter_Grad}.RGB
    Alpha=NamedRerouteUsage()
   )) {constA=1, constAlpha=0.75}
  NamedRerouteDeclaration(None=@n38) {name=Innercolor, nodeColor=<rgba 0 1 0.1562 1>}
  n39 = NamedRerouteUsage()
  n40 = Fresnel(ExponentIn=LinearInterpolate(
    A=ScalarParameter() {defaultValue=0.01, group=NeuralCosmos, parameterName=CM_NC_InnerGlowFresnel}
    B=ScalarParameter() {defaultValue=0.05, group=NeuralCosmos, parameterName=CM_NC_InnerGlowFresnelReactive}
    Alpha=NamedRerouteUsage()
   )) {baseReflectFraction=0}
  n41 = ScalarParameter() {defaultValue=10, group=NeuralCosmos, parameterName=CM_NC_InnerGlowIntensity}
  n42 = ScalarParameter() {defaultValue=20, group=NeuralCosmos, parameterName=CM_NC_InnerGlowIntensityReactive}
  n43 = NamedRerouteUsage()
  n44 = Power(
   Base=Multiply(A=OneMinus(None=@n40), B=LinearInterpolate(A=@n41, B=@n42, Alpha=@n43))
   Exp=ScalarParameter() {defaultValue=1, group=NeuralCosmos, parameterName=CM_NC_InnerGlowContrast}
  )
  NamedRerouteDeclaration(None=Add(A=Multiply(A=@n39, B=@n44), B=@n44)) {name=InnerGlow, nodeColor=<rgba 1 0 0.1641 1>}
  NamedRerouteDeclaration(None=ComponentMask(None=@n33.XYZ Texture) {b=true}) {name=Glow, nodeColor=<rgba 0 1 0.7656 1>}
  NamedRerouteDeclaration(None=Multiply(A=@n28, B=ScalarParameter() {defaultValue=0.05, group=NeuralCosmos, parameterName=CM_NC_WarpAmount})) {name=UVWarp, nodeColor=<rgba 0.5938 0 1 1>}
  n45 = Time()
  n46 = Clamp(None=Sine(None=Multiply(A=@n45) {constB=0.25})) {minDefault=0.25}
  n47 = Add(A=Multiply(A=LinearInterpolate(A=@n22, B=@n22, Alpha=@n46)) {constB=0.2}, B=LinearInterpolate(A=@n20, B=@n21, Alpha=@n46))
  NamedRerouteDeclaration(None=@n47) {name=GlowColor, nodeColor=<rgba 0.7109 1 0 1>}
  NamedRerouteDeclaration(None=@n19) {name=Hueshift, nodeColor=<rgba 0.6172 0 1 1>}
  n48 = Multiply(A=@n16, B=ScalarParameter() {defaultValue=1, group=NeuralCosmos, parameterName=CM_NC_EventWarpMultiplier})
  NamedRerouteDeclaration(None=MaterialFunctionCall(In=Min(A=@n48) {constB=30, desc=prevents crazy values}) {materialFunction=&SmoothCeil}.Result) {name=Warp, nodeColor=<rgba 0.2188 0 1 1>}
  NamedRerouteDeclaration(None=@n32) {name=ContrastTriplanar, nodeColor=<rgba 1 0.6563 0 1>}
  NamedRerouteDeclaration(None=@n23.Calculated Bounds) {name=Calcs, nodeColor=<rgba 0 0.4141 1 1>}
  n49 = ScalarParameter() {defaultValue=1, group=NeuralCosmos, parameterName=CM_NC_EmissiveIntensity}
  n50 = ScalarParameter() {defaultValue=2.5, group=NeuralCosmos, parameterName=CM_NC_EmissiveIntensityReactive}
  n51 = NamedRerouteUsage()
  n52 = NamedRerouteUsage()
  NamedRerouteDeclaration(None=Multiply(A=LinearInterpolate(A=@n49, B=@n50, Alpha=@n51), B=Add(A=@n52, B=@n13))) {name=Emissive, nodeColor=<rgba 0 1 1 1>}
  n53 = ScalarParameter() {defaultValue=12, group=NeuralCosmos, parameterName=CM_NC_FrenselMaskExp}
  NamedRerouteDeclaration(None=Multiply(A=@n38, B=Fresnel(ExponentIn=@n53) {baseReflectFraction=0})) {name=Rainbow, nodeColor=<rgba 0 0.3672 1 1>}
  n54 = Add(A=@n47, B=ComponentMask(None=@n33.XYZ Texture) {r=true})
  n55 = ComponentMask(None=@n33.XYZ Texture) {a=true}
  n56 = NamedRerouteUsage()
  NamedRerouteDeclaration(None=Add(A=@n54, B=LinearInterpolate(A=@n54, B=@n55, Alpha=@n56))) {name=Mask, nodeColor=<rgba 1 0 0.375 1>}
  n57 = TextureObjectParameter() {
   group=NeuralCosmos
   parameterName=CM_NC_UVWarpTexture
   texture=&T_CM_NeuralCosmos_M2
  }
  n58 = Multiply(A=NamedRerouteUsage()) {constB=10}
  n59 = Time()
  n60 = ScalarParameter() {defaultValue=0.1, group=NeuralCosmos, parameterName=CM_NC_ColorSpeed}
  n61 = ScalarParameter() {defaultValue=0.25, group=NeuralCosmos, parameterName=CM_NC_ColorSpeedReactive}
  n62 = NamedRerouteUsage()
  n63 = Add(A=NamedRerouteUsage(), B=Add(A=Multiply(A=@n59, B=@n60), B=LinearInterpolate(A=@n60, B=@n61, Alpha=@n62)))
  n64 = ComponentMask(None=MaterialFunctionCall(
    TextureObject=@n57
    TextureSize=ScalarParameter() {defaultValue=2, group=NeuralCosmos, parameterName=CM_NC_ColorUVTile}
    WorldPosition=Add(A=@n58, B=@n63)
    World Space Normal=NamedRerouteUsage()
    ProjectionTransitionContrast=ScalarParameter() {defaultValue=1, group=NeuralCosmos, parameterName=CM_NC_TriplannarContrast}
   ) {materialFunction=&MF_Primitive_TriplanarUVSimple}.XYZ Texture) {a=true}
  n65 = CurveAtlasRowParameter(CurveTime=@n64) {
   atlas=/Game/Creative/Sets/PropSets/Primitives/Customization/Textures/ColorCurve/CC_CM_Primitives_01
   curve=/Game/Creative/Sets/PropSets/Primitives/Customization/Textures/ColorCurve/C_CM_CosmosColors_01
   defaultValue=9
   group=NeuralCosmos
   parameterName=CM_NC_ColorCurve01
  }
  n66 = CurveAtlasRowParameter(CurveTime=@n64) {
   atlas=&CC_CM_Primitives_01
   curve=/Game/Creative/Sets/PropSets/Primitives/Customization/Textures/ColorCurve/C_CM_CosmosColors_02
   defaultValue=10
   group=NeuralCosmos
   parameterName=CM_NC_ColorCurve02
  }
  n67 = LinearInterpolate(A=@n65, B=@n66, Alpha=NamedRerouteUsage())
  n68 = NamedRerouteUsage()
  n69 = Add(A=@n67, B=Multiply(A=@n68, B=@n68))
  NamedRerouteDeclaration(None=Multiply(A=Add(A=@n67, B=@n69), B=NamedRerouteUsage())) {name=MaskFinal, nodeColor=<rgba 1 0 0.375 1>}
  NamedRerouteDeclaration(None=Divide(A=@n23.WorldPosition, B=ScalarParameter() {defaultValue=175, group=NeuralCosmos, parameterName=CM_NC_WorldScale})) {name=CalculatedBounds, nodeColor=<rgba 1 0.1406 0 1>}
  NamedRerouteDeclaration(None=@n23.World Normals) {name=WorldNorm, nodeColor=<rgba 0.7344 0 1 1>}
  NamedRerouteUsage()
  NamedRerouteUsage()
  NamedRerouteUsage()
  NamedRerouteUsage()
  ScalarParameter() {defaultValue=-0.05, group=NeuralCosmos, parameterName=CM_NC_BaseSpeed}
 ]}
