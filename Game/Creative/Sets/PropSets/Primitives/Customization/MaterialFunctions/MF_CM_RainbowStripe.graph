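# Material function that adds an optional rainbow stripe customization look to primitive props, overriding incoming material attributes behind the Primitive_RainbowStripe static switch.
# Two stripe textures, projected in world space or by calculated bounds, blend by Fresnel alongside a panning reflection-based rainbow gradient.
# Music amplitude, blended with primitive audio setup, drives emissive hue shift and intensity.
# Triplanar noise supplies roughness and specular masks.
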MaterialFunction /Game/Creative/Sets/PropSets/Primitives/Customization/MaterialFunctions/MF_CM_RainbowStripe
graph {nodes [
  n1 = ComponentMask(None=ReflectionVectorWS()) {g=true, r=true}
  Add(A=@n1, B=ComponentMask(None=NamedRerouteUsage()) {g=true, r=true})
  n2 = MaterialFunctionCall() {materialFunction=&/MF_Primitive_CalculatedBounds}
  ComponentMask(None=@n2.WorldPosition) {b=true}
  n3 = VectorParameter() {group=RainbowStripe, parameterName=CM_RS_BaseColor, defaultValue=<rgba 0 0.0006 0.0003 1>}
  n4 = NamedRerouteUsage()
  n5 = ScalarParameter() {defaultValue=0.85, group=RainbowStripe, parameterName=CM_RS_BaseColorBlend}
  n6 = OneMinus(None=NamedRerouteUsage())
  n7 = FunctionInput() {inputType=FunctionInput_MaterialAttributes}
  n8 = SetMaterialAttributes!(
   MaterialAttributes=@n7
   BaseColor=LinearInterpolate(A=@n3.RGB, B=LinearInterpolate(A=@n3.RGB, B=@n4, Alpha=@n5), Alpha=Power(Base=@n6))
   EmissiveColor=NamedRerouteUsage()
   Specular=NamedRerouteUsage()
   Roughness=NamedRerouteUsage()
   Metallic=ScalarParameter() {group=RainbowStripe, parameterName=CM_RS_Metallic}
   Normal
  ) {attributeSetTypes [
    #1
    #2
    #3
    #4
    #5
    #6
   ]}
  FunctionOutput(None=StaticSwitchParameter(True=@n8, False=@n7) {group=Customization, parameterName=Primitive_RainbowStripe})
  n9 = CollectionParameter() {collection=/Game/Athena/Events/AudioAnalysis/MPC_Audio_Reactive_Parameters, parameterName=Amplitude_Music}
  n10 = MaterialFunctionCall() {materialFunction=&/MF_Primitive_AudioSetup}
  n11 = ScalarParameter() {defaultValue=0.25, group=RainbowStripe, parameterName=CM_RS_EventAudioBlend}
  n12 = Min(A=Multiply(
    A=LinearInterpolate(A=@n9, B=@n10.Result, Alpha=@n11) {constAlpha=0.25}
    B=ScalarParameter() {defaultValue=1, group=RainbowStripe, parameterName=CM_RS_EventAudioMultiplier}
   )) {constB=30, desc=prevents crazy values}
  MaterialFunctionCall!(In=@n12, Ramp Width=ScalarParameter() {defaultValue=0.5, group=RainbowStripe, parameterName=CM_RS_EventAudioRamp}) {materialFunction=/Engine/Functions/Engine_MaterialFunctions01/Math/SmoothCeil}
  n13 = WorldPosition()
  n14 = ObjectPositionWS()
  n15 = Divide(
   A=Subtract(A=@n13.XYZ, B=TransformPosition(None=@n14) {transformSourceType=TRANSFORMPOSSOURCE_World})
   B=MaterialFunctionCall() {materialFunction=&ObjectLocalBounds}.Local Bounds Size
  )
  Multiply(A=ComponentMask(None=Add(A=@n15) {constB=0.5}) {b=true, g=true}, B=Constant2Vector() {g=-1, r=1})
  NamedRerouteDeclaration(None=@n2.LocalPosition) {name=Local Position, nodeColor=<rgba 1 0.75 0 1>}
  n16 = NamedRerouteUsage()
  n17 = TextureObjectParameter() {
   group=RainbowStripe
   parameterName=CM_RS_BaseColorTexture
   texture=/Game/Creative/Sets/PropSets/Primitives/Customization/Textures/RainbowStripe/T_CM_RainbowStripe_BC
  }
  n18 = TextureSample(UVs=@n16, Tex=@n17) {samplerSource=SSM_Wrap_WorldGroupSettings, texture=/Game/Weapons/WeaponSkins/Wraps/Textures/RainbowGlitter/T_WRap_RainbowGlitter_Grad}
  n19 = TextureSample(UVs=@n16, Tex=@n17) {samplerSource=SSM_Clamp_WorldGroupSettings, texture=&T_WRap_RainbowGlitter_Grad}
  n20 = TextureObjectParameter() {
   group=RainbowStripe
   parameterName=CM_RS_OverlayTexture
   texture=/Game/Creative/Sets/PropSets/Primitives/Customization/Textures/RainbowStripe/T_CM_RainbowStripe_BC2
  }
  n21 = TextureSample(UVs=@n16, Tex=@n20) {samplerSource=SSM_Wrap_WorldGroupSettings, texture=&T_WRap_RainbowGlitter_Grad}
  n22 = TextureSample(UVs=@n16, Tex=@n20) {samplerSource=SSM_Clamp_WorldGroupSettings, texture=&T_WRap_RainbowGlitter_Grad}
  n23 = LinearInterpolate(
   A=StaticSwitchParameter(True=@n18.RGB, False=@n19.RGB) {defaultValue=true, group=RainbowStripe, parameterName=UseWorldSpace, sortPriority=0}
   B=StaticSwitchParameter(True=@n21.RGB, False=@n22.RGB) {defaultValue=true, group=RainbowStripe, parameterName=UseWorldSpace, sortPriority=0}
   Alpha=Fresnel(ExponentIn=ScalarParameter() {defaultValue=1, group=RainbowStripe, parameterName=CM_RS_HighlightFresnel}) {baseReflectFraction=0}
  )
  NamedRerouteDeclaration(None=@n23) {name=Color, nodeColor=<rgba 1 0 0.8203 1>}
  NamedRerouteDeclaration(None=LinearInterpolate(
    A=ScalarParameter() {group=RainbowStripe, parameterName=CM_RS_Specular}
    B=ScalarParameter() {defaultValue=1, group=RainbowStripe, parameterName=CM_RS_SpecularHi}
    Alpha=NamedRerouteUsage()
   )) {name=FinalSpec, nodeColor=<rgba 0 0.4072 0.3968 1>}
  NamedRerouteDeclaration(None=@n12) {name=Hueshift, nodeColor=<rgba 0.6172 0 1 1>}
  n24 = Multiply(A=@n9, B=ScalarParameter() {defaultValue=1, group=RainbowStripe, parameterName=CM_RS_EventWarpMultiplier})
  NamedRerouteDeclaration(None=MaterialFunctionCall(In=Min(A=@n24) {constB=30, desc=prevents crazy values}) {materialFunction=&SmoothCeil}.Result) {name=Warp, nodeColor=<rgba 0.2188 0 1 1>}
  NamedRerouteDeclaration(None=LinearInterpolate(
    A=ScalarParameter() {group=RainbowStripe, parameterName=CM_RS_Roughness}
    B=ScalarParameter() {defaultValue=0.15, group=RainbowStripe, parameterName=CM_RS_RoughnessHi}
    Alpha=NamedRerouteUsage()
   )) {name=FinalRough, nodeColor=<rgba 0.4072 0.2831 0 1>}
  n25 = Fresnel(ExponentIn=ScalarParameter() {defaultValue=4, group=RainbowStripe, parameterName=CM_RS_BloomShadowExp}) {baseReflectFraction=0}
  n26 = ScalarParameter() {group=RainbowStripe, parameterName=CM_RS_BloomShadowAdd}
  NamedRerouteDeclaration(None=Saturate(None=Add(A=@n25, B=@n26))) {name=FresnelMask, nodeColor=<rgba 0 0.6016 1 1>}
  n27 = TextureObjectParameter() {
   group=RainbowStripe
   parameterName=CM_RS_NoiseTex
   texture=/Game/Creative/Sets/PropSets/Primitives/Customization/Textures/NeonGasCloud/T_CM_NeonGasCloud_M2
  }
  n28 = MaterialFunctionCall!(
   TextureObject=@n27
   TextureSize=ScalarParameter() {defaultValue=1, group=RainbowStripe, parameterName=CM_RS_Noise_UVTile}
   WorldPosition=NamedRerouteUsage()
   World Space Normal=NamedRerouteUsage()
   ProjectionTransitionContrast=ScalarParameter() {defaultValue=10, group=RainbowStripe, parameterName=CM_RS_Noise_TriplannarContrast}
  ) {materialFunction=&/MF_Primitive_TriplanarUVSimple}
  NamedRerouteDeclaration(None=ComponentMask(None=@n28.XYZ Texture) {r=true}) {name=Roughness, nodeColor=<rgba 1 0.1172 0 1>}
  NamedRerouteDeclaration(None=ComponentMask(None=@n28.XYZ Texture) {a=true}) {name=Specular, nodeColor=<rgba 0 0.6016 1 1>}
  n29 = ScalarParameter() {group=RainbowStripe, parameterName=CM_RS_EventHueShiftAmount}
  n30 = NamedRerouteUsage()
  n31 = Add(
   A=Multiply(A=Time(), B=@n29)
   B=LinearInterpolate(A=@n29, B=ScalarParameter() {defaultValue=0.2, group=RainbowStripe, parameterName=CM_RS_EventHueShiftAmountReactive}, Alpha=@n30)
  )
  n32 = Ceil(None=@n30)
  n33 = Multiply(A=NamedRerouteUsage(), B=Add(A=NamedRerouteUsage(), B=@n23))
  n34 = LinearInterpolate(
   A=ScalarParameter() {defaultValue=1.25, group=RainbowStripe, parameterName=CM_RS_HueShiftMultiplier}
   B=ScalarParameter() {defaultValue=2.5, group=RainbowStripe, parameterName=CM_RS_HueShiftMultiplierReactive}
   Alpha=@n30
  )
  NamedRerouteDeclaration(None=MaterialFunctionCall(Hue Shift Percentage=LinearInterpolate(B=@n31, Alpha=@n32), Texture=Multiply(A=@n33, B=@n34)) {materialFunction=&HueShift}.Result) {name=Emissive, nodeColor=<rgba 1 0 0.2109 1>}
  NamedRerouteDeclaration(None=Divide(A=@n2.WorldPosition, B=ScalarParameter() {defaultValue=175, group=RainbowStripe, parameterName=CM_RS_WorldScale})) {name=WorldPos, nodeColor=<rgba 0.6406 0 1 1>}
  NamedRerouteDeclaration(None=@n2.Calculated Bounds) {name=Calcs, nodeColor=<rgba 0 0.4141 1 1>}
  n35 = NamedRerouteUsage()
  n36 = TextureObject() {texture=/Game/Creative/Sets/PropSets/Primitives/Customization/Textures/RainbowStripe/T_CM_RainbowStripe_N}
  n37 = TextureSample(UVs=@n35, Tex=@n36) {samplerSource=SSM_Clamp_WorldGroupSettings, samplerType=SAMPLERTYPE_Normal, texture=&FlatNormalBlue}
  n38 = TextureSample(UVs=@n35, Tex=@n36) {samplerSource=SSM_Wrap_WorldGroupSettings, samplerType=SAMPLERTYPE_Normal, texture=&FlatNormalBlue}
  NamedRerouteDeclaration(None=StaticSwitchParameter(True=@n37.RGB, False=@n38.RGB) {defaultValue=true, group=RainbowStripe, parameterName=UseWorldSpace, sortPriority=0}) {name=Normal, nodeColor=<rgba 0.0781 1 0 1>}
  n39 = NamedRerouteUsage()
  n40 = TextureObjectParameter() {
   group=RainbowStripe
   parameterName=CM_RS_GradTexture01
   texture=&T_WRap_RainbowGlitter_Grad
  }
  n41 = TextureSample(UVs=Panner(Coordinate=Add(A=@n1, B=@n39)) {speedX=0.25, speedY=0.25}, Tex=@n40) {samplerSource=SSM_Wrap_WorldGroupSettings, texture=&T_WRap_RainbowGlitter_Grad}
  n42 = LinearInterpolate(B=@n41.RGB) {constA=1, constAlpha=0.75}
  n43 = Fresnel(ExponentIn=ScalarParameter() {defaultValue=12, group=RainbowStripe, parameterName=CM_RS_FrenselMaskExp}) {baseReflectFraction=0}
  NamedRerouteDeclaration(None=Multiply(A=Multiply(A=@n42, B=@n43)) {constB=0.5}) {name=Rainbow, nodeColor=<rgba 0 0.3672 1 1>}
  n44 = ScalarParameter() {defaultValue=175, group=RainbowStripe, parameterName=CM_RS_BaseColorWorldScale}
  n45 = MaterialFunctionCall(Color=Multiply(A=Divide(A=@n2.WorldPosition, B=@n44), B=Constant3Vector() {constant=<rgba 1 1 -1 0>})) {materialFunction=&SplitComponents}
  n46 = MaterialFunctionCall(Color=Multiply(A=NamedRerouteUsage(), B=Constant3Vector() {constant=<rgba 1 1 -1 0>})) {materialFunction=&SplitComponents}
  NamedRerouteDeclaration(None=StaticSwitchParameter(True=AppendVector(A=@n45.G, B=@n45.G), False=AppendVector(A=@n46.G, B=@n46.G)) {defaultValue=true, group=RainbowStripe, parameterName=UseWorldSpace, sortPriority=0}) {name=CalculatedBounds, nodeColor=<rgba 1 0.1406 0 1>}
  NamedRerouteDeclaration(None=@n2.World Normals) {name=WorldNorm, nodeColor=<rgba 0.7344 0 1 1>}
  NamedRerouteUsage()
  Transform(None=@n13.XYZ) {transformSourceType=TRANSFORMSOURCE_World, transformType=TRANSFORM_Local}
 ]}
