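# Material function that applies a sand castle look to Creative primitive props when the Primitive_SandCastle static switch is on.
# Triplanar projections on calculated bounds supply tinted base color, two blended normal layers and packed specular, roughness and AO.
# Sparkles come from a time-driven hue-shifted detail texture dotted with view direction, scaled by camera-depth fade and a sunlight term.
# The attributes, with emissive zeroed on mobile, pass through MF_PrincessCastle_SandFuzz before output.
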MaterialFunction /Game/Creative/Sets/PropSets/Primitives/Customization/MaterialFunctions/MF_CM_SandCastle
graph {nodes [
  n1 = CollectionParameter() {
   collection=&FortniteMaterialParameters
   parameterName=SunlightDirectionalVector
  }
  n2 = MaterialFunctionCall(Float4=@n1) {materialFunction=&BreakOutFloat4Components}
  n3 = DotProduct(
   A=Normalize(VectorInput=MaterialFunctionCall(X=@n2.R, Y=@n2.R, Z=@n2.R) {materialFunction=&MakeFloat3}.Result)
   B=PixelNormalWS()
  )
  n4 = PreSkinnedPosition()
  n5 = ShaderStageSwitch(PixelShader=VertexInterpolator(VS=@n4).PS, VertexShader=@n4)
  n6 = VectorParameter() {group=SandCastle, parameterName=GradientZ_Offset, sortPriority=500, defaultValue=<rgba 0 0 0 1>}
  n7 = ChannelMaskParameter(None=Divide(
    A=Add(A=@n5, B=@n6.RGB)
    B=MaterialFunctionCall() {materialFunction=&ObjectLocalBounds}.Local Bounds Size
   )) {group=SandCastle, maskChannel=Blue, parameterName=ZMask, sortPriority=500, defaultValue=<rgba 0 0 1 0>}
  Add(A=Saturate(None=@n3), B=@n7.R)
  n8 = Multiply(A=TextureCoordinate()) {constB=4}
  Add(A=@n8, B=Multiply(A=Time()) {constB=0.05})
  n9 = NamedRerouteUsage()
  Add(
   A=@n9
   B=TextureSample(UVs=@n8) {samplerType=SAMPLERTYPE_Normal, texture=/Game/Weapons/WeaponSkins/Wraps/Textures/T_Wrap_metallicFlakes_Small}.RGB
  )
  Constant() {r=0.5}
  n10 = Multiply(A=NamedRerouteUsage(), B=NamedRerouteUsage())
  n11 = Multiply(A=ScalarParameter() {defaultValue=0.25, group=SandCastle, parameterName=CM_Emissive}, B=@n10)
  n12 = Multiply(A=ScalarParameter() {defaultValue=2, group=SandCastle, parameterName=CM_EmissiveMax}, B=@n10)
  n13 = MaterialFunctionCall(Fade Length=Constant() {r=2500}, Fade Offset=Constant() {r=250}) {materialFunction=&CameraDepthFade}
  n14 = MakeMaterialAttributes!(
   BaseColor=NamedRerouteUsage()
   Metallic=ScalarParameter() {group=SandCastle, parameterName=CM_Metallic}
   Specular=NamedRerouteUsage()
   Roughness=NamedRerouteUsage()
   Anisotropy
   EmissiveColor=ShadingPathSwitch(Default=LinearInterpolate(A=@n11, B=@n12, Alpha=@n13.Result), Mobile=Constant())
   Opacity
   OpacityMask
   Normal=NamedRerouteUsage()
   Tangent
   WorldPositionOffset
   SubsurfaceColor
   ClearCoat
   ClearCoatRoughness
   AmbientOcclusion=NamedRerouteUsage()
   Refraction
   CustomizedUVs_0
   CustomizedUVs_1
   CustomizedUVs_2
   CustomizedUVs_3
   CustomizedUVs_4
   CustomizedUVs_5
   CustomizedUVs_6
   CustomizedUVs_7
   PixelDepthOffset
   ShadingModel
   Displacement
  )
  FunctionOutput(None=StaticSwitchParameter(
    True=MaterialFunctionCall!(In=@n14) {materialFunction=/Game/Environments/Sets/PrincessCastle/MaterialFunctions/MF_PrincessCastle_SandFuzz}.Result
    False=NamedRerouteUsage()
   ) {group=Customization, parameterName=Primitive_SandCastle})
  n15 = TextureObjectParameter() {
   group=SandCastle
   parameterName=CM_SandCastle_S
   sortPriority=0
   texture=/Game/Creative/Sets/PropSets/Primitives/Customization/Textures/SandCastle/T_CM_SandCastle_S
  }
  n16 = ScalarParameter() {defaultValue=12, group=SandCastle, parameterName=CM_SandCastle_NormalLargeTile, sortPriority=99}
  n17 = NamedRerouteUsage()
  n18 = NamedRerouteUsage()
  n19 = Constant()
  n20 = ScalarParameter() {defaultValue=4, group=SandCastle, parameterName=CM_SandCastle_NormalLargeContrast, sortPriority=99}
  n21 = NamedRerouteUsage()
  n22 = TextureObjectParameter() {
   group=SandCastle
   parameterName=CM_SandCastle_NormalLarge
   samplerType=SAMPLERTYPE_Normal
   sortPriority=0
   texture=/Game/Creative/Sets/PropSets/Primitives/Customization/Textures/SandCastle/T_CM_SandCastle_N2
  }
  n23 = MaterialFunctionCall!(
   TextureObject_Diffuse=@n15
   TextureScale=@n16
   CalculatedBounds=@n17
   WorldNormals=@n18
   RotationX=@n19
   RotationY=@n19
   RotationZ=@n19
   ProjectionTransitionContrast=@n20
   Material Normals=@n21
   NormalTexture=@n22
   Normal Texture Intensity=ScalarParameter() {defaultValue=1, group=SandCastle, parameterName=CM_SandCastle_NormalLargeIntensity, sortPriority=99}
  ) {materialFunction=&/MF_Primitive_TriplanarUVs}
  n24 = ComponentMask(None=VertexNormalWS()) {b=true}
  n25 = OneMinus(None=@n24)
  LinearInterpolate(
   A=@n23.Normal Texture
   B=MaterialFunctionCall(Normal=@n23.Normal Texture, Flatness=ScalarParameter() {defaultValue=-2, group=SandCastle, parameterName=AdditionalNormalStrength}) {materialFunction=&FlattenNormal}.Result
   Alpha=Saturate(None=StaticSwitchParameter(True=@n25, False=@n24) {defaultValue=true, group=SandCastle, parameterName=AdditionalNormalStrengthZaxis, sortPriority=-100})
  )
  n26 = LinearInterpolate(
   A=ScalarParameter() {defaultValue=0.33, group=SandCastle, parameterName=CM_SandCastle_RoughnessMin}
   B=ScalarParameter() {defaultValue=2, group=SandCastle, parameterName=CM_SandCastle_RoughnessMax}
   Alpha=NamedRerouteUsage()
  )
  LinearInterpolate(A=@n26, B=ScalarParameter() {defaultValue=0.44, group=SandCastle, parameterName=CM_SandCastle_RoughnessSparkle}, Alpha=NamedRerouteUsage())
  n27 = TextureObjectParameter() {
   group=SandCastle
   parameterName=CM_SandCastle_BaseColor
   sortPriority=0
   texture=/Game/Creative/Sets/PropSets/Primitives/Customization/Textures/SandCastle/T_CM_SandCastle_BC
  }
  n28 = ScalarParameter() {defaultValue=1, group=SandCastle, parameterName=CM_SandCastle_UVTile, sortPriority=100}
  n29 = ScalarParameter() {defaultValue=4, group=SandCastle, parameterName=CM_SandCastle_TriplannarContrast, sortPriority=101}
  n30 = TextureObjectParameter() {
   group=SandCastle
   parameterName=CM_SandCastle_Normal
   samplerType=SAMPLERTYPE_Normal
   sortPriority=0
   texture=/Game/Creative/Sets/PropSets/Primitives/Customization/Textures/SandCastle/T_CM_SandCastle_N1
  }
  n31 = MaterialFunctionCall(
   TextureObject_Diffuse=@n27
   TextureScale=@n28
   CalculatedBounds=@n17
   WorldNormals=@n18
   RotationX=@n19
   RotationY=@n19
   RotationZ=@n19
   ProjectionTransitionContrast=@n29
   Material Normals=@n21
   NormalTexture=@n30
   Normal Texture Intensity=ScalarParameter() {defaultValue=0.75, group=SandCastle, parameterName=CM_SandCastle_NormalIntensity, sortPriority=99}
  ) {materialFunction=&MF_Primitive_TriplanarUVs}
  n32 = ComponentMask(None=@n31.RGBA) {b=true, g=true, r=true}
  n33 = VectorParameter() {group=SandCastle, parameterName=CM_SandCastle_BaseColorTint, defaultValue=<rgba 0.8078 0.6471 0.5294 1>}
  n34 = MaterialFunctionCall(Base=@n32, Blend=@n33.RGB) {materialFunction=&Blend_Overlay}
  LinearInterpolate(A=@n34.Result, B=@n33.RGB)
  Multiply(A=@n1) {constB=-1}
  n35 = GetMaterialAttributes(None=FunctionInput() {inputType=FunctionInput_MaterialAttributes}) {attributeGetTypes [#1, #2, #3]}
  NamedRerouteDeclaration(None=LinearInterpolate(
    A=@n35.Normal
    B=Constant3Vector() {constant=<rgba 0 0 1 0>}
    Alpha=ScalarParameter() {defaultValue=0.65, group=SandCastle, parameterName=CM_BaseNormalBlend}
   )) {name=InNormal, nodeColor=<rgba 0 0.2266 1 1>}
  NamedRerouteDeclaration(None=@n35.MaterialAttributes) {name=Material Attribute, nodeColor=<rgba 0.6641 0 1 1>}
  n36 = ScalarParameter() {group=SandCastle, parameterName=CM_SandCastle_SpecularMin}
  n37 = ScalarParameter() {defaultValue=1, group=SandCastle, parameterName=CM_SandCastle_SpecularMax}
  n38 = NamedRerouteUsage()
  NamedRerouteDeclaration(None=LinearInterpolate(
    A=LinearInterpolate(A=@n36, B=@n37, Alpha=@n38)
    B=ScalarParameter() {defaultValue=0.75, group=SandCastle, parameterName=CM_SandCastle_SpecularSparkle}
    Alpha=NamedRerouteUsage()
   )) {name=FinalSpec, nodeColor=<rgba 0 0.4072 0.3968 1>}
  NamedRerouteDeclaration(None=@n26) {name=FinalRough, nodeColor=<rgba 0.4072 0.2831 0 1>}
  n39 = MaterialFunctionCall!(TextureObject=@n15, TextureSize=@n28, WorldPosition=@n17, World Space Normal=@n18, ProjectionTransitionContrast=@n29) {materialFunction=&/MF_Primitive_TriplanarUVSimple}
  NamedRerouteDeclaration(None=ComponentMask(None=@n39.XYZ Texture) {b=true}) {name=Specular, nodeColor=<rgba 0 0.6016 1 1>}
  n40 = MaterialFunctionCall() {materialFunction=&/MF_Primitive_CalculatedBounds}
  NamedRerouteDeclaration(None=Divide(A=@n40.WorldPosition, B=ScalarParameter() {defaultValue=100, group=SandCastle, parameterName=CM_SandCastle_WorldScale, sortPriority=100})) {name=CalculatedBounds, nodeColor=<rgba 1 0.1406 0 1>}
  NamedRerouteDeclaration(None=@n40.World Normals) {name=WorldNorm, nodeColor=<rgba 0.7344 0 1 1>}
  NamedRerouteDeclaration(None=NamedRerouteUsage()) {name=UVs, nodeColor=<rgba 0 1 0.4375 1>}
  n41 = Saturate(None=NamedRerouteUsage())
  n42 = NamedRerouteUsage()
  n43 = NamedRerouteUsage()
  NamedRerouteDeclaration(None=LinearInterpolate(A=LinearInterpolate(A=@n32, B=@n34.Result, Alpha=@n41), B=@n34.Result, Alpha=Max(A=@n42, B=@n43))) {name=BaseColor, nodeColor=<rgba 0 0.8594 1 1>}
  NamedRerouteDeclaration(None=ComponentMask(None=@n39.XYZ Texture) {r=true}) {name=Roughness, nodeColor=<rgba 0.5703 1 0 1>}
  n44 = DotProduct(A=VertexNormalWS(), B=CameraVectorWS())
  NamedRerouteDeclaration(None=Power(Base=OneMinus(None=@n44))) {name=Fresnel, nodeColor=<rgba 0 1 0.7891 1>}
  n45 = Desaturation(None=NamedRerouteUsage(), Fraction=Constant() {r=1})
  n46 = Round(None=Multiply(A=Power(Base=@n45) {constExponent=3}) {constB=15})
  n47 = NamedRerouteDeclaration(None=MaterialFunctionCall(BaseNormal=@n31.Normal Texture, AdditionalNormal=@n23.Normal Texture) {materialFunction=&BlendAngleCorrectedNormals}.Result) {name=Normal, nodeColor=<rgba 0 0.8594 1 1>}
  n48 = ComponentMask(None=@n47) {r=true}
  n49 = ComponentMask(None=@n47) {g=true}
  n50 = LinearInterpolate(Alpha=Add(A=Add(A=@n48, B=@n49), B=NamedRerouteUsage())) {constA=0.33}
  n51 = TextureObjectParameter() {
   group=SandCastle
   parameterName=CM_SandCastle_Detail
   sortPriority=0
   texture=/Game/Creative/Sets/PropSets/Primitives/Customization/Textures/SandCastle/T_CM_SandCastle_M
  }
  n52 = ComponentMask(None=MaterialFunctionCall(
    TextureObject_Diffuse=@n51
    TextureScale=ScalarParameter() {defaultValue=0.7, group=SandCastle, parameterName=CM_SandCastle_NormalDetailTile, sortPriority=99}
    CalculatedBounds=@n17
    WorldNormals=@n18
    RotationX=@n19
    RotationY=@n19
    RotationZ=@n19
    ProjectionTransitionContrast=ScalarParameter() {defaultValue=4, group=SandCastle, parameterName=CM_SandCastle_NormalDetailContrast, sortPriority=99}
    Material Normals=@n21
    Normal Texture Intensity=ScalarParameter() {defaultValue=0.5, group=SandCastle, parameterName=CM_SandCastle_NormalDetailIntensity, sortPriority=99}
   ) {materialFunction=&MF_Primitive_TriplanarUVs}.RGBA) {b=true, g=true, r=true}
  n53 = MaterialFunctionCall(
   Hue Shift Percentage=Multiply(A=Time(), B=ScalarParameter() {defaultValue=0.175, group=SandCastle, parameterName=CM_SandCastle_SparkleSpeed, sortPriority=99}) {constB=0.075}
   Texture=Normalize(VectorInput=Subtract(A=@n52) {constB=0.5})
  ) {materialFunction=&HueShift}
  n54 = Subtract(A=WorldPosition().XYZ, B=Transform(None=CameraPositionWS()) {transformSourceType=TRANSFORMSOURCE_Camera})
  n55 = Normalize(VectorInput=@n54)
  n56 = NamedRerouteUsage()
  n57 = Subtract(
   A=DotProduct(A=@n53.Result, B=OneMinus(None=@n55))
   B=LinearInterpolate(B=ScalarParameter() {defaultValue=0.5, group=SandCastle, parameterName=CM_SandCastle_SparkleSubtract, sortPriority=99}, Alpha=@n56) {constA=1.1}
  )
  n58 = LinearInterpolate(
   A=ScalarParameter() {defaultValue=2.15, group=SandCastle, parameterName=CM_SandCastle_SparkleIntensityNear, sortPriority=99}
   B=ScalarParameter() {defaultValue=50, group=SandCastle, parameterName=CM_SandCastle_SparkleIntensityFar, sortPriority=99}
   Alpha=@n56
  ) {constA=2}
  n59 = NamedRerouteUsage()
  n60 = Saturate(None=@n3)
  n61 = Multiply(
   A=Power(
    Base=Multiply(A=@n57, B=@n58)
    Exp=ScalarParameter() {defaultValue=1, group=SandCastle, parameterName=CM_SandCastle_SparkleContrast, sortPriority=99}
   )
   B=Add(A=Add(A=@n59) {constB=0.05}, B=@n60)
  )
  n62 = Add(A=Power(Base=@n59) {constExponent=2.5}, B=@n60)
  NamedRerouteDeclaration(None=Multiply(A=Multiply(A=@n46, B=@n50), B=Multiply(A=@n61, B=@n62))) {name=Sparkles, nodeColor=<rgba 0.8281 1 0 1>}
  NamedRerouteDeclaration(None=ComponentMask(None=@n23.RGBA) {g=true}) {name=LargeNoise, nodeColor=<rgba 1 0 0.7969 1>}
  n63 = Power(Base=@n44) {constExponent=5}
  NamedRerouteDeclaration(None=Saturate(None=Add(A=@n63) {constB=0.1})) {name=FresnelInvert, nodeColor=<rgba 1 0.0938 0 1>}
  NamedRerouteDeclaration(None=@n33.RGB) {name=SparkleColor, nodeColor=<rgba 0 0.1328 1 1>}
  NamedRerouteDeclaration(None=@n13.Result) {name=CameraDepth, nodeColor=<rgba 0 0.3203 1 1>}
  NamedRerouteDeclaration(None=@n32) {name=Finalcolor, nodeColor=<rgba 0 0.5781 1 1>}
  NamedRerouteDeclaration(None=ComponentMask(None=MaterialFunctionCall(TextureObject=@n15, TextureSize=@n16, WorldPosition=@n17, World Space Normal=@n18, ProjectionTransitionContrast=@n20) {materialFunction=&MF_Primitive_TriplanarUVSimple}.XYZ Texture) {g=true}) {name=AO, nodeColor=<rgba 1 0.1172 0 1>}
  n64 = ScalarParameter() {defaultValue=0.5, group=SandCastle, parameterName=CM_SandCastle_AOMin}
  n65 = ScalarParameter() {defaultValue=1, group=SandCastle, parameterName=CM_SandCastle_AOMax}
  n66 = NamedRerouteUsage()
  n67 = NamedRerouteUsage()
  n68 = ScalarParameter() {defaultValue=1.25, group=SandCastle, parameterName=CM_SandCastle_AORoughnessContrast}
  n69 = Add(
   A=OneMinus(None=Power(Base=@n67, Exp=@n68))
   B=ScalarParameter() {defaultValue=0.6, group=SandCastle, parameterName=CM_SandCastle_AORoughnessAdd}
  )
  NamedRerouteDeclaration(None=Multiply(A=LinearInterpolate(A=@n64, B=@n65, Alpha=@n66), B=Saturate(None=@n69))) {name=FinalAO, nodeColor=<rgba 0 1 0.8828 1>}
  n70 = CameraVectorWS()
  n71 = Normalize(VectorInput=Transform(None=@n9))
  NamedRerouteDeclaration(None=Power(Base=DotProduct(A=@n70, B=@n71), Exp=ScalarParameter() {defaultValue=8, parameterName=Power})) {name=GeneralSparkle, nodeColor=<rgba 1 0 0.8906 1>}
  NamedRerouteDeclaration(None=@n53.Result) {name=Sparkle, nodeColor=<rgba 0.8984 0 1 1>}
  NamedRerouteDeclaration(None=@n35.AmbientOcclusion) {name=InAO, nodeColor=<rgba 0 1 0.7422 1>}
  NamedRerouteDeclaration(None=@n35.Roughness) {name=InRoughness, nodeColor=<rgba 0 0.6016 1 1>}
  OneMinus(None=@n49)
  Subtract(A=@n54, B=MaterialFunctionCall() {materialFunction=&CameraDirectionVector}.Result)
  TransformPosition(None=@n40.WorldPosition) {transformSourceType=TRANSFORMPOSSOURCE_World}
 ]}
