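# Material function for Creative primitive prop customization that turns the surface into a glossy, animated Slurpy liquid when the Primitive_Slurpy static switch is on, passing input attributes through otherwise.
# Two counter-scrolling triplanar samples form waves, blended normals and warp for bubble layers.
# Color is emissive over black base, built from Fresnel gradients, a curve-atlas wave ramp, bubbles and a reflection-vector rainbow rim.
# Music amplitude and MF_Primitive_AudioSetup drive noise-based UV warp strength.
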MaterialFunction /Game/Creative/Sets/PropSets/Primitives/Customization/MaterialFunctions/MF_CM_Slurpy
graph {nodes [
  n1 = GetMaterialAttributes(None=FunctionInput() {inputType=FunctionInput_MaterialAttributes}) {attributeGetTypes [#1]}
  n2 = SetMaterialAttributes!(
   MaterialAttributes=@n1.MaterialAttributes
   BaseColor=Constant()
   EmissiveColor=NamedRerouteUsage()
   Specular=ScalarParameter() {defaultValue=0.5, group=SLURPY, parameterName=CM_SLURPY_Specular}
   Roughness=ScalarParameter() {defaultValue=0.2, group=SLURPY, parameterName=CM_SLURPY_Roughness}
   Normal=NamedRerouteUsage()
   Metallic=ScalarParameter() {group=SLURPY, parameterName=CM_SLURPY_Metallic}
  ) {attributeSetTypes [
    #2
    #3
    #4
    #5
    #1
    #6
   ]}
  FunctionOutput(None=StaticSwitchParameter(True=@n2, False=@n1.MaterialAttributes) {group=Customization, parameterName=Primitive_Slurpy})
  n3 = Fresnel(ExponentIn=ScalarParameter() {defaultValue=3, group=SLURPY, parameterName=CM_SLURPY_FrenselEdgeColor01}) {baseReflectFraction=0}
  n4 = Fresnel(ExponentIn=ScalarParameter() {defaultValue=4, group=SLURPY, parameterName=CM_SLURPY_FrenselEdgeColor02}) {baseReflectFraction=0}
  n5 = LinearInterpolate(
   A=LinearInterpolate(A=Constant3Vector() {constant=<rgba 0.008 0.0452 0.117 0>}, B=Constant3Vector() {constant=<rgba 0.0168 0.4508 0.8879 0>}, Alpha=@n3)
   B=Constant3Vector() {constant=<rgba 0 0.5972 0.2542 1>}
   Alpha=@n4
  )
  n6 = NamedRerouteUsage()
  n7 = Add(A=@n5, B=@n6)
  n8 = NamedRerouteUsage()
  LinearInterpolate(A=@n5, B=@n7, Alpha=@n8)
  Max(A=@n5, B=@n6)
  n9 = Add(
   A=Multiply(A=NamedRerouteUsage(), B=VectorParameter() {group=SLURPY, parameterName=CM_SLURPY_BubbleColor, defaultValue=<rgba 0.0056 0.3 0.1911 1>}.RGB)
   B=@n7
  )
  NamedRerouteDeclaration(None=Add(A=NamedRerouteUsage(), B=LinearInterpolate(A=@n7, B=@n9, Alpha=@n8))) {name=Emissive, nodeColor=<rgba 0.3594 1 0 1>}
  n10 = TextureObjectParameter() {
   group=SLURPY
   parameterName=CM_SLURPY_MaskTexture
   samplerType=SAMPLERTYPE_LinearColor
   texture=/Game/Creative/Sets/PropSets/Primitives/Customization/Textures/Slurpy/T_CM_Slurpy_M
  }
  n11 = ScalarParameter() {defaultValue=3, group=SLURPY, parameterName=CM_SLURPY_BaseUVTile}
  n12 = NamedRerouteUsage()
  n13 = Time()
  n14 = VectorParameter() {group=SLURPY, parameterName=CM_SLURPY_Normal01_Speed, defaultValue=<rgba 0.05 0.05 0.05 1>}
  n15 = NamedRerouteUsage()
  n16 = Constant()
  n17 = ScalarParameter() {defaultValue=4, group=SLURPY, parameterName=CM_SLURPY_BaseTriplannarContrast}
  n18 = NamedRerouteUsage()
  n19 = TextureObjectParameter() {
   group=SLURPY
   parameterName=CM_SLURPY_NormalTexture
   samplerType=SAMPLERTYPE_Normal
   texture=/Game/Creative/Sets/PropSets/Primitives/Customization/Textures/Slurpy/T_CM_Slurpy_N
  }
  n20 = ScalarParameter() {defaultValue=1, group=SLURPY, parameterName=CM_SLURPY_NormalIntensity}
  n21 = MaterialFunctionCall!(
   TextureObject_Diffuse=@n10
   TextureScale=@n11
   CalculatedBounds=Add(A=@n12, B=Multiply(A=@n13, B=@n14.RGB))
   WorldNormals=@n15
   RotationX=@n16
   RotationY=@n16
   RotationZ=@n16
   ProjectionTransitionContrast=@n17
   Material Normals=@n18
   NormalTexture=@n19
   Normal Texture Intensity=@n20
  ) {materialFunction=&/MF_Primitive_TriplanarUVs}
  n22 = ScalarParameter() {defaultValue=3, group=SLURPY, parameterName=CM_SLURPY_WavePower}
  n23 = Time()
  n24 = VectorParameter() {group=SLURPY, parameterName=CM_SLURPY_Normal02_Speed, defaultValue=<rgba -0.03 -0.03 -0.03 1>}
  n25 = Constant() {r=0.25}
  n26 = MaterialFunctionCall(
   TextureObject_Diffuse=@n10
   TextureScale=@n11
   CalculatedBounds=Add(A=@n12, B=Multiply(A=@n23, B=@n24.RGB))
   WorldNormals=@n15
   RotationX=@n25
   RotationY=@n25
   RotationZ=@n25
   ProjectionTransitionContrast=@n17
   Material Normals=@n18
   NormalTexture=@n19
   Normal Texture Intensity=@n20
  ) {materialFunction=&MF_Primitive_TriplanarUVs}
  n27 = LinearInterpolate(
   A=Power(Base=ComponentMask(None=@n21.RGBA) {g=true}, Exp=@n22) {constExponent=3}
   B=Power(Base=ComponentMask(None=@n26.RGBA) {g=true}, Exp=@n22) {constExponent=3}
  )
  NamedRerouteDeclaration(None=@n27) {name=BubbleWarp, nodeColor=<rgba 1 0 0.1641 1>}
  n28 = Multiply(A=Time(), B=VectorParameter() {group=SLURPY, parameterName=CM_SLURPY_Bubble01Speed, defaultValue=<rgba -0.0125 -0.03 0 1>}.RGB)
  n29 = NamedRerouteUsage()
  n30 = NamedRerouteUsage()
  n31 = ScalarParameter() {defaultValue=0.9, group=SLURPY, parameterName=CM_SLURPY_Bubble01WarpIntensity}
  n32 = NamedRerouteUsage()
  n33 = ScalarParameter() {defaultValue=4, group=SLURPY, parameterName=CM_SLURPY_BubblesTriplannarContrast}
  n34 = MaterialFunctionCall!(
   TextureObject=@n10
   TextureSize=ScalarParameter() {defaultValue=3, group=SLURPY, parameterName=CM_SLURPY_Bubbles01UVTile}
   WorldPosition=Add(A=Add(A=@n28, B=@n29), B=Multiply(A=@n30, B=@n31))
   World Space Normal=@n32
   ProjectionTransitionContrast=@n33
  ) {materialFunction=&/MF_Primitive_TriplanarUVSimple}
  n35 = Add(
   A=Multiply(A=Time(), B=VectorParameter() {group=SLURPY, parameterName=CM_SLURPY_Bubble02Speed, defaultValue=<rgba -0.0125 -0.03 0 1>}.RGB)
   B=@n29
  )
  n36 = Multiply(A=NamedRerouteUsage(), B=ScalarParameter() {defaultValue=0.5, group=SLURPY, parameterName=CM_SLURPY_Bubble02WarpIntensity})
  n37 = ComponentMask(None=MaterialFunctionCall(
    TextureObject=@n10
    TextureSize=ScalarParameter() {defaultValue=1, group=SLURPY, parameterName=CM_SLURPY_Bubbles02UVTile}
    WorldPosition=Add(A=@n35, B=@n36)
    World Space Normal=@n32
    ProjectionTransitionContrast=@n33
   ) {materialFunction=&MF_Primitive_TriplanarUVSimple}.XYZ Texture) {r=true}
  NamedRerouteDeclaration(None=Add(A=ComponentMask(None=@n34.XYZ Texture) {r=true}, B=Multiply(A=@n37) {constB=0.25})) {name=Bubbles01, nodeColor=<rgba 0 0.6719 1 1>}
  NamedRerouteDeclaration(None=OneMinus(None=Add(A=@n4, B=@n3))) {name=EdgeMask, nodeColor=<rgba 1 0 0.8906 1>}
  n38 = NamedRerouteUsage()
  n39 = TextureObjectParameter() {
   group=SLURPY
   parameterName=CM_SLURPY_UVWarpTexture
   samplerType=SAMPLERTYPE_Masks
   texture=/Game/Global/Textures/Default/Noise/T_WrapEdgeWear_Noise
  }
  n40 = Divide(A=@n38, B=ScalarParameter() {defaultValue=500, group=SLURPY, parameterName=CM_SLURPY_UVWarpTile})
  n41 = Multiply(A=Time(), B=VectorParameter() {group=SLURPY, parameterName=CM_SLURPY_UVWarpSpeed, defaultValue=<rgba -0.05 -0.05 0 1>}.RGB)
  n42 = ComponentMask(None=MaterialFunctionCall(
    TextureObject=@n39
    TextureSize=ScalarParameter() {defaultValue=1, group=SLURPY, parameterName=CM_SLURPY_WarpTextureSize}
    WorldPosition=Add(A=@n40, B=@n41)
    World Space Normal=NamedRerouteUsage()
    ProjectionTransitionContrast=ScalarParameter() {defaultValue=4, group=SLURPY, parameterName=CM_SLURPY_BaseTriplannarContrast}
   ) {materialFunction=&MF_Primitive_TriplanarUVSimple}.XYZ Texture) {b=true, g=true, r=true}
  n43 = LinearInterpolate(
   A=CollectionParameter() {collection=/Game/Athena/Events/AudioAnalysis/MPC_Audio_Reactive_Parameters, parameterName=Amplitude_Music}
   B=MaterialFunctionCall() {materialFunction=&/MF_Primitive_AudioSetup}.Result
   Alpha=ScalarParameter() {defaultValue=0.1, group=SLURPY, parameterName=CM_SLURPY_EventAudioBlend}
  ) {constAlpha=0.25}
  n44 = ScalarParameter() {defaultValue=1, group=SLURPY, parameterName=CM_SLURPY_EventAudioMultiplier}
  n45 = Abs(None=Min(A=Multiply(A=@n43, B=@n44)) {constB=30, desc=prevents crazy values})
  n46 = LinearInterpolate(
   A=ScalarParameter() {group=SLURPY, parameterName=CM_SLURPY_UVWarpAmount_Low}
   B=ScalarParameter() {defaultValue=0.5, group=SLURPY, parameterName=CM_SLURPY_UVWarpAmount_High}
   Alpha=NamedRerouteDeclaration(None=Saturate(None=@n45)) {name=Warp, nodeColor=<rgba 0.2188 0 1 1>}
  )
  NamedRerouteDeclaration(None=Add(A=@n38, B=Multiply(A=@n42, B=@n46))) {name=UVsWarp, nodeColor=<rgba 0 1 0.4609 1>}
  NamedRerouteDeclaration(None=NamedRerouteUsage()) {name=UVs, nodeColor=<rgba 0 1 0.4375 1>}
  NamedRerouteDeclaration(None=MaterialFunctionCall(BaseNormal=@n21.Normal Texture, AdditionalNormal=@n26.Normal Texture) {materialFunction=&BlendAngleCorrectedNormals}.Result) {name=FinalNormal, nodeColor=<rgba 0.8047 0 1 1>}
  NamedRerouteDeclaration(None=@n1.Normal) {name=InNormal, nodeColor=<rgba 0 0.2266 1 1>}
  n47 = Constant3Vector() {constant=<rgba 0.0035 0.1533 0.3601 1>}
  n48 = ComponentMask(None=ReflectionVectorWS()) {g=true, r=true}
  n49 = Fresnel() {baseReflectFraction=0, exponent=0.5}
  n50 = TextureSample(UVs=Panner(Coordinate=Multiply(A=@n48, B=@n49)) {speedX=0.25, speedY=0.25}) {texture=/Game/Creative/Sets/PropSets/Primitives/Customization/Textures/Slurpy/T_CM_Slurpy_Grad}
  n51 = ScalarParameter() {defaultValue=12, group=SLURPY, parameterName=CM_SLURPY_FrenselMaskExp}
  NamedRerouteDeclaration(None=Multiply(A=LinearInterpolate(A=@n47, B=@n50.RGB) {constA=1, constAlpha=0.75}, B=Fresnel(ExponentIn=@n51) {baseReflectFraction=0})) {name=Rainbow, nodeColor=<rgba 0 0.3672 1 1>}
  n52 = CurveAtlasRowParameter(CurveTime=Frac(None=@n27)) {
   atlas=/Game/Creative/Sets/PropSets/Primitives/Customization/Textures/ColorCurve/CC_CM_Primitives_01
   curve=/Game/Creative/Sets/PropSets/Primitives/Customization/Textures/ColorCurve/C_CM_Slurpy_01
   defaultValue=5
   group=SLURPY
   parameterName=CM_SLURPY_Gradient
  }
  NamedRerouteDeclaration(None=@n52) {name=WavesColor, nodeColor=<rgba 0 0.7656 1 1>}
  n53 = MaterialFunctionCall() {materialFunction=&/MF_Primitive_CalculatedBounds}
  NamedRerouteDeclaration(None=Divide(A=@n53.WorldPosition, B=ScalarParameter() {defaultValue=100, group=SLURPY, parameterName=CM_SLURPY_WorldScale})) {name=CalculatedBounds, nodeColor=<rgba 1 0.1406 0 1>}
  NamedRerouteDeclaration(None=@n53.World Normals) {name=WorldNorm, nodeColor=<rgba 0.7344 0 1 1>}
 ]}
