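# Material function for Creative primitive props that adds a music-reactive synthwave look behind the Primitive_SynthAudioWaves static switch, passing input attributes through when off.
# Two panning triplanar warp textures, plus a music-amplitude offset from the audio-reactive parameter collection, distort triplanar greyscale sampling.
# A color-curve atlas and time-animated hue shift produce emissive, lighten-blended with an inverted-Fresnel core glow.
# Base color becomes black; specular and roughness are flat or texture-driven.
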MaterialFunction /Game/Creative/Sets/PropSets/Primitives/Customization/MaterialFunctions/MF_CM_SynthWaveAudio
graph {nodes [
  n1 = Constant() {r=0.08}
  n2 = Constant() {r=0.01}
  n3 = Multiply(
   A=OneMinus(None=Fresnel(ExponentIn=@n1, BaseReflectFractionIn=@n2))
   B=ScalarParameter() {defaultValue=0.15, parameterName=CM_SynthWaveCoreGlow}
  )
  n4 = Sine(None=Time()) {period=5}
  n5 = Ceil(None=NamedRerouteUsage())
  n6 = LinearInterpolate(B=Divide(A=Add(A=@n4)), Alpha=@n5)
  n7 = LinearInterpolate(Alpha=NamedRerouteUsage()) {constA=0.1, constB=0.14}
  n8 = MaterialFunctionCall(Color=NamedRerouteUsage()) {materialFunction=&SplitComponents}
  n9 = ScalarParameter() {group=SynthWave, parameterName=CM_SynthWave_UseYAxis}
  n10 = ScalarParameter() {group=SynthWave, parameterName=CM_SynthWave_UseZAxis}
  n11 = CurveAtlasRowParameter(CurveTime=LinearInterpolate(A=LinearInterpolate(A=@n8.R, B=@n8.R, Alpha=@n9), B=LinearInterpolate(A=@n8.G, B=@n8.G, Alpha=@n10), Alpha=@n10)) {
   atlas=/Game/Creative/Sets/PropSets/Primitives/Customization/Textures/ColorCurve/CC_CM_Primitives_01
   curve=/Game/Creative/Sets/PropSets/Primitives/Customization/Textures/ColorCurve/C_CM_SynthWave_01
   defaultValue=2
   group=SynthWave
   parameterName=CM_SynthWave_ColorCurve
  }
  n12 = MaterialFunctionCall(Hue Shift Percentage=LinearInterpolate(A=@n6, B=Add(A=@n6, B=@n7), Alpha=@n5), Texture=@n11) {materialFunction=&HueShift}
  n13 = NamedRerouteUsage()
  n14 = FunctionInput(Preview=SetMaterialAttributes!(
    MaterialAttributes
    BaseColor=Constant3Vector() {constant=<rgba 1 0 0 0>}
    Normal=TextureSample() {samplerType=SAMPLERTYPE_Normal, texture=/Game/Weapons/FORT_Rifles/T_SCAR_N}.RGB
   ) {attributeSetTypes [#1, #2]}) {inputType=FunctionInput_MaterialAttributes}
  n15 = SetMaterialAttributes!(
   MaterialAttributes=@n14
   BaseColor=Constant3Vector()
   EmissiveColor=MaterialFunctionCall!(Blend=Multiply(A=@n3, B=@n12.Result), Base=LinearInterpolate(B=@n12.Result, Alpha=@n13)) {materialFunction=/Engine/Functions/Engine_MaterialFunctions03/Blends/Blend_Lighten}.Result
   Specular=ComponentMask(None=NamedRerouteUsage()) {r=true}
   Roughness=ComponentMask(None=NamedRerouteUsage()) {r=true}
  ) {attributeSetTypes [
    #1
    #3
    #4
    #5
   ]}
  FunctionOutput(None=StaticSwitchParameter(True=@n15, False=@n14) {group=Customization, parameterName=Primitive_SynthAudioWaves})
  n16 = TextureObjectParameter() {
   group=GasCloud
   parameterName=CM_SynthWave_WarpTex
   samplerType=SAMPLERTYPE_LinearGrayscale
   texture=/Game/Creative/Sets/PropSets/Primitives/Customization/Textures/SynthAudioWaves/T_CM_WarpTexture_03
  }
  n17 = Time()
  n18 = Constant3Vector() {constant=<rgba 0.0621 0.0921 0 0>}
  n19 = VectorParameter() {group=GasCloud, parameterName=CM_SynthWave_RotationWarpLarge, defaultValue=<rgba 0 0 0 1>}
  n20 = MaterialFunctionCall!(
   TextureObject_Diffuse=@n16
   TextureScale=ScalarParameter() {defaultValue=1, group=SynthWave, parameterName=CM_SynthWave_WarpUVTile_Large}
   CalculatedBounds=Add(A=Multiply(A=@n17, B=@n18), B=NamedRerouteUsage())
   WorldNormals=NamedRerouteUsage()
   RotationX=@n19.R
   RotationY=@n19.R
   RotationZ=@n19.R
   ProjectionTransitionContrast=ScalarParameter() {defaultValue=1, group=SynthWave, parameterName=CM_SynthWave_Triplannar_Contrast}
   Material Normals
   NormalTexture
   Normal Texture Intensity
  ) {materialFunction=&/MF_Primitive_TriplanarUVs}
  NamedRerouteDeclaration(None=Multiply(
    A=ComponentMask(None=@n20.RGBA) {r=true}
    B=ScalarParameter() {defaultValue=1, group=SynthWave, parameterName=CM_Synthwave_LargeWarpIntensity}
   )) {name=Warp UVS_Large, nodeColor=<rgba 1 0 0.375 1>}
  n21 = NamedRerouteUsage()
  n22 = ScalarParameter() {defaultValue=1, group=SynthWave, parameterName=CM_SynthAudioWaves_TextSpecInst, sliderMax=1}
  NamedRerouteDeclaration(None=LinearInterpolate(
    A=ScalarParameter() {group=SynthWave, parameterName=CM_SynthAudioWaves_FlatSpec}
    B=Multiply(A=@n21, B=@n22)
    Alpha=ScalarParameter() {group=SynthWave, parameterName=CM_Sythwave_Use_TextureforSpec}
   )) {name=Spec, nodeColor=<rgba 0 0.4072 0.3968 1>}
  n23 = CollectionParameter() {collection=/Game/Athena/Events/AudioAnalysis/MPC_Audio_Reactive_Parameters, parameterName=Amplitude_Music}
  NamedRerouteDeclaration(None=@n23) {name=AmplitudeMusic, nodeColor=<rgba 0 1 0.3906 1>}
  n24 = NamedRerouteUsage()
  n25 = ScalarParameter() {defaultValue=1, group=SynthWave, parameterName=CM_SynthAudioWaves_TextRoughInst, sliderMax=1}
  NamedRerouteDeclaration(None=LinearInterpolate(
    A=ScalarParameter() {group=SynthWave, parameterName=CM_SynthAudioWaves_FlatRough}
    B=Multiply(A=@n24, B=@n25)
    Alpha=ScalarParameter() {group=SynthWave, parameterName=CM_SynthWave_Use_TextureforRough}
   )) {name=Rough, nodeColor=<rgba 0.4072 0.2831 0 1>}
  n26 = TextureObject() {texture=/Game/Creative/Sets/PropSets/Primitives/Customization/Textures/SynthAudioWaves/T_CM_SynthWave_BC}
  n27 = ScalarParameter() {defaultValue=1, group=SynthWave, parameterName=CM_SynthWave_UVTile}
  n28 = NamedRerouteUsage()
  n29 = NamedRerouteUsage()
  n30 = Multiply(A=ScalarParameter() {defaultValue=1, group=SynthWave, parameterName=CM_SynthWave_MusicAmp_Mult}, B=@n23)
  n31 = Ceil(None=@n23)
  n32 = MaterialFunctionCall!(In=Min(A=LinearInterpolate(B=@n30, Alpha=@n31) {constA=1}) {constB=10}, Ramp Width) {materialFunction=/Engine/Functions/Engine_MaterialFunctions01/Math/SmoothCeil}
  n33 = Add(A=NamedRerouteUsage(), B=Add(A=Add(A=@n28, B=@n29), B=LinearInterpolate(Alpha=@n32.Result) {constA=0.04, constB=-0.04}))
  n34 = NamedRerouteUsage()
  n35 = VectorParameter() {group=GasCloud, parameterName=CM_GasCloud_Rotation, defaultValue=<rgba 0 0 0 1>}
  n36 = ScalarParameter() {defaultValue=1, group=SynthWave, parameterName=CM_SynthWave_Triplannar_Contrast}
  NamedRerouteDeclaration(None=ComponentMask(None=MaterialFunctionCall(
     TextureObject_Diffuse=@n26
     TextureScale=@n27
     CalculatedBounds=@n33
     WorldNormals=@n34
     RotationX=@n35.R
     RotationY=@n35.R
     RotationZ=@n35.R
     ProjectionTransitionContrast=@n36
    ) {materialFunction=&MF_Primitive_TriplanarUVs}.RGBA) {r=true}) {name=Warped Greyscale, nodeColor=<rgba 0.4072 0.0437 0.0529 1>}
  n37 = TextureObjectParameter() {
   group=GasCloud
   parameterName=CM_SynthWave_WarpTex
   samplerType=SAMPLERTYPE_Masks
   texture=/Game/Creative/Sets/PropSets/Primitives/Customization/Textures/SynthAudioWaves/T_CM_WarpTexture_02
  }
  n38 = Time()
  n39 = Constant3Vector() {constant=<rgba 0.04 -0.04 0 0>}
  n40 = MaterialFunctionCall() {materialFunction=&/MF_Primitive_CalculatedBounds}
  n41 = Divide(A=@n40.WorldPosition, B=ScalarParameter() {defaultValue=250, group=SynthWave, parameterName=CM_SynthWave_WorldSpace})
  n42 = VectorParameter() {group=GasCloud, parameterName=CM_GasCloud_Rotation, defaultValue=<rgba 0 0 0 1>}
  n43 = MaterialFunctionCall(
   TextureObject_Diffuse=@n37
   TextureScale=ScalarParameter() {defaultValue=1, group=SynthWave, parameterName=CM_SynthWave_WarpUVTile}
   CalculatedBounds=Add(A=Multiply(A=@n38, B=@n39), B=NamedRerouteDeclaration(None=@n41) {name=CalculatedBounds, nodeColor=<rgba 1 0.1406 0 1>})
   WorldNormals=NamedRerouteUsage()
   RotationX=@n42.R
   RotationY=@n42.R
   RotationZ=@n42.R
   ProjectionTransitionContrast=ScalarParameter() {defaultValue=1, group=SynthWave, parameterName=CM_SynthWave_Triplannar_Contrast}
  ) {materialFunction=&MF_Primitive_TriplanarUVs}
  NamedRerouteDeclaration(None=Multiply(A=ComponentMask(None=@n43.RGBA) {r=true}, B=ScalarParameter() {defaultValue=1, group=SynthWave, parameterName=CM_Synthwave_WarpIntensity})) {name=Warp UVS, nodeColor=<rgba 1 0 0.375 1>}
  NamedRerouteDeclaration(None=@n40.World Normals) {name=WorldNormal, nodeColor=<rgba 0.4766 0 1 1>}
  NamedRerouteDeclaration(None=@n40.Calculated Bounds) {name=Calc, nodeColor=<rgba 0 1 0.0859 1>}
  NamedRerouteDeclaration(None=@n40.LocalPosition) {name=local, nodeColor=<rgba 1 0 0.2344 1>}
  NamedRerouteUsage()
 ]}
